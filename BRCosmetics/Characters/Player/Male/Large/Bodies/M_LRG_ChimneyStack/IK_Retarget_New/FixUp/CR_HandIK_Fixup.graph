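# Control Rig for the M_LRG_ChimneyStack player body that fixes up hand IK, judging by its name and folder, after IK retargeting.
# A sequence runs three toggleable stages: global uniform scaling of listed bones by Weapon Scale, snapping ik_hand_gun between hand_l and hand_r by a retargeting weight, then an IK Rig solve.
# The snap propagates to child bones ik_hand_l and ik_hand_r, whose transforms become the IK Rig goals, with per-hand alphas capped by a shared clamp.
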ControlRigBlueprint /BRCosmetics/Characters/Player/Male/Large/Bodies/M_LRG_ChimneyStack/IK_Retarget_New/FixUp/CR_HandIK_Fixup
graphs [
 RigVMFunctionLibrary {links [], nodes []}
 RigVMFunction_Sequence_1_ContainedGraph {nodes [
   n1 = RigVMFunctionEntryNode()
   RigVMFunction_Sequence = &RigVMFunction_Sequence(ExecuteContext=@n1.ExecuteContext)
   n2 = &RigVMFunction_Sequence(ExecuteContext=@RigVMFunction_Sequence.B)
   RigVMFunctionReturnNode(A=@RigVMFunction_Sequence.A, B=@n2.A, C=@n2.B)
  ]}
 RigVMModel {nodes [
   RigUnit_BeginExecution = &RigUnit_BeginExecution()
   RigVMFunction_Sequence = RigVMAggregateNode(ExecuteContext=@RigUnit_BeginExecution.ExecutePin)
   RigVMFunction_ControlFlowBranch = &RigVMFunction_ControlFlowBranch(ExecuteContext=@RigVMFunction_Sequence.A, Condition=RigVMVariableNode() {variable=Enable Weapon Scale}.Value)
   For_Each = RigVMDispatchNode(ExecuteContext=@RigVMFunction_ControlFlowBranch.True, Array=RigVMVariableNode() {variable=SetScale_Item_Name}.Value) {template="DISPATCH_RigVMDispatch_ArrayIterator(in Array,out Element,out Index,out Count,out Ratio)"}
   n1 = RigVMVariableNode() {variable=Weapon Scale}
   > /Script/ControlRig.RigUnit_SetScale(
    Item=(Type=Bone,Name="None")
    Scale=(X=1,Y=1,Z=1)
    Space=GlobalSpace
    Weight=1
    bInitial=False
    bPropagateToChildren=true
    Item.Name=@For_Each.Element
    Scale.X=@n1.Value
    Scale.Y=@n1.Value
    Scale.Z=@n1.Value
   ) {is=struct RigVMUnitNode}
   n2 = &RigVMFunction_ControlFlowBranch(Condition=RigVMVariableNode() {variable=Snap IK Hand Gun}.Value, ExecuteContext=@RigVMFunction_Sequence.B)
   Interpolate = /Script/RigVM.RigVMFunction_MathVectorLerp(
    A=&RigUnit_GetTransform(Item=(Type=Bone,Name="hand_l"), bInitial=false).Transform.Translation
    B=&RigUnit_GetTransform(Item=(Type=Bone,Name="hand_r"), bInitial=false).Transform.Translation
    T=RigVMVariableNode() {variable=Hand IK Retargeting Weight}.Value
   ) {is=struct RigVMUnitNode, template="Interpolate::Execute(in A,in B,in T,out Result)"}
   RigUnit_SetTranslation = /Script/ControlRig.RigUnit_SetTranslation(
    Item=(Type=Bone,Name="ik_hand_gun")
    Space=GlobalSpace
    Weight=1
    bInitial=False
    bPropagateToChildren=true
    Value=@Interpolate.Result
    ExecutePin=@n2.True
   ) {is=struct RigVMUnitNode, template="Set Transform::Execute(in Item,in Space,in bInitial,in Value,in Weight,in bPropagateToChildren)"}
   n3 = &RigVMFunction_ControlFlowBranch(Condition=RigVMVariableNode() {variable=Run IK}.Value, ExecuteContext=@RigVMFunction_Sequence.C)
   Clamp = &RigVMFunction_MathDoubleClamp(Value=RigVMVariableNode() {variable=Left Hand IK Alpha}.Value, Maximum=RigVMVariableNode() {variable=Hand IK Alpha Clamp}.Value)
   n4 = &RigVMFunction_MathDoubleClamp(Maximum=RigVMVariableNode() {variable=Hand IK Alpha Clamp}.Value, Value=RigVMVariableNode() {variable=Right Hand IK Alpha}.Value)
   RigUnit_GetTransform = &RigUnit_GetTransform(Item=(Type=Bone,Name="ik_hand_l"), bInitial=false)
   Make_Array = RigVMDispatchNode(
    Values="((GoalName=\"LeftHandIK\",Transform=(Rotation=(X=0,Y=0,Z=0,W=1),Translation=(X=0,Y=0,Z=0),Scale3D=(X=1,Y=1,Z=1)),PositionAlpha=1,RotationAlpha=1),(GoalName=\"RighthandIK\",Transform=(Rotation=(X=0,Y=0,Z=0,W=1),Translation=(X=0,Y=0,Z=0),Scale3D=(X=1,Y=1,Z=1)),P ...(truncated)"
    Values.0.PositionAlpha=@Clamp.Result
    Values.0.RotationAlpha=@Clamp.Result
    Values.1.PositionAlpha=@n4.Result
    Values.1.RotationAlpha=@n4.Result
    Values.0.Transform=@RigUnit_GetTransform.Transform
    Values.1.Transform=&RigUnit_GetTransform(Item=(Type=Bone,Name="ik_hand_r"), bInitial=false).Transform
   ) {template="DISPATCH_RigVMDispatch_ArrayMake(in Values,out Array)"}
   IKRig = /Script/IKRig.RigUnit_IKRig(Goals=@Make_Array.Array, ExecutePin=@n3.True, IKRigAsset=RigVMVariableNode() {variable=IKRig_IKRigAsset}.Value) {is=struct RigVMUnitNode}
  ]}
]
hierarchy [
 Root {loc=<0 0 0>, rot=<0 0 0 1>}
 pelvis {parent=Root, loc=<0 -0.7071 93.1191>, rot=<0.0127 0.707 -0.0127 -0.707>}
 Spine_01 {parent=pelvis, loc=<6.7494 -0.6388 0>, rot=<0 0 0.034 -0.9994>}
 Spine_02 {parent=Spine_01, loc=<7.2518 -0.035 0>, rot=<0 0 0.0238 -0.9997>}
 Spine_03 {parent=Spine_02, loc=<6.0104 0 0>, rot=<0 0 0.0149 -0.9999>}
 spine_04 {parent=Spine_03, loc=<8.1024 0 0>, rot=<0 0 -0.0132 -0.9999>}
 spine_05 {parent=spine_04, loc=<10.0146 0 0>, rot=<0 0 -0.0015 -1>}
 clavicle_l {parent=spine_05, loc=<4.3575 -0.1466 -2.3223>, rot=<0.0465 -0.7289 -0.009 -0.6829>}
 upperarm_l {parent=clavicle_l, loc=<15.5111 0 0>, rot=<0.0305 -0.536 -0.0057 -0.8437>}
 lowerarm_l {parent=upperarm_l, loc=<26.6514 0 0>, rot=<0.0391 -0.0099 -0.2453 -0.9686>}
 hand_l {parent=lowerarm_l, loc=<22.3491 0 0>, rot=<-0.7081 -0.0507 0.0975 -0.6975>}
 index_metacarpal_l {parent=hand_l, loc=<1.6535 0.0183 -2.96>, rot=<0.0158 -0.0124 -0.0173 -0.9996>}
 index_01_l {parent=index_metacarpal_l, loc=<8.3008 0.0357 -0.0524>, rot=<-0.0044 0.0485 0.0924 -0.9945>}
 index_02_l {parent=index_01_l, loc=<3.3587 -0.3004 0.0452>, rot=<-0.002 0.0027 0.3225 -0.9466>}
 index_03_l {parent=index_02_l, loc=<2.5577 -0.1233 -0.0252>, rot=<-0.001 0.0068 0.2558 -0.9667>}
 middle_metacarpal_l {parent=hand_l, loc=<2.1856 -0.1725 -0.5984>, rot=<0.006 0.0168 -0.0329 -0.9993>}
 middle_01_l {parent=middle_metacarpal_l, loc=<7.3275 -0.2447 0.0129>, rot=<-0.0208 0.0042 0.0858 -0.9961>}
 middle_02_l {parent=middle_01_l, loc=<4.0131 0.0722 0.0029>, rot=<0 0 0.3511 -0.9363>}
 middle_03_l {parent=middle_02_l, loc=<2.6079 0.1152 0.0049>, rot=<0 0 0.1707 -0.9853>}
 Water_Capsule_2_3 {parent=middle_01_l, type=Null, loc=<0 0 0>, rot=<0 0 0 1>, scale=<5 5 5>}
 pinky_metacarpal_l {parent=hand_l, loc=<1.7616 -0.9753 3.3814>, rot=<-0.0397 0.0814 -0.069 -0.9935>}
 pinky_01_l {parent=pinky_metacarpal_l, loc=<7.1709 -0.4831 -0.1094>, rot=<-0.0654 -0.0035 0.2242 -0.9723>}
 pinky_02_l {parent=pinky_01_l, loc=<2.8274 0.0893 -0.1043>, rot=<0.0037 0.0011 0.2854 -0.9584>}
 pinky_03_l {parent=pinky_02_l, loc=<2.2046 0.0307 -0.0078>, rot=<-0.0002 0.0098 0.1445 -0.9895>}
 ring_metacarpal_l {parent=hand_l, loc=<2.2074 -0.3625 1.3638>, rot=<-0.0284 0.0722 -0.0458 -0.9959>}
 ring_01_l {parent=ring_metacarpal_l, loc=<7.0295 -0.6197 -0.1446>, rot=<-0.0489 -0.0252 0.1892 -0.9804>}
 ring_02_l {parent=ring_01_l, loc=<3.6174 0.4647 -0.1027>, rot=<0 0 0.2806 -0.9598>}
 ring_03_l {parent=ring_02_l, loc=<2.4152 0.114 0.0199>, rot=<0 0 0.1189 -0.9929>}
 thumb_01_l {parent=hand_l, loc=<1.3118 -0.3191 -2.459>, rot=<0.6546 -0.4042 0.0664 -0.6354>}
 thumb_02_l {parent=thumb_01_l, loc=<5.9336 -0.0448 0.0401>, rot=<-0.0037 0.0213 0.298 -0.9543>}
 thumb_03_l {parent=thumb_02_l, loc=<2.2223 0.3682 -0.0907>, rot=<0.0118 -0.013 0.0066 -0.9998>}
 weapon_l {parent=hand_l, loc=<-0.1974 -3.4111 0.3811>, rot=<0 0 0 1>}
 BluePrintPaper {parent=weapon_l, type=Null, loc=<0 0 0>, rot=<0 0 0 1>}
 LeftHand {parent=weapon_l, type=Null, loc=<0 0 0>, rot=<0 0 -0.7071 0.7071>}
 Melee_L_Upper {parent=weapon_l, type=Null, loc=<0.0002 0 -107.0446>, rot=<0 0 0 1>}
 Melee_L_Lower {parent=weapon_l, type=Null, loc=<0.0014 0.0002 -35.4661>, rot=<0 0 0 1>}
 Melee_LSword_Lower {parent=weapon_l, type=Null, loc=<0 0.0002 42>, rot=<0 0 0 1>}
 Melee_LSword_Upper {parent=weapon_l, type=Null, loc=<0 -0.0004 99>, rot=<0 0 0 1>}
 FX_SoulSuck_L {parent=weapon_l, type=Null, loc=<0 3.3208 0>, rot=<-0.7071 0 0 0.7071>}
 hand_attach_l {parent=hand_l, loc=<8.5 -3 0>, rot=<0.7071 0.7071 0 0>}
 FX_Snap {parent=hand_l, type=Null, loc=<18.825 -12.0082 -0.9417>, rot=<-0.5914 0.3535 -0.4031 0.6023>}
 hand_pickaxe_l {parent=hand_l, type=Null, loc=<6 0 0>, rot=<-0.7071 0 -0.7071 0>}
 lowerarm_twist_l {parent=lowerarm_l, loc=<26.946 0.9829 0.7898>, rot=<-0.4993 -0.0461 0.0438 -0.8641>}
 lowerarm_twist_01_l {parent=lowerarm_l, loc=<17.8549 0 0>, rot=<0 0 0 1>}
 lowerarm_twist_02_l {parent=lowerarm_l, loc=<9.6199 0 0>, rot=<0 0 0 1>}
 SM_L_ArmFore {parent=lowerarm_l, type=Null, loc=<0 0 0>, rot=<-0.4789 -0.2712 0.835 0>}
 ConstructorShield {parent=lowerarm_l, type=Null, loc=<16.2696 -0.0002 0>, rot=<-0.3196 -0.3196 -0.6308 0.6308>}
 Water_Capsule_2_2 {parent=lowerarm_l, type=Null, loc=<0 0 0>, rot=<0 0 0 1>, scale=<5 5 5>}
 FX_Contrail_Hand_L {parent=lowerarm_l, type=Null, loc=<21.1502 -0.0002 0>, rot=<-1 0 0 0>}
 lowerarm_weapon_l {parent=lowerarm_l, type=Null, loc=<10 0 0>, rot=<-0.5 0.5 -0.5 0.5>}
 wrist_pickaxe_l {parent=lowerarm_l, type=Null, loc=<31 1 -0.5>, rot=<0 -0.7934 -0.6088 0>}
 upperarm_twist_l {parent=upperarm_l, loc=<0 0 0>, rot=<0 0 0 1>}
 upperarm_twist_01_l {parent=upperarm_l, loc=<0.0041 0 0>, rot=<0 0 0 1>}
 upperarm_twist_02_l {parent=upperarm_l, loc=<13.7941 0 0>, rot=<0 0 0 1>}
 SM_L_Shoulder {parent=upperarm_l, type=Null, loc=<0.6431 -0.0984 0.8326>, rot=<-0.4762 -0.0286 0.8786 0.0225>}
 Water_Capsule_2_1 {parent=upperarm_l, type=Null, loc=<0 0 0>, rot=<0 0 0 1>, scale=<5 5 5>}
 clavicle_r {parent=spine_05, loc=<4.3575 -0.1466 2.3223>, rot=<0.7289 0.0465 -0.6829 0.009>}
 upperarm_r {parent=clavicle_r, loc=<-15.5111 0 0>, rot=<0.0305 -0.536 -0.0057 -0.8437>}
 lowerarm_r {parent=upperarm_r, loc=<-26.6515 0 0>, rot=<0.0391 -0.0099 -0.2453 -0.9686>}
 hand_r {parent=lowerarm_r, loc=<-22.3492 0 0>, rot=<-0.7081 -0.0507 0.0975 -0.6975>}
 index_metacarpal_r {parent=hand_r, loc=<-1.6534 -0.0181 2.9601>, rot=<0.0158 -0.0124 -0.0173 -0.9996>}
 index_01_r {parent=index_metacarpal_r, loc=<-8.3008 -0.0357 0.0524>, rot=<-0.0044 0.0485 0.0924 -0.9945>}
 index_02_r {parent=index_01_r, loc=<-3.3589 0.3003 -0.0453>, rot=<-0.002 0.0027 0.3225 -0.9466>}
 index_03_r {parent=index_02_r, loc=<-2.5578 0.1231 0.0251>, rot=<-0.002 0.0067 0.2558 -0.9667>}
 middle_metacarpal_r {parent=hand_r, loc=<-2.1856 0.1726 0.5984>, rot=<0.006 0.0168 -0.0329 -0.9993>}
 middle_01_r {parent=middle_metacarpal_r, loc=<-7.3273 0.2447 -0.0129>, rot=<-0.0208 0.0042 0.0858 -0.9961>}
 middle_02_r {parent=middle_01_r, loc=<-4.0142 -0.1411 -0.07>, rot=<-0.0217 0.0259 0.3512 -0.9357>}
 middle_03_r {parent=middle_02_r, loc=<-2.6074 -0.0534 -0.1312>, rot=<0.0083 0.0188 0.1708 -0.9851>}
 FX_Candy {parent=middle_01_r, type=Null, loc=<0 0 0>, rot=<0.0534 0.2587 0.7482 0.6086>}
 Water_Capsule_3_3 {parent=middle_01_r, type=Null, loc=<0 0 0>, rot=<0 0 0 1>, scale=<5 5 5>}
 pinky_metacarpal_r {parent=hand_r, loc=<-1.7617 0.9753 -3.3814>, rot=<-0.0397 0.0814 -0.069 -0.9935>}
 pinky_01_r {parent=pinky_metacarpal_r, loc=<-7.1707 0.4832 0.1094>, rot=<-0.0654 -0.0035 0.2242 -0.9723>}
 pinky_02_r {parent=pinky_01_r, loc=<-2.8274 -0.0892 0.1043>, rot=<0.0037 0.0011 0.2854 -0.9584>}
 pinky_03_r {parent=pinky_02_r, loc=<-2.2045 -0.031 0.0078>, rot=<-0.0002 0.0098 0.1445 -0.9895>}
 ring_metacarpal_r {parent=hand_r, loc=<-2.2073 0.3627 -1.3638>, rot=<-0.0284 0.0722 -0.0458 -0.9959>}
 ring_01_r {parent=ring_metacarpal_r, loc=<-7.0297 0.6197 0.1446>, rot=<-0.0489 -0.0252 0.1892 -0.9804>}
 ring_02_r {parent=ring_01_r, loc=<-3.6171 -0.4648 0.1027>, rot=<0 0 0.2806 -0.9598>}
 ring_03_r {parent=ring_02_r, loc=<-2.4153 -0.1141 -0.0199>, rot=<0 0 0.1189 -0.9929>}
 thumb_01_r {parent=hand_r, loc=<-1.3119 0.3192 2.459>, rot=<0.6546 -0.4042 0.0664 -0.6354>}
 thumb_02_r {parent=thumb_01_r, loc=<-5.9335 0.0447 -0.0401>, rot=<-0.0037 0.0213 0.298 -0.9543>}
 thumb_03_r {parent=thumb_02_r, loc=<-2.2223 -0.3682 0.0907>, rot=<0.0118 -0.013 0.0066 -0.9998>}
 weapon_r {parent=hand_r, loc=<0.1975 3.4112 -0.3811>, rot=<0 0 0 -1>}
 RightHand {parent=weapon_r, type=Null, loc=<0 0 0>, rot=<0 0 -0.7071 0.7071>}
 BluePrintPen {parent=weapon_r, type=Null, loc=<0 0 0>, rot=<0 0 0 1>}
 Melee_R_Upper {parent=weapon_r, type=Null, loc=<0 0 52>, rot=<0 0 0 1>}
 Melee_R_Lower {parent=weapon_r, type=Null, loc=<0 0 -15>, rot=<0 0 0 1>}
 Melee_RSword_Lower {parent=weapon_r, type=Null, loc=<-0.0004 0 42>, rot=<0 0 0 1>}
 Melee_RSword_Upper {parent=weapon_r, type=Null, loc=<-0.0001 -0.0001 99>, rot=<0 0 0 1>}
 FX_SoulSuck_R {parent=weapon_r, type=Null, loc=<0 3.3208 -0.0002>, rot=<-0.7071 0 0 0.7071>}
 Melee_R_UpperLeft {parent=weapon_r, type=Null, loc=<-16 -6.6 55.0001>, rot=<0 0 0.11 0.9939>}
 Melee_R_UpperRight {parent=weapon_r, type=Null, loc=<-16 6.6 55>, rot=<0 0 0.8191 0.5737>}
 FX_Melee_R_Carbide {parent=weapon_r, type=Null, loc=<0 -10 40.5>, rot=<0 0 0 1>}
 FX_Melee_R_Carbide1 {parent=weapon_r, type=Null, loc=<0 10 40.5>, rot=<0 0 0 1>}
 FX_Melee_R_Speedy {parent=weapon_r, type=Null, loc=<0 0 67>, rot=<0 0 0 1>}
 FX_Melee_R_Speedy1 {parent=weapon_r, type=Null, loc=<0 0 27>, rot=<0 0 0 1>}
 Melee_IBlade_Lower {parent=weapon_r, type=Null, loc=<-0.0004 0 42>, rot=<0 0 0 1>}
 Melee_IBlade_Upper {parent=weapon_r, type=Null, loc=<-0.0001 -0.0019 138.6462>, rot=<0 0 0 1>}
 hand_attach_r {parent=hand_r, loc=<-8.5 3 0>, rot=<0 0 -0.7071 -0.7071>}
 weapon_r_socket {parent=hand_r, type=Null, loc=<0.1975 3.4112 -0.3811>, rot=<0 0 0 1>}
 FX_EmoteSalt {parent=hand_r, type=Null, loc=<-9.4677 -0.0042 2.5096>, rot=<-0.5048 0.5826 -0.2599 0.5816>}
 hand_pickaxe_r {parent=hand_r, type=Null, loc=<-6 0 0>, rot=<0 -0.7071 0 0.7071>}
 hand_QuickMelee_r {parent=hand_r, type=Null, loc=<-7 2.5 1>, rot=<0 -0.2772 0 0.9608>}
 lowerarm_twist_r {parent=lowerarm_r, loc=<-26.9461 -0.983 -0.7899>, rot=<-0.4993 -0.0461 0.0438 -0.8641>}
 lowerarm_twist_01_r {parent=lowerarm_r, loc=<-17.855 0 0>, rot=<0 0 0 1>}
 dyn_hose01 {parent=lowerarm_twist_01_r, loc=<0 0 0>, rot=<0 0 0 1>}
 lowerarm_twist_02_r {parent=lowerarm_r, loc=<-9.6199 0 0>, rot=<0 0 0 1>}
 SM_R_ArmFore {parent=lowerarm_r, type=Null, loc=<-0.3441 2.1005 1.4988>, rot=<-0.8596 0.2166 -0.4548 0.086>}
 FX_HandBeam_R {parent=lowerarm_r, type=Null, loc=<-36.4997 0 0>, rot=<0 -1 0 0>}
 Water_Capsule_3_2 {parent=lowerarm_r, type=Null, loc=<0 0 0>, rot=<0 0 0 1>, scale=<5 5 5>}
 FX_Contrail_Hand_R {parent=lowerarm_r, type=Null, loc=<-20.0926 0.0005 0>, rot=<0 -1 0 0>}
 lowerarm_weapon_r {parent=lowerarm_r, type=Null, loc=<-10 0 0>, rot=<0.5 -0.5 -0.5 0.5>}
 wrist_pickaxe_r {parent=lowerarm_r, type=Null, loc=<-31 -1 0.5>, rot=<0.6088 0 0 0.7934>}
 upperarm_twist_r {parent=upperarm_r, loc=<0 0 0>, rot=<0 0 0 1>}
 upperarm_twist_01_r {parent=upperarm_r, loc=<-0.0041 0 0>, rot=<0 0 0 1>}
 mech_arm_root {parent=upperarm_twist_01_r, loc=<-5.3638 0.8311 -15.0265>, rot=<0.7836 -0.0189 0.6208 -0.015>}
 mech_arm_upperarm {parent=mech_arm_root, loc=<-12.4844 -1.6674 -1.2368>, rot=<-0.6111 0.0041 -0.7909 0.0316>}
 mech_arm_lowerArm {parent=mech_arm_upperarm, loc=<-29.0827 -1.2816 0.0087>, rot=<0.0129 0.0027 0.1144 0.9933>}
 mech_arm_claw_root {parent=mech_arm_lowerArm, loc=<-20.8105 4.7837 -5.722>, rot=<0.1152 -0.6864 0.7002 -0.1587>}
 mech_arm_claw_spin {parent=mech_arm_claw_root, loc=<0.0096 2.6231 0.1286>, rot=<-0.0245 0.0134 -0.0209 0.9994>}
 mech_arm_claw_base {parent=mech_arm_claw_spin, loc=<0.2755 3.2046 -0.1306>, rot=<0.5319 -0.4815 -0.4893 0.4958>}
 mech_arm_clawA {parent=mech_arm_claw_base, loc=<0.1056 -9.7661 -0.0952>, rot=<-0.9999 -0.016 0 0>}
 mech_arm_clawB {parent=mech_arm_claw_base, loc=<-0.4809 9.7861 0>, rot=<0 0 0.014 0.9999>}
 mech_arm_pistonBottom {parent=mech_arm_lowerArm, loc=<14.5396 13.0442 0.0895>, rot=<0.6536 -0.2514 0.6743 0.2344>}
 mech_arm_pistonTop {parent=mech_arm_upperarm, loc=<-0.852 6.7365 0.2039>, rot=<-0.1369 -0.6921 -0.1597 0.6904>}
 mech_arm_upArmTwist {parent=mech_arm_upperarm, loc=<0 0 0>, rot=<0 0 0 1>}
 upperarm_twist_02_r {parent=upperarm_r, loc=<-13.7942 0 0>, rot=<0 0 0 1>}
 SM_R_Shoulders {parent=upperarm_r, type=Null, loc=<-1.8976 0.0001 3.162>, rot=<-0.8777 0.0864 -0.4691 0.0462>}
 Water_Capsule_3_1 {parent=upperarm_r, type=Null, loc=<0 0 0>, rot=<0 0 0 1>, scale=<5 5 5>}
 neck_01 {parent=spine_05, loc=<11.3138 0.1529 0>, rot=<0 0 -0.2003 -0.9797>}
 neck_02 {parent=neck_01, loc=<4.5627 0 0>, rot=<0 0 0.0504 -0.9987>}
 head {parent=neck_02, loc=<5.3914 0 0>, rot=<0 0 0.115 -0.9934>}
 Jaw {parent=head, loc=<2.7572 -0.5862 0>, rot=<0.0082 -0.7071 -0.0082 -0.7071>}
 l_brow_inner {parent=head, loc=<5.3622 8.2744 -1.3658>, rot=<0.0082 -0.7071 -0.0082 -0.7071>}
 l_brow_mid_a {parent=head, loc=<5.6645 7.8866 -2.6785>, rot=<0.0082 -0.7071 -0.0082 -0.7071>}
 l_brow_mid_b {parent=head, loc=<5.4943 6.9734 -4.0039>, rot=<0.0082 -0.7071 -0.0082 -0.7071>}
 l_brow_outer {parent=head, loc=<4.9434 5.8827 -4.7441>, rot=<0.0082 -0.7071 -0.0082 -0.7071>}
 brow_mid {parent=head, loc=<5.4506 8.5738 0.1846>, rot=<0.0082 -0.7071 -0.0082 -0.7071>}
 l_nostril {parent=head, loc=<0.538 8.7003 -0.8582>, rot=<0.0082 -0.7071 -0.0082 -0.7071>}
 l_nose {parent=head, loc=<0.6504 8.3098 -1.6992>, rot=<0.0082 -0.7071 -0.0082 -0.7071>}
 l_naso_upper {parent=head, loc=<2.1876 7.7056 -1.4304>, rot=<0.0082 -0.7071 -0.0082 -0.7071>}
 l_naso_mid {parent=head, loc=<0.8323 6.9414 -2.6324>, rot=<0.0082 -0.7071 -0.0082 -0.7071>}
 nosebridge {parent=head, loc=<4.146 8.7255 -0.0051>, rot=<0.0082 -0.7071 -0.0082 -0.7071>}
 Nose {parent=head, loc=<0.8985 9.4805 0>, rot=<0.0082 -0.7071 -0.0082 -0.7071>}
 l_ear {parent=head, loc=<1.9022 0.054 -5.772>, rot=<0.0082 -0.7071 -0.0082 -0.7071>}
 l_squint_mid_b {parent=head, loc=<2.4771 6.6715 -3.4906>, rot=<0.0082 -0.7071 -0.0082 -0.7071>}
 l_squint_outer {parent=head, loc=<3.3388 5.7576 -4.5892>, rot=<0.0082 -0.7071 -0.0082 -0.7071>}
 l_squint_mid_a {parent=head, loc=<2.9144 7.1169 -1.812>, rot=<0.0082 -0.7071 -0.0082 -0.7071>}
 l_forehead_outer {parent=head, loc=<7.8016 6.7955 -3.3107>, rot=<0.0082 -0.7071 -0.0082 -0.7071>}
 l_forehead_inner {parent=head, loc=<7.6873 7.4321 -0.7914>, rot=<0.0082 -0.7071 -0.0082 -0.7071>}
 l_naso_lower {parent=head, loc=<-0.7022 6.0103 -3.4425>, rot=<0.0082 -0.7071 -0.0082 -0.7071>}
 l_lip_corner_outer {parent=head, loc=<-2.139 5.3057 -3.9779>, rot=<0.0082 -0.7071 -0.0082 -0.7071>}
 l_cheek {parent=head, loc=<-0.9641 3.6064 -4.8529>, rot=<0.0082 -0.7071 -0.0082 -0.7071>}
 l_cheekbone_lower {parent=head, loc=<1.375 5.5983 -4.627>, rot=<0.0082 -0.7071 -0.0082 -0.7071>}
 l_nose_wrinkle {parent=head, loc=<1.8761 8.8925 -1.2608>, rot=<0.0082 -0.7071 -0.0082 -0.7071>}
 chin_mid {parent=head, loc=<-5.8719 8.76 0>, rot=<0.0082 -0.7071 -0.0082 -0.7071>}
 chin_lower {parent=head, loc=<-6.8547 7.5583 0>, rot=<0.0082 -0.7071 -0.0082 -0.7071>}
 chin_upper {parent=head, loc=<-4.527 8.4555 0>, rot=<0.0082 -0.7071 -0.0082 -0.7071>}
 lip_lower_mid {parent=head, loc=<-2.2205 8.5933 0>, rot=<0.0082 -0.7071 -0.0082 -0.7071>}
 l_lip_lower_inner {parent=head, loc=<-2.3301 7.9013 -1.2771>, rot=<0.0082 -0.7071 -0.0082 -0.7071>}
 l_lip_lower_outer {parent=head, loc=<-2.204 6.817 -2.3122>, rot=<0.0082 -0.7071 -0.0082 -0.7071>}
 l_lip_corner {parent=head, loc=<-2.203 5.9961 -2.9991>, rot=<0.0082 -0.7071 -0.0082 -0.7071>}
 l_lip_upper_outer {parent=head, loc=<-1.5528 6.7704 -2.3016>, rot=<0.0082 -0.7071 -0.0082 -0.7071>}
 l_lip_upper_inner {parent=head, loc=<-1.4785 8.036 -1.3095>, rot=<0.0082 -0.7071 -0.0082 -0.7071>}
 lip_upper_mid {parent=head, loc=<-1.4946 8.5437 0>, rot=<0.0082 -0.7071 -0.0082 -0.7071>}
 l_philtrum {parent=head, loc=<-0.2636 8.3223 -0.9648>, rot=<0.0082 -0.7071 -0.0082 -0.7071>}
 l_lip_upper_above {parent=head, loc=<-0.4586 7.2489 -2.1413>, rot=<0.0082 -0.7071 -0.0082 -0.7071>}
 l_naso_frown {parent=head, loc=<-4.4164 6.4833 -2.7996>, rot=<0.0082 -0.7071 -0.0082 -0.7071>}
 l_jaw {parent=head, loc=<-3.2994 1.835 -4.8708>, rot=<0.0082 -0.7071 -0.0082 -0.7071>}
 l_chin {parent=head, loc=<-5.0311 7.9204 -1.6297>, rot=<0.0082 -0.7071 -0.0082 -0.7071>}
 chin_under {parent=head, loc=<-6.8237 4.1979 0>, rot=<0.0082 -0.7071 -0.0082 -0.7071>}
 lip_lower_below_mid {parent=head, loc=<-3.4628 8.3288 0>, rot=<0.0082 -0.7071 -0.0082 -0.7071>}
 l_lip_lower_below {parent=head, loc=<-3.5217 7.7849 -1.8249>, rot=<0.0082 -0.7071 -0.0082 -0.7071>}
 l_squint_inner {parent=head, loc=<3.8269 7.6165 -1.1266>, rot=<0.0082 -0.7071 -0.0082 -0.7071>}
 l_cheekbone_upper {parent=head, loc=<2.7228 3.3178 -5.1801>, rot=<0.0082 -0.7071 -0.0082 -0.7071>}
 l_temple {parent=head, loc=<5.6469 3.4203 -5.2502>, rot=<0.0082 -0.7071 -0.0082 -0.7071>}
 r_temple {parent=head, loc=<5.6473 3.4203 5.2502>, rot=<0.0082 -0.7071 -0.0082 -0.7071>}
 r_cheekbone_upper {parent=head, loc=<2.7228 3.3177 5.1801>, rot=<0.0082 -0.7071 -0.0082 -0.7071>}
 r_squint_inner {parent=head, loc=<3.827 7.6165 1.1266>, rot=<0.0082 -0.7071 -0.0082 -0.7071>}
 r_lip_lower_below {parent=head, loc=<-3.5218 7.7849 1.8249>, rot=<0.0082 -0.7071 -0.0082 -0.7071>}
 r_chin {parent=head, loc=<-5.0314 7.9204 1.6297>, rot=<0.0082 -0.7071 -0.0082 -0.7071>}
 r_jaw {parent=head, loc=<-3.2996 1.835 4.8708>, rot=<0.0082 -0.7071 -0.0082 -0.7071>}
 r_naso_frown {parent=head, loc=<-4.4161 6.4833 2.7996>, rot=<0.0082 -0.7071 -0.0082 -0.7071>}
 r_lip_upper_above {parent=head, loc=<-0.4587 7.2489 2.1413>, rot=<0.0082 -0.7071 -0.0082 -0.7071>}
 r_philtrum {parent=head, loc=<-0.2634 8.3223 0.9648>, rot=<0.0082 -0.7071 -0.0082 -0.7071>}
 r_lip_upper_inner {parent=head, loc=<-1.4781 8.036 1.3095>, rot=<0.0082 -0.7071 -0.0082 -0.7071>}
 r_lip_upper_outer {parent=head, loc=<-1.5529 6.7704 2.3016>, rot=<0.0082 -0.7071 -0.0082 -0.7071>}
 r_lip_corner {parent=head, loc=<-2.2032 5.9961 2.9991>, rot=<0.0082 -0.7071 -0.0082 -0.7071>}
 r_lip_lower_outer {parent=head, loc=<-2.2042 6.817 2.3122>, rot=<0.0082 -0.7071 -0.0082 -0.7071>}
 r_lip_lower_inner {parent=head, loc=<-2.3302 7.9013 1.2771>, rot=<0.0082 -0.7071 -0.0082 -0.7071>}
 r_nose_wrinkle {parent=head, loc=<1.8761 8.8925 1.2608>, rot=<0.0082 -0.7071 -0.0082 -0.7071>}
 r_cheekbone_lower {parent=head, loc=<1.3749 5.5983 4.627>, rot=<0.0082 -0.7071 -0.0082 -0.7071>}
 r_cheek {parent=head, loc=<-0.9638 3.6064 4.8529>, rot=<0.0082 -0.7071 -0.0082 -0.7071>}
 r_lip_corner_outer {parent=head, loc=<-2.1393 5.3057 3.9779>, rot=<0.0082 -0.7071 -0.0082 -0.7071>}
 r_naso_lower {parent=head, loc=<-0.7022 6.0103 3.4425>, rot=<0.0082 -0.7071 -0.0082 -0.7071>}
 r_forehead_inner {parent=head, loc=<7.6873 7.4321 0.7914>, rot=<0.0082 -0.7071 -0.0082 -0.7071>}
 r_forehead_outer {parent=head, loc=<7.802 6.7955 3.3107>, rot=<0.0082 -0.7071 -0.0082 -0.7071>}
 r_squint_mid_a {parent=head, loc=<2.9143 7.1169 1.812>, rot=<0.0082 -0.7071 -0.0082 -0.7071>}
 r_squint_outer {parent=head, loc=<3.3387 5.7576 4.5892>, rot=<0.0082 -0.7071 -0.0082 -0.7071>}
 r_squint_mid_b {parent=head, loc=<2.4774 6.6716 3.4906>, rot=<0.0082 -0.7071 -0.0082 -0.7071>}
 r_ear {parent=head, loc=<1.9019 0.054 5.772>, rot=<0.0082 -0.7071 -0.0082 -0.7071>}
 r_naso_mid {parent=head, loc=<0.8328 6.9414 2.6324>, rot=<0.0082 -0.7071 -0.0082 -0.7071>}
 r_naso_upper {parent=head, loc=<2.1875 7.7056 1.4304>, rot=<0.0082 -0.7071 -0.0082 -0.7071>}
 r_nose {parent=head, loc=<0.6502 8.3098 1.6992>, rot=<0.0082 -0.7071 -0.0082 -0.7071>}
 r_nostril {parent=head, loc=<0.5381 8.7003 0.8582>, rot=<0.0082 -0.7071 -0.0082 -0.7071>}
 r_brow_outer {parent=head, loc=<4.9433 5.8827 4.7441>, rot=<0.0082 -0.7071 -0.0082 -0.7071>}
 r_brow_mid_b {parent=head, loc=<5.4943 6.9735 4.0039>, rot=<0.0082 -0.7071 -0.0082 -0.7071>}
 r_brow_mid_a {parent=head, loc=<5.6643 7.8866 2.6785>, rot=<0.0082 -0.7071 -0.0082 -0.7071>}
 r_brow_inner {parent=head, loc=<5.3622 8.2744 1.3658>, rot=<0.0082 -0.7071 -0.0082 -0.7071>}
 l_lid_upper_mid_pivot {parent=head, loc=<4.5041 5.4698 -2.9459>, rot=<0.0082 -0.7071 -0.0082 -0.7071>}
 l_lid_upper_mid {parent=l_lid_upper_mid_pivot, loc=<0 1.7137 0.3395>, rot=<0 0 0 1>}
 l_lid_upper_inner_pivot {parent=head, loc=<4.5039 5.4698 -2.6261>, rot=<0.0082 -0.7071 -0.0082 -0.7071>}
 l_lid_upper_inner {parent=l_lid_upper_inner_pivot, loc=<-0.7994 1.6829 0.0594>, rot=<0 0 0 1>}
 l_lid_upper_outer_pivot {parent=head, loc=<4.5039 5.4698 -3.3097>, rot=<0.0082 -0.7071 -0.0082 -0.7071>}
 l_lid_upper_outer {parent=l_lid_upper_outer_pivot, loc=<0.6783 1.3652 0.0152>, rot=<0 0 0 1>}
 l_lid_lower_inner_pivot {parent=head, loc=<4.2002 5.4628 -2.6261>, rot=<0.0082 -0.7071 -0.0082 -0.7071>}
 l_lid_lower_inner {parent=l_lid_lower_inner_pivot, loc=<-0.7212 1.6233 -0.2995>, rot=<0 0 0 1>}
 l_lid_lower_outer_pivot {parent=head, loc=<4.2002 5.4628 -3.3097>, rot=<0.0082 -0.7071 -0.0082 -0.7071>}
 l_lid_lower_outer {parent=l_lid_lower_outer_pivot, loc=<0.5549 1.2153 -0.3146>, rot=<0 0 0 1>}
 l_lid_lower_mid_pivot {parent=head, loc=<4.2004 5.4628 -2.9459>, rot=<0.0082 -0.7071 -0.0082 -0.7071>}
 l_lid_lower_mid {parent=l_lid_lower_mid_pivot, loc=<0.0133 1.6234 -0.4365>, rot=<0 0 0 1>}
 r_lid_upper_mid_pivot {parent=head, loc=<4.5037 5.4698 2.9459>, rot=<0.7071 0.0082 -0.7071 0.0082>}
 r_lid_upper_mid {parent=r_lid_upper_mid_pivot, loc=<0 -1.7137 -0.34>, rot=<0 0 0 1>}
 r_lid_upper_inner_pivot {parent=head, loc=<4.5037 5.4698 2.6261>, rot=<0.7071 0.0082 -0.7071 0.0082>}
 r_lid_upper_inner {parent=r_lid_upper_inner_pivot, loc=<0.7994 -1.6829 -0.06>, rot=<0 0 0 1>}
 r_lid_upper_outer_pivot {parent=head, loc=<4.5037 5.4698 3.3097>, rot=<0.7071 0.0082 -0.7071 0.0082>}
 r_lid_upper_outer {parent=r_lid_upper_outer_pivot, loc=<-0.6783 -1.3652 -0.015>, rot=<0 0 0 1>}
 r_lid_lower_inner_pivot {parent=head, loc=<4.1998 5.4628 2.6261>, rot=<0.7071 0.0082 -0.7071 0.0082>}
 r_lid_lower_inner {parent=r_lid_lower_inner_pivot, loc=<0.7212 -1.6233 0.299>, rot=<0 0 0 1>}
 r_lid_lower_outer_pivot {parent=head, loc=<4.1998 5.4628 3.3097>, rot=<0.7071 0.0082 -0.7071 0.0082>}
 r_lid_lower_outer {parent=r_lid_lower_outer_pivot, loc=<-0.5549 -1.2153 0.314>, rot=<0 0 0 1>}
 r_lid_lower_mid_pivot {parent=head, loc=<4.2008 5.4628 2.9459>, rot=<0.7071 0.0082 -0.7071 0.0082>}
 r_lid_lower_mid {parent=r_lid_lower_mid_pivot, loc=<-0.0133 -1.6234 0.437>, rot=<0 0 0 1>}
 l_eyeball {parent=head, loc=<4.5071 5.2985 -2.6011>, rot=<0.0082 -0.7071 -0.0082 -0.7071>}
 l_eyeball_end {parent=l_eyeball, loc=<0.2677 2.5612 0>, rot=<0 0 0 1>}
 r_eyeball {parent=head, loc=<4.5066 5.2985 2.6011>, rot=<0.0082 -0.7071 -0.0082 -0.7071>}
 r_eyeball_end {parent=r_eyeball, loc=<-0.2677 2.5612 0>, rot=<0 0 0 1>}
 lower_teeth {parent=head, loc=<-4.1646 7.8348 0>, rot=<0.0082 -0.7071 -0.0082 -0.7071>}
 upper_teeth {parent=head, loc=<-2.3203 7.8752 0>, rot=<0.0082 -0.7071 -0.0082 -0.7071>}
 VB head_fx {parent=head, loc=<0 0 0>, rot=<0 0 0 1>}
 b_MF_Head {parent=head, type=Null, loc=<0 0 0>, rot=<0 0 0 1>}
 Glasses {parent=head, type=Null, loc=<5.1462 3.7909 0>, rot=<0 0.7071 0 0.7071>}
 FX_Head {parent=head, type=Null, loc=<0 0 0>, rot=<0 0 0 1>}
 Hat {parent=head, type=Null, loc=<0 0 0>, rot=<0 0.7071 0 0.7071>}
 Water_Sphere_1 {parent=head, type=Null, loc=<2.4369 0 0>, rot=<0 0 0 1>, scale=<10 10 10>}
 SM_Helmet {parent=neck_02, type=Null, loc=<-4.997 0.4345 0>, rot=<0 0.7071 0 0.7071>}
 Water_Capsule_1_1 {parent=neck_01, type=Null, loc=<0 0 0>, rot=<0 0 0 1>, scale=<17 17 17>}
 Backpack_Root {parent=spine_05, loc=<1.335 -16.6059 0>, rot=<0.023 -0.7067 -0.023 -0.7067>}
 backpack_base {parent=Backpack_Root, loc=<0 -2.321 -12.4453>, rot=<0.0115 0 0 -0.9999>}
 backpack_base_jiggle {parent=backpack_base, loc=<0 0 0>, rot=<0 0 0 1>}
 bedroll_l {parent=backpack_base_jiggle, loc=<8.1598 4.2534 -8.1691>, rot=<0 0 0 1>}
 bedroll_l_jiggle {parent=bedroll_l, loc=<5.6597 0.0315 -4.0128>, rot=<0 0 0 1>}
 bedroll_r {parent=backpack_base_jiggle, loc=<-8.7467 4.251 -7.8692>, rot=<1 0 0 0>}
 bedroll_r_jiggle {parent=bedroll_r, loc=<-3.8493 -0.0303 3.8573>, rot=<0 0 0 1>}
 backpack_bullets_01 {parent=Backpack_Root, loc=<13.2715 -7.379 9.0503>, rot=<-0.3216 -0.2232 -0.6805 -0.6194>}
 backpack_bullets_02 {parent=backpack_bullets_01, loc=<0 -5.6641 0>, rot=<-0.1384 -0.2776 -0.1909 -0.9313>}
 backpack_bullets_03 {parent=backpack_bullets_02, loc=<0 -3.862 0>, rot=<-0.0015 0.016 -0.0913 -0.9957>}
 backpak_handle {parent=Backpack_Root, loc=<0 4.6451 13.4752>, rot=<0 0 0 1>}
 backpack_handle_jiggle {parent=backpak_handle, loc=<0 0.086 5.6563>, rot=<0 0 0 1>}
 strap_l {parent=Backpack_Root, loc=<9.7494 21.621 3.4509>, rot=<0 0 0 1>}
 strap_r {parent=Backpack_Root, loc=<-9.7494 21.6209 3.4511>, rot=<1 0 0 0>}
 attach_backpack {parent=spine_05, loc=<-0.123 -13.473 0>, rot=<-0.0093 -0.707 0.0093 -0.707>}
 attach_cape {parent=spine_05, loc=<9.017 -11.714 0>, rot=<0.0616 -0.7044 -0.0616 -0.7044>}
 VB spine_05_weapon_r {parent=spine_05, loc=<-41.4719 3.1166 30.5419>, rot=<0.7642 -0.139 0.0493 -0.6278>}
 VB VB spine_05_weapon_r_ik_hand_gun {parent=VB spine_05_weapon_r, loc=<-0.1975 -3.4112 0.3811>, rot=<0 0 0 -1>}
 VB VB spine_05_weapon_r_ik_hand_l {parent=VB spine_05_weapon_r, loc=<6.5093 62.3928 14.3591>, rot=<0.0992 0.9733 0.2068 0>}
 VB spine_05_upperarm_r {parent=spine_05, loc=<3.3828 -1.007 17.7788>, rot=<-0.981 -0.0606 0.184 -0.0088>}
 VB VB spine_05_upperarm_r_lowerarm_r {parent=VB spine_05_upperarm_r, loc=<-26.6515 0 0>, rot=<0.0391 -0.0099 -0.2453 -0.9686>}
 VB VB VB spine_05_upperarm_r_lowerarm_r_hand_r {parent=VB VB spine_05_upperarm_r_lowerarm_r, loc=<-22.3492 0 0>, rot=<-0.7081 -0.0507 0.0975 -0.6975>}
 VB spine_05_hand_attach_r {parent=spine_05, loc=<-49.6338 5.9346 31.7261>, rot=<0.4421 -0.6387 -0.4091 -0.4788>}
 VB VB spine_05_hand_attach_r_ik_hand_gun {parent=VB spine_05_hand_attach_r, loc=<-3 -8.5 0>, rot=<0 0 0.7071 -0.7071>}
 VB VB spine_05_hand_attach_r_ik_hand_l {parent=VB spine_05_hand_attach_r, loc=<62.8041 -15.2067 13.978>, rot=<0.7584 0.6181 0.1462 0.1462>}
 FX_Chest {parent=spine_05, type=Null, loc=<0 0 0>, rot=<0.4995 0.4996 0.5004 0.5005>}
 FX_TechStack {parent=spine_05, type=Null, loc=<54.7019 -20.2627 -13.8847>, rot=<0 0 0 1>}
 SM_UpperBody {parent=spine_05, type=Null, loc=<0 0 0>, rot=<0 0.7071 0 0.7071>}
 Backpack {parent=spine_05, type=Null, loc=<0 0 0>, rot=<0 0.7071 0 0.7071>}
 Messenger {parent=spine_04, type=Null, loc=<-6.2174 -0.4163 5.1361>, rot=<-0.0349 0.2505 0.005 0.9675>}
 FX_TechClip {parent=spine_04, type=Null, loc=<0.5271 14.2169 18.362>, rot=<0.4583 0.5385 0.5 0.5>}
 LootEntryPosition {parent=spine_04, type=Null, loc=<13.9262 -12.918 0>, rot=<0 0 0 1>}
 FX_Spine {parent=Spine_03, type=Null, loc=<0 0 0>, rot=<0.4996 0.4995 0.5005 0.5004>}
 Shield {parent=Spine_03, type=Null, loc=<26.9186 60.8061 -17.0839>, rot=<0.4056 0.5792 0.4056 0.5792>}
 cloth_back_01 {parent=Spine_02, loc=<0.3988 -11.6844 0.046>, rot=<0.0282 -0.7065 -0.0282 -0.7065>}
 cloth_back_02 {parent=cloth_back_01, loc=<0 -0.0893 -10.5643>, rot=<0 0 0 1>}
 cloth_back_03 {parent=cloth_back_02, loc=<0 -1.1168 -10.8365>, rot=<0 0 0 1>}
 thigh_l {parent=pelvis, loc=<-2.5693 -0.0109 -7.7501>, rot=<0.0012 0.0572 0.0155 -0.9982>}
 calf_l {parent=thigh_l, loc=<-39.0484 0 0>, rot=<0.0006 -0.0161 -0.0391 -0.9991>}
 foot_l {parent=calf_l, loc=<-42.561 0.1077 0.4292>, rot=<0.0418 -0.0455 0.0343 -0.9975>}
 ball_l {parent=foot_l, loc=<-6.1599 13.8264 0.6806>, rot=<0.0001 -0.0001 -0.7063 -0.7079>}
 Toe_L {parent=ball_l, type=Null, loc=<0.0248 2.9999 -0.0143>, rot=<0 0 0 1>}
 foot_l_toe_tip {parent=ball_l, type=Null, loc=<7 3 0>, rot=<0 0 0 1>}
 foot_l_heel {parent=ball_l, type=Null, loc=<-20 3 0>, rot=<0 0 0 1>}
 foot_l_outer {parent=ball_l, type=Null, loc=<0 3 -6.5>, rot=<0 0 0 1>}
 foot_l_inner {parent=ball_l, type=Null, loc=<0 3 5>, rot=<0 0 0 1>}
 FX_foot_lSocket {parent=foot_l, type=Null, loc=<-5.7715 -8.6604 0>, rot=<0 0 0 1>}
 foot_l {parent=foot_l, type=Null, loc=<-9.366 0.0992 -0.0446>, rot=<0 0 0 1>}
 FX_Skydive_Foot_L {parent=foot_l, type=Null, loc=<0 0 0>, rot=<0 0 0 1>}
 Water_Capsule_5_3 {parent=foot_l, type=Null, loc=<0 0 0>, rot=<0 0 0 1>, scale=<8 8 8>}
 calf_twist_l {parent=calf_l, loc=<-17.7251 0 0>, rot=<0 0 0 1>}
 calf_twist_01_l {parent=calf_l, loc=<-35.813 -0.0967 0.0397>, rot=<0.0237 0.0095 -0.0035 -0.9997>}
 calf_twist_02_l {parent=calf_l, loc=<-18.8064 -0.0967 0.0397>, rot=<0.0237 0.0095 -0.0035 -0.9997>}
 lowerleg_twist_01_l {parent=calf_l, loc=<-38.7753 0 0>, rot=<0 0 0 1>}
 lowerleg_twist_02_l {parent=calf_l, loc=<-20.0451 0 0>, rot=<0 0 0 1>}
 SM_L_Shin {parent=calf_l, type=Null, loc=<0 0 0>, rot=<0.0295 0.7074 0.0017 0.7062>}
 calf_L_socket {parent=calf_l, type=Null, loc=<0 0 0>, rot=<0.5 0.5 0.5 0.5>}
 Water_Capsule_5_2 {parent=calf_l, type=Null, loc=<0 0 0>, rot=<0 0 0 1>, scale=<8 8 8>}
 FX_Contrail_Foot_L {parent=calf_l, type=Null, loc=<-38.0343 -0.0002 0>, rot=<0 -0.7071 0.7071 0>}
 thigh_twist_l {parent=thigh_l, loc=<0 0 0>, rot=<0 0 0 1>}
 thigh_twist_01_l {parent=thigh_l, loc=<-21.6303 0 0>, rot=<0.0013 0 0 -1>}
 thigh_twist_01_lRigidBody1 {parent=thigh_twist_01_l, loc=<0 0 0>, rot=<0 0 0 1>}
 thigh_twist_01_lRigidBodyShape1Transform {parent=thigh_twist_01_lRigidBody1, loc=<0.1148 -1.7829 0.9561>, rot=<0 0 0.0574 -0.9983>}
 thigh_twist_01_lConstraint1 {parent=thigh_twist_01_lRigidBody1, loc=<0 0 0>, rot=<0 -1 0 0>}
 thigh_lRigidBody1 {parent=thigh_l, loc=<0 0 0>, rot=<0 0 0 1>}
 thigh_lRigidBodyShape1Transform {parent=thigh_lRigidBody1, loc=<-5.3539 3.327 2.7638>, rot=<0.1698 -0.0002 -0.6067 -0.7766>}
 thigh_lConstraint1 {parent=thigh_lRigidBody1, loc=<0 0 0>, rot=<0 -0.7665 0.0256 -0.6418>}
 upperleg_twist_01_l {parent=thigh_l, loc=<-0.3846 0 0>, rot=<0 0 0 1>}
 SM_L_Thigh {parent=thigh_l, type=Null, loc=<0 0 0>, rot=<0.0674 0.7029 0.0367 0.7071>}
 Water_Capsule_5_1 {parent=thigh_l, type=Null, loc=<0 0 0>, rot=<0 0 0 1>, scale=<8 8 8>}
 thigh_r {parent=pelvis, loc=<-2.5691 -0.0109 7.7501>, rot=<0.0572 -0.0012 0.9982 0.0155>}
 calf_r {parent=thigh_r, loc=<39.0487 0 0>, rot=<0.0006 -0.0161 -0.0391 -0.9991>}
 foot_r {parent=calf_r, loc=<42.5612 -0.1077 -0.4292>, rot=<0.0418 -0.0455 0.0343 -0.9975>}
 ball_r {parent=foot_r, loc=<6.1598 -13.8264 -0.6806>, rot=<0.0001 -0.0001 -0.7063 -0.7079>}
 Toe_R {parent=ball_r, type=Null, loc=<-0.0248 -2.9999 0.0143>, rot=<0 0 0 1>}
 foot_r_inner {parent=ball_r, type=Null, loc=<0 -3 -5>, rot=<0 0 0 1>}
 foot_r_outer {parent=ball_r, type=Null, loc=<0 -3 6.5>, rot=<0 0 0 1>}
 foot_r_heel {parent=ball_r, type=Null, loc=<20 -3 0>, rot=<0 0 0 1>}
 foot_r_toe_tip {parent=ball_r, type=Null, loc=<-7 -3 0>, rot=<0 0 0 1>}
 FX_foot_rSocket {parent=foot_r, type=Null, loc=<5.7715 8.6604 0>, rot=<0 0 1 0>}
 foot_r {parent=foot_r, type=Null, loc=<9.3659 -0.0992 0.0446>, rot=<0 0 0 1>}
 FX_Skydive_Foot_R {parent=foot_r, type=Null, loc=<0 0 0>, rot=<0.5761 -0.6242 -0.5237 0.0646>}
 Water_Capsule_4_3 {parent=foot_r, type=Null, loc=<0 0 0>, rot=<0 0 0 1>, scale=<8 8 8>}
 calf_twist_r {parent=calf_r, loc=<17.7252 0 0>, rot=<0 0 0 1>}
 calf_twist_01_r {parent=calf_r, loc=<35.8132 0.0967 -0.0397>, rot=<0.0237 0.0095 -0.0035 -0.9997>}
 calf_twist_02_r {parent=calf_r, loc=<18.8064 0.0967 -0.0397>, rot=<0.0237 0.0095 -0.0035 -0.9997>}
 lowerleg_twist_01_r {parent=calf_r, loc=<38.7753 0 0>, rot=<0 0 0 1>}
 lowerleg_twist_02_r {parent=calf_r, loc=<20.0452 0 0>, rot=<0 0 0 1>}
 SM_R_Shin {parent=calf_r, type=Null, loc=<-0.0155 -0.1695 -0.0087>, rot=<0.7 -0.0276 -0.7136 0>}
 FX_Contrail_Foot_R {parent=calf_r, type=Null, loc=<38.2472 0.0012 0.0007>, rot=<0.7071 0 0 0.7071>}
 Water_Capsule_4_2 {parent=calf_r, type=Null, loc=<0 0 0>, rot=<0 0 0 1>, scale=<8 8 8>}
 thigh_twist_r {parent=thigh_r, loc=<0 0 0>, rot=<0 0 0 1>}
 thigh_twist_01_r {parent=thigh_r, loc=<21.6303 0 0>, rot=<0.0013 0 0 -1>}
 thigh_twist_01_rRigidBody1 {parent=thigh_twist_01_r, loc=<0 0 0>, rot=<0 0 0 1>}
 thigh_twist_01_rRigidBodyShape1Transform {parent=thigh_twist_01_rRigidBody1, loc=<2.1093 -3.9335 -2.7962>, rot=<-0.046 0.0124 -0.0081 -0.9988>}
 thigh_twist_01_rConstraint1 {parent=thigh_twist_01_rRigidBody1, loc=<0 0 0>, rot=<0 0 0 1>}
 pack {parent=thigh_twist_01_r, loc=<0.099 -3.599 5.3503>, rot=<-0.7068 -0.0171 0.7069 0.0197>}
 pack_jiggle {parent=pack, loc=<0 0 0>, rot=<0 0 0 1>}
 wrench {parent=thigh_twist_01_r, loc=<-3.3868 -9.6206 -0.0457>, rot=<-0.5571 0.0105 0.8276 -0.0671>}
 thigh_rRigidBody1 {parent=thigh_r, loc=<0 0 0>, rot=<0 0 0 1>}
 thigh_rRigidBodyShape1Transform {parent=thigh_rRigidBody1, loc=<2.9317 -9.4629 -1.4637>, rot=<0.4095 0.0102 -0.4674 -0.7834>}
 thigh_rConstraint1 {parent=thigh_rRigidBody1, loc=<0 0 0>, rot=<0 0.7665 0.0256 -0.6418>}
 upperleg_twist_01_r {parent=thigh_r, loc=<0.3846 0 0>, rot=<0 0 0 1>}
 SM_R_Thigh {parent=thigh_r, type=Null, loc=<0 0 0>, rot=<0.7059 -0.0308 -0.7077 0>}
 Water_Capsule_4_1 {parent=thigh_r, type=Null, loc=<0 0 0>, rot=<0 0 0 1>, scale=<8 8 8>}
 belt_back {parent=pelvis, loc=<-3.1326 -10.407 0>, rot=<-0.0013 -0.7071 0.0013 -0.7071>}
 belt_front {parent=pelvis, loc=<-2.9332 9.8423 0>, rot=<-0.0013 -0.7071 0.0013 -0.7071>}
 belt_l {parent=pelvis, loc=<4.9138 0 -11.2468>, rot=<-0.0013 -0.7071 0.0013 -0.7071>}
 pelvisRigidBody1 {parent=pelvis, loc=<0 0 0>, rot=<0 0 0 1>}
 pelvisRigidBodyShape1Transform {parent=pelvisRigidBody1, loc=<0 0 0>, rot=<0 0 0 1>, scale=<1.124 1.124 1.124>}
 belt01_r {parent=pelvis, loc=<0.4733 12.5671 3.0499>, rot=<-0.0127 -0.707 0.0127 -0.707>}
 belt02_r {parent=belt01_r, loc=<-0.217 1.0304 -5.9212>, rot=<0 0 0 1>}
 belt03_r {parent=belt02_r, loc=<-0.61 0.9064 -7.6285>, rot=<0 0 0 1>}
 belt04_r {parent=belt03_r, loc=<-0.1659 0.3748 -6.6814>, rot=<0 0 0 1>}
 belt01_l {parent=pelvis, loc=<1.4574 12.8641 -2.6501>, rot=<-0.0127 -0.707 0.0127 -0.707>}
 belt02_l {parent=belt01_l, loc=<0.9484 0.7969 -6.7385>, rot=<0 0 0 1>}
 belt03_l {parent=belt02_l, loc=<0.5025 0.863 -6.2875>, rot=<0 0 0 1>}
 VB pelvis_hand_attach_r {parent=pelvis, loc=<-10.6998 7.4186 31.7261>, rot=<-0.4043 0.6633 0.3806 0.5017>}
 VB pelvis_ik_hand_gun {parent=pelvis, loc=<-2.3857 4.6238 33.8029>, rot=<0.7549 -0.1831 0.0856 -0.6239>}
 VB pelvis_ik_hand_r {parent=pelvis, loc=<-2.3857 4.6238 33.8029>, rot=<-0.7549 0.1831 -0.0856 0.6239>}
 VB pelvis_ik_hand_l {parent=pelvis, loc=<-2.3856 4.6238 -33.8029>, rot=<0.1831 0.7549 -0.6239 -0.0856>}
 VB pelvis_attach_emote_01 {parent=pelvis, loc=<-93.034 4.0424 0>, rot=<0.0127 0.707 -0.0127 0.707>}
 VB pelvis_attach_emote_02 {parent=pelvis, loc=<-93.034 4.0424 0>, rot=<0.0127 0.707 -0.0127 0.707>}
 VB pelvis_attach_emote_03 {parent=pelvis, loc=<-93.034 4.0424 0>, rot=<0.0127 0.707 -0.0127 0.707>}
 SM_Pelvis {parent=pelvis, type=Null, loc=<0 0 0>, rot=<0 0.7071 0 0.7071>}
 Water_Capsule_1_2 {parent=pelvis, type=Null, loc=<0 0 0>, rot=<0 0 0 1>, scale=<17 17 17>}
 attach {parent=Root, loc=<0 0 0>, rot=<0 0 0 1>}
 VB attach_ik_hand_gun {parent=attach, loc=<-33.8029 3.8283 90.5694>, rot=<0.4834 0.3223 0.5887 -0.5621>}
 VB VB attach_ik_hand_gun_ik_hand_l {parent=VB attach_ik_hand_gun, loc=<6.7067 65.8041 13.978>, rot=<-0.0992 -0.9733 -0.2068 0>}
 VB VB attach_ik_hand_gun_ik_hand_r {parent=VB attach_ik_hand_gun, loc=<0 0 0>, rot=<0 0 0 1>}
 VB attach_ik_foot_l {parent=attach, loc=<15.2608 -4.4176 9.4433>, rot=<0.0357 0.7045 0.0284 -0.7082>}
 VB attach_ik_foot_r {parent=attach, loc=<-15.2609 -4.4176 9.4431>, rot=<0.7082 0.0284 -0.7045 0.0357>}
 AttachSocket {parent=attach, type=Null, loc=<0 0 0>, rot=<0 0 0 1>}
 ik_foot_root {parent=Root, loc=<0 0 0>, rot=<0 0 0 1>}
 ik_foot_l {parent=ik_foot_root, loc=<15.2608 -4.4176 9.4433>, rot=<0.0357 0.7045 0.0284 -0.7082>}
 ik_foot_r {parent=ik_foot_root, loc=<-15.2609 -4.4176 9.4431>, rot=<0.7082 0.0284 -0.7045 0.0357>}
 VB ik_foot_root_weapon_l {parent=ik_foot_root, loc=<30.5419 2.8604 90.0703>, rot=<0.5621 0.5887 -0.3223 0.4834>}
 VB ik_foot_root_weapon_r {parent=ik_foot_root, loc=<-30.5419 2.8605 90.0703>, rot=<-0.4834 -0.3223 -0.5887 0.5621>}
 VB ik_foot_root_hand_attach_l {parent=ik_foot_root, loc=<31.726 6.3234 82.1605>, rot=<-0.5697 -0.1139 0.0188 0.8137>}
 VB ik_foot_root_hand_attach_r {parent=ik_foot_root, loc=<-31.7261 6.3234 82.1604>, rot=<-0.5697 0.1139 -0.0188 0.8137>}
 ik_hand_root {parent=Root, loc=<0 0 0>, rot=<0 0 0 1>}
 ik_hand_gun {parent=ik_hand_root, loc=<-33.8029 3.8283 90.5694>, rot=<0.4834 0.3223 0.5887 -0.5621>}
 ik_hand_l {parent=ik_hand_gun, loc=<6.7067 65.8041 13.978>, rot=<-0.0992 -0.9733 -0.2068 0>}
 ik_hand_r {parent=ik_hand_gun, loc=<0 0 0>, rot=<0 0 0 1>}
 VB ik_hand_root_ik_hand_l {parent=ik_hand_root, loc=<33.8029 3.8282 90.5694>, rot=<0.5621 0.5887 -0.3223 0.4834>}
 nxRagdoll1 {parent=Root, loc=<0 -5.2501 0>, rot=<0 0 0 1>}
 warp_point_01 {parent=Root, loc=<0 0 0>, rot=<0 0 0 1>}
 warp_point_02 {parent=Root, loc=<0 0 0>, rot=<0 0 0 1>}
 warp_point_03 {parent=Root, loc=<0 0 0>, rot=<0 0 0 1>}
 attach_emote_01 {parent=Root, loc=<0 0 0>, rot=<0 0 0 1>}
 attach_emote_02 {parent=Root, loc=<0 0 0>, rot=<0 0 0 1>}
 attach_emote_03 {parent=Root, loc=<0 0 0>, rot=<0 0 0 1>}
 root_fp {parent=Root, loc=<0 0 0>, rot=<0 0 0 1>}
 camera_root {parent=root_fp, loc=<0 0 0>, rot=<0 0 0 1>}
 camera {parent=camera_root, loc=<0 0 0>, rot=<0 0 0 1>}
 attach_fp {parent=root_fp, loc=<0 0 0>, rot=<0 0 0 1>}
 VB root_prop {parent=Root, loc=<0 0 0>, rot=<0 0 0 1>}
 VB Root_hand_l {parent=Root, loc=<33.8029 3.8282 90.5694>, rot=<0.5621 0.5887 -0.3223 0.4834>}
 VB Root_hand_r {parent=Root, loc=<-33.8029 3.8283 90.5694>, rot=<-0.4834 -0.3223 -0.5887 0.5621>}
 Root_Effect {parent=Root, type=Null, loc=<0 0 0>, rot=<0 0 0.7071 0.7071>}
 FX_Feet {parent=Root, type=Null, loc=<0 0 15>, rot=<0 0 0 1>}
 FX_Combat {parent=Root, type=Null, loc=<0 -0.0002 182.224>, rot=<0 0 0.7071 0.7071>}
 r_knee_corrective {type=Curve, loc=<0 0 0>, rot=<0 0 0 1>}
 l_knee_corrective {type=Curve, loc=<0 0 0>, rot=<0 0 0 1>}
 Male_Med_Head_01_Pout {type=Curve, loc=<0 0 0>, rot=<0 0 0 1>}
 Male_Med_Head_02_Pout {type=Curve, loc=<0 0 0>, rot=<0 0 0 1>}
 Male_Med_Head_03_Pout {type=Curve, loc=<0 0 0>, rot=<0 0 0 1>}
 Male_Med_Head_05_Effort {type=Curve, loc=<0 0 0>, rot=<0 0 0 1>}
 Stretch {type=Curve, loc=<0 0 0>, rot=<0 0 0 1>}
 Squash {type=Curve, loc=<0 0 0>, rot=<0 0 0 1>}
 polySurface151 {type=Curve, loc=<0 0 0>, rot=<0 0 0 1>}
 l_cheek_puff {type=Curve, loc=<0 0 0>, rot=<0 0 0 1>}
 r_cheek_puff {type=Curve, loc=<0 0 0>, rot=<0 0 0 1>}
 l_smile_fix {type=Curve, loc=<0 0 0>, rot=<0 0 0 1>}
 r_smile_fix {type=Curve, loc=<0 0 0>, rot=<0 0 0 1>}
 jaw_open_fix {type=Curve, loc=<0 0 0>, rot=<0 0 0 1>}
 r_lip_sneer_fix {type=Curve, loc=<0 0 0>, rot=<0 0 0 1>}
 l_lip_sneer_fix {type=Curve, loc=<0 0 0>, rot=<0 0 0 1>}
 L_Blink {type=Curve, loc=<0 0 0>, rot=<0 0 0 1>}
 R_Blink {type=Curve, loc=<0 0 0>, rot=<0 0 0 1>}
 L_Arm_Stretch {type=Curve, loc=<0 0 0>, rot=<0 0 0 1>}
 Hair_Hat {type=Curve, loc=<0 0 0>, rot=<0 0 0 1>}
 Hair_Helmet {type=Curve, loc=<0 0 0>, rot=<0 0 0 1>}
 Male_Med_Head_02_What {type=Curve, loc=<0 0 0>, rot=<0 0 0 1>}
 Male_Med_Head_02_Smile {type=Curve, loc=<0 0 0>, rot=<0 0 0 1>}
 Male_Med_Head_02_Scowl {type=Curve, loc=<0 0 0>, rot=<0 0 0 1>}
 Male_Med_Head_02_Hmmph {type=Curve, loc=<0 0 0>, rot=<0 0 0 1>}
 Male_Med_Head_02_Grrr {type=Curve, loc=<0 0 0>, rot=<0 0 0 1>}
 Male_Med_Head_02_Effort {type=Curve, loc=<0 0 0>, rot=<0 0 0 1>}
 Male_Med_Head_03_What {type=Curve, loc=<0 0 0>, rot=<0 0 0 1>}
 Male_Med_Head_03_Grrr {type=Curve, loc=<0 0 0>, rot=<0 0 0 1>}
 Male_Med_Head_03_Effort {type=Curve, loc=<0 0 0>, rot=<0 0 0 1>}
 Male_Med_Head_03_Hmmph {type=Curve, loc=<0 0 0>, rot=<0 0 0 1>}
 Male_Med_Head_03_Scowl {type=Curve, loc=<0 0 0>, rot=<0 0 0 1>}
 Male_Med_Head_03_Smile {type=Curve, loc=<0 0 0>, rot=<0 0 0 1>}
 Male_Med_Head_05_Teeth_Low_What {type=Curve, loc=<0 0 0>, rot=<0 0 0 1>}
 Male_Med_Head_05_Teeth_Low_Pout {type=Curve, loc=<0 0 0>, rot=<0 0 0 1>}
 Male_Med_Head_05_Teeth_Low_Hmmph {type=Curve, loc=<0 0 0>, rot=<0 0 0 1>}
 Male_Med_Head_05_Teeth_Low_Grrr {type=Curve, loc=<0 0 0>, rot=<0 0 0 1>}
 Male_Med_Head_05_Teeth_Low_Effort {type=Curve, loc=<0 0 0>, rot=<0 0 0 1>}
 Male_Med_Head_05_What {type=Curve, loc=<0 0 0>, rot=<0 0 0 1>}
 Male_Med_Head_05_Grrr {type=Curve, loc=<0 0 0>, rot=<0 0 0 1>}
 Male_Med_Head_05_Hmmph {type=Curve, loc=<0 0 0>, rot=<0 0 0 1>}
 Male_Med_Head_05_Pout {type=Curve, loc=<0 0 0>, rot=<0 0 0 1>}
 Male_Med_Head_05_Tongue_What {type=Curve, loc=<0 0 0>, rot=<0 0 0 1>}
 Male_Med_Head_05_Tongue_Pout {type=Curve, loc=<0 0 0>, rot=<0 0 0 1>}
 Male_Med_Head_05_Tongue_Hmmph {type=Curve, loc=<0 0 0>, rot=<0 0 0 1>}
 Male_Med_Head_05_Tongue_Grrr {type=Curve, loc=<0 0 0>, rot=<0 0 0 1>}
 Male_Med_Head_05_Tongue_Effort {type=Curve, loc=<0 0 0>, rot=<0 0 0 1>}
 AdditiveLayer {type=Curve, loc=<0 0 0>, rot=<0 0 0 1>}
 RightHandIK {type=Curve, loc=<0 0 0>, rot=<0 0 0 1>}
 LeftHandIK {type=Curve, loc=<0 0 0>, rot=<0 0 0 1>}
 LeftHandFK {type=Curve, loc=<0 0 0>, rot=<0 0 0 1>}
 Male_Med_Head_01_Grrr {type=Curve, loc=<0 0 0>, rot=<0 0 0 1>}
 WeaponAiming {type=Curve, loc=<0 0 0>, rot=<0 0 0 1>}
 Male_Med_Head_01_Hmmph {type=Curve, loc=<0 0 0>, rot=<0 0 0 1>}
 DragonSlash {type=Curve, loc=<0 0 0>, rot=<0 0 0 1>}
 Male_Med_Head_01_Effort {type=Curve, loc=<0 0 0>, rot=<0 0 0 1>}
 LeftHankFK {type=Curve, loc=<0 0 0>, rot=<0 0 0 1>}
 blendOrient1 {type=Curve, loc=<0 0 0>, rot=<0 0 0 1>}
 blendParent1 {type=Curve, loc=<0 0 0>, rot=<0 0 0 1>}
 bottom_R_Corner_Down {type=Curve, loc=<0 0 0>, rot=<0 0 0 1>}
 bottom_R_Corner_Up {type=Curve, loc=<0 0 0>, rot=<0 0 0 1>}
 top_L_Corner_Up {type=Curve, loc=<0 0 0>, rot=<0 0 0 1>}
 top_L_Corner_Down {type=Curve, loc=<0 0 0>, rot=<0 0 0 1>}
 top_R_Corner_Down {type=Curve, loc=<0 0 0>, rot=<0 0 0 1>}
 top_R_Corner_Up {type=Curve, loc=<0 0 0>, rot=<0 0 0 1>}
 bottom_L_Corner_Down {type=Curve, loc=<0 0 0>, rot=<0 0 0 1>}
 bottomt_L_Corner_Up {type=Curve, loc=<0 0 0>, rot=<0 0 0 1>}
 TrailCurve1 {type=Curve, loc=<0 0 0>, rot=<0 0 0 1>}
 BothHandFK {type=Curve, loc=<0 0 0>, rot=<0 0 0 1>}
 bottom_R_Corner_Down1 {type=Curve, loc=<0 0 0>, rot=<0 0 0 1>}
 blendPoint1 {type=Curve, loc=<0 0 0>, rot=<0 0 0 1>}
 WeaponAimingFreeze {type=Curve, loc=<0 0 0>, rot=<0 0 0 1>}
 l_calf_top_taper {type=Curve, loc=<0 0 0>, rot=<0 0 0 1>}
 JogBlendSpaceRotation {type=Curve, loc=<0 0 0>, rot=<0 0 0 1>}
 ZeroOutPitchWeight {type=Curve, loc=<0 0 0>, rot=<0 0 0 1>}
 MeleeTwist {type=Curve, loc=<0 0 0>, rot=<0 0 0 1>}
 MeleeAnim {type=Curve, loc=<0 0 0>, rot=<0 0 0 1>}
 l_sneer_fix {type=Curve, loc=<0 0 0>, rot=<0 0 0 1>}
 jaw_fix {type=Curve, loc=<0 0 0>, rot=<0 0 0 1>}
 r_sneer_fix {type=Curve, loc=<0 0 0>, rot=<0 0 0 1>}
 TurnRotationAmount {type=Curve, loc=<0 0 0>, rot=<0 0 0 1>}
 TurnEnd {type=Curve, loc=<0 0 0>, rot=<0 0 0 1>}
 BlendIn {type=Curve, loc=<0 0 0>, rot=<0 0 0 1>}
 BlendOut {type=Curve, loc=<0 0 0>, rot=<0 0 0 1>}
 BodyRotateX {type=Curve, loc=<0 0 0>, rot=<0 0 0 1>}
 BodyTranslateX {type=Curve, loc=<0 0 0>, rot=<0 0 0 1>}
 CrescentKickScale {type=Curve, loc=<0 0 0>, rot=<0 0 0 1>}
 mask_geo {type=Curve, loc=<0 0 0>, rot=<0 0 0 1>}
 RemoveJumpAdditive {type=Curve, loc=<0 0 0>, rot=<0 0 0 1>}
 RightLeg_IK {type=Curve, loc=<0 0 0>, rot=<0 0 0 1>}
 SnapWeapon_LHand {type=Curve, loc=<0 0 0>, rot=<0 0 0 1>}
 PivotAnim {type=Curve, loc=<0 0 0>, rot=<0 0 0 1>}
 DefaultAdditiveStateWeight {type=Curve, loc=<0 0 0>, rot=<0 0 0 1>}
 StartAnim {type=Curve, loc=<0 0 0>, rot=<0 0 0 1>}
 StopAnim {type=Curve, loc=<0 0 0>, rot=<0 0 0 1>}
 DisableSpeedWarping {type=Curve, loc=<0 0 0>, rot=<0 0 0 1>}
 DisableDBNOSlopeWarp {type=Curve, loc=<0 0 0>, rot=<0 0 0 1>}
 MaskRightArm {type=Curve, loc=<0 0 0>, rot=<0 0 0 1>}
 TurnInPlaceAnim {type=Curve, loc=<0 0 0>, rot=<0 0 0 1>}
 Head_MESH {type=Curve, loc=<0 0 0>, rot=<0 0 0 1>}
 Hair_MESH {type=Curve, loc=<0 0 0>, rot=<0 0 0 1>}
 LaunchPadTransitionOff {type=Curve, loc=<0 0 0>, rot=<0 0 0 1>}
 C_glabella_up_pose {type=Curve, loc=<0 0 0>, rot=<0 0 0 1>}
 jaw_open_pose {type=Curve, loc=<0 0 0>, rot=<0 0 0 1>}
 L_blink_pose {type=Curve, loc=<0 0 0>, rot=<0 0 0 1>}
 L_brow_down_pose {type=Curve, loc=<0 0 0>, rot=<0 0 0 1>}
 L_brow_squeeze_in_pose {type=Curve, loc=<0 0 0>, rot=<0 0 0 1>}
 L_brow_up_pose {type=Curve, loc=<0 0 0>, rot=<0 0 0 1>}
 L_frown_pose {type=Curve, loc=<0 0 0>, rot=<0 0 0 1>}
 L_lip_corner_narrow_pose {type=Curve, loc=<0 0 0>, rot=<0 0 0 1>}
 L_squint_inner_pose {type=Curve, loc=<0 0 0>, rot=<0 0 0 1>}
 L_wide_pose {type=Curve, loc=<0 0 0>, rot=<0 0 0 1>}
 look_down_pose {type=Curve, loc=<0 0 0>, rot=<0 0 0 1>}
 look_left_pose {type=Curve, loc=<0 0 0>, rot=<0 0 0 1>}
 look_right_pose {type=Curve, loc=<0 0 0>, rot=<0 0 0 1>}
 look_up_pose {type=Curve, loc=<0 0 0>, rot=<0 0 0 1>}
 phoneme_fv_pose {type=Curve, loc=<0 0 0>, rot=<0 0 0 1>}
 phoneme_oo_pose {type=Curve, loc=<0 0 0>, rot=<0 0 0 1>}
 R_blink_pose {type=Curve, loc=<0 0 0>, rot=<0 0 0 1>}
 R_brow_down_pose {type=Curve, loc=<0 0 0>, rot=<0 0 0 1>}
 R_brow_squeeze_in_pose {type=Curve, loc=<0 0 0>, rot=<0 0 0 1>}
 R_brow_up_pose {type=Curve, loc=<0 0 0>, rot=<0 0 0 1>}
 R_frown_pose {type=Curve, loc=<0 0 0>, rot=<0 0 0 1>}
 R_lip_corner_narrow_pose {type=Curve, loc=<0 0 0>, rot=<0 0 0 1>}
 R_squint_inner_pose {type=Curve, loc=<0 0 0>, rot=<0 0 0 1>}
 R_wide_pose {type=Curve, loc=<0 0 0>, rot=<0 0 0 1>}
 DisableArmsHeadAdditive {type=Curve, loc=<0 0 0>, rot=<0 0 0 1>}
 FrontendRotation {type=Curve, loc=<0 0 0>, rot=<0 0 0 1>}
 BothLegFK {type=Curve, loc=<0 0 0>, rot=<0 0 0 1>}
 AllowTurnInPlace {type=Curve, loc=<0 0 0>, rot=<0 0 0 1>}
 MontageTimeStretchCurve {type=Curve, loc=<0 0 0>, rot=<0 0 0 1>}
 ParaGlideLeanAlpha {type=Curve, loc=<0 0 0>, rot=<0 0 0 1>}
 WeaponRInChestSpace {type=Curve, loc=<0 0 0>, rot=<0 0 0 1>}
 JimBeam_AO {type=Curve, loc=<0 0 0>, rot=<0 0 0 1>}
 R_squeeze_pose {type=Curve, loc=<0 0 0>, rot=<0 0 0 1>}
 L_squeeze_pose {type=Curve, loc=<0 0 0>, rot=<0 0 0 1>}
 R_upper_lip_raiser_pose {type=Curve, loc=<0 0 0>, rot=<0 0 0 1>}
 L_cheek_up_pose {type=Curve, loc=<0 0 0>, rot=<0 0 0 1>}
 R_cheek_up_pose {type=Curve, loc=<0 0 0>, rot=<0 0 0 1>}
 DisableFaceOverride {type=Curve, loc=<0 0 0>, rot=<0 0 0 1>}
 EnableFullBodyOverride {type=Curve, loc=<0 0 0>, rot=<0 0 0 1>}
 C_glabella_down_pose {type=Curve, loc=<0 0 0>, rot=<0 0 0 1>}
 C_nose_wrinkler_pose {type=Curve, loc=<0 0 0>, rot=<0 0 0 1>}
 eye_convergence_pose {type=Curve, loc=<0 0 0>, rot=<0 0 0 1>}
 jaw_back_pose {type=Curve, loc=<0 0 0>, rot=<0 0 0 1>}
 jaw_forward_pose {type=Curve, loc=<0 0 0>, rot=<0 0 0 1>}
 jaw_left_pose {type=Curve, loc=<0 0 0>, rot=<0 0 0 1>}
 jaw_right_pose {type=Curve, loc=<0 0 0>, rot=<0 0 0 1>}
 jaw_up_pose {type=Curve, loc=<0 0 0>, rot=<0 0 0 1>}
 L_lip_corner_wide_pose {type=Curve, loc=<0 0 0>, rot=<0 0 0 1>}
 L_lower_lip_down_pose {type=Curve, loc=<0 0 0>, rot=<0 0 0 1>}
 L_lower_lip_up_pose {type=Curve, loc=<0 0 0>, rot=<0 0 0 1>}
 L_smile_pose {type=Curve, loc=<0 0 0>, rot=<0 0 0 1>}
 L_upper_lip_lower_pose {type=Curve, loc=<0 0 0>, rot=<0 0 0 1>}
 L_upper_lip_raiser_pose {type=Curve, loc=<0 0 0>, rot=<0 0 0 1>}
 L_upperLidDown_pose {type=Curve, loc=<0 0 0>, rot=<0 0 0 1>}
 mouth_down_pose {type=Curve, loc=<0 0 0>, rot=<0 0 0 1>}
 mouth_up_pose {type=Curve, loc=<0 0 0>, rot=<0 0 0 1>}
 phoneme_ch_pose {type=Curve, loc=<0 0 0>, rot=<0 0 0 1>}
 phoneme_mbp_pose {type=Curve, loc=<0 0 0>, rot=<0 0 0 1>}
 R_lip_corner_wide_pose {type=Curve, loc=<0 0 0>, rot=<0 0 0 1>}
 R_lower_lip_down_pose {type=Curve, loc=<0 0 0>, rot=<0 0 0 1>}
 R_lower_lip_up_pose {type=Curve, loc=<0 0 0>, rot=<0 0 0 1>}
 R_smile_pose {type=Curve, loc=<0 0 0>, rot=<0 0 0 1>}
 R_upper_lip_lower_pose {type=Curve, loc=<0 0 0>, rot=<0 0 0 1>}
 R_upperLidDown_pose {type=Curve, loc=<0 0 0>, rot=<0 0 0 1>}
 tongue_back_pose {type=Curve, loc=<0 0 0>, rot=<0 0 0 1>}
 tongue_down_pose {type=Curve, loc=<0 0 0>, rot=<0 0 0 1>}
 tongue_forward_pose {type=Curve, loc=<0 0 0>, rot=<0 0 0 1>}
 tongue_up_pose {type=Curve, loc=<0 0 0>, rot=<0 0 0 1>}
 LiveLink {type=Curve, loc=<0 0 0>, rot=<0 0 0 1>}
 ParaGlideDeploy {type=Curve, loc=<0 0 0>, rot=<0 0 0 1>}
 HideLeftWeapon {type=Curve, loc=<0 0 0>, rot=<0 0 0 1>}
 DisableLeftHandIKGolfCart {type=Curve, loc=<0 0 0>, rot=<0 0 0 1>}
 DisableRightHandIKGolfCart {type=Curve, loc=<0 0 0>, rot=<0 0 0 1>}
 DisableHandIKRetargeting {type=Curve, loc=<0 0 0>, rot=<0 0 0 1>}
 DisableConsumableOffsetPose {type=Curve, loc=<0 0 0>, rot=<0 0 0 1>}
 UpperBodyNoAdditive {type=Curve, loc=<0 0 0>, rot=<0 0 0 1>}
 MaskLeftArm {type=Curve, loc=<0 0 0>, rot=<0 0 0 1>}
 ForceUpperBody {type=Curve, loc=<0 0 0>, rot=<0 0 0 1>}
 RightHandFK {type=Curve, loc=<0 0 0>, rot=<0 0 0 1>}
 ParaGlideLeanTransition {type=Curve, loc=<0 0 0>, rot=<0 0 0 1>}
 InMotionEmote {type=Curve, loc=<0 0 0>, rot=<0 0 0 1>}
 ForceFullBodyInFrontEnd {type=Curve, loc=<0 0 0>, rot=<0 0 0 1>}
 WeaponLinFloorSpace {type=Curve, loc=<0 0 0>, rot=<0 0 0 1>}
 FrontEnd_Idle_Bypass {type=Curve, loc=<0 0 0>, rot=<0 0 0 1>}
 DisableSlopeWarping {type=Curve, loc=<0 0 0>, rot=<0 0 0 1>}
 Disable_Custom {type=Curve, loc=<0 0 0>, rot=<0 0 0 1>}
 EmoteRequiresWeapon {type=Curve, loc=<0 0 0>, rot=<0 0 0 1>}
 DisableIKRootAdditive {type=Curve, loc=<0 0 0>, rot=<0 0 0 1>}
 DisableFullBodySprint {type=Curve, loc=<0 0 0>, rot=<0 0 0 1>}
 DetachFeet_Vehicle {type=Curve, loc=<0 0 0>, rot=<0 0 0 1>}
 IndigoBoost {type=Curve, loc=<0 0 0>, rot=<0 0 0 1>}
 DisableUpperBodyAdditiveMeshSpace {type=Curve, loc=<0 0 0>, rot=<0 0 0 1>}
 MeleeAOWeight {type=Curve, loc=<0 0 0>, rot=<0 0 0 1>}
 HandsInRootSpace {type=Curve, loc=<0 0 0>, rot=<0 0 0 1>}
 CrouchPelvisAdjustment {type=Curve, loc=<0 0 0>, rot=<0 0 0 1>}
 EnableAccelYawCorrection {type=Curve, loc=<0 0 0>, rot=<0 0 0 1>}
 ParaGlideModifyRoot {type=Curve, loc=<0 0 0>, rot=<0 0 0 1>}
 UseScalarParameterForBlink {type=Curve, loc=<0 0 0>, rot=<0 0 0 1>}
 BlinkControl {type=Curve, loc=<0 0 0>, rot=<0 0 0 1>}
 ForceTargeting {type=Curve, loc=<0 0 0>, rot=<0 0 0 1>}
 BlinkControlProp {type=Curve, loc=<0 0 0>, rot=<0 0 0 1>}
 UseScalarParameterForBlinkProp {type=Curve, loc=<0 0 0>, rot=<0 0 0 1>}
 CrackshotCozyHeadCorrection {type=Curve, loc=<0 0 0>, rot=<0 0 0 1>}
 ForwardMotion {type=Curve, loc=<0 0 0>, rot=<0 0 0 1>}
 EmoteForcedRootYaw {type=Curve, loc=<0 0 0>, rot=<0 0 0 1>}
 MeleeTrailWidth {type=Curve, loc=<0 0 0>, rot=<0 0 0 1>}
 space_world {type=Curve, loc=<0 0 0>, rot=<0 0 0 1>}
 PeelyBlenderFace {type=Curve, loc=<0 0 0>, rot=<0 0 0 1>}
 Width1 {type=Curve, loc=<0 0 0>, rot=<0 0 0 1>}
 OverrideIKHands {type=Curve, loc=<0 0 0>, rot=<0 0 0 1>}
 ParaGlideAddAlpha {type=Curve, loc=<0 0 0>, rot=<0 0 0 1>}
 ParaGlideTurn {type=Curve, loc=<0 0 0>, rot=<0 0 0 1>}
 DisablePelvisAdjustment {type=Curve, loc=<0 0 0>, rot=<0 0 0 1>}
 UncleBrollyDeploy {type=Curve, loc=<0 0 0>, rot=<0 0 0 1>}
 ForceRootSlopeOrientation {type=Curve, loc=<0 0 0>, rot=<0 0 0 1>}
 DisableAdditiveOffsetPose {type=Curve, loc=<0 0 0>, rot=<0 0 0 1>}
 DisableIdleNoiseAdditive {type=Curve, loc=<0 0 0>, rot=<0 0 0 1>}
 UsingConsumable {type=Curve, loc=<0 0 0>, rot=<0 0 0 1>}
 DisableRightArmAdditive {type=Curve, loc=<0 0 0>, rot=<0 0 0 1>}
 NoGlider {type=Curve, loc=<0 0 0>, rot=<0 0 0 1>}
 ParaGlideModifyPelvis {type=Curve, loc=<0 0 0>, rot=<0 0 0 1>}
 ParaGlideModifyLegs {type=Curve, loc=<0 0 0>, rot=<0 0 0 1>}
 HT_L1_TransitionCurve {type=Curve, loc=<0 0 0>, rot=<0 0 0 1>}
 KeepPelvisAdjustment {type=Curve, loc=<0 0 0>, rot=<0 0 0 1>}
 PatrolDisableIdleDuringWalk {type=Curve, loc=<0 0 0>, rot=<0 0 0 1>}
 PatrolWalkUseFullbodyBlendspace {type=Curve, loc=<0 0 0>, rot=<0 0 0 1>}
 RemoveJumpUpAdditive {type=Curve, loc=<0 0 0>, rot=<0 0 0 1>}
 RemoveJumpLoopAdditive {type=Curve, loc=<0 0 0>, rot=<0 0 0 1>}
 RemoveJumpFallAdditive {type=Curve, loc=<0 0 0>, rot=<0 0 0 1>}
 JH_TransformCurve {type=Curve, loc=<0 0 0>, rot=<0 0 0 1>}
 PhysicsOff {type=Curve, loc=<0 0 0>, rot=<0 0 0 1>}
 DisableWeaponAdditivePose {type=Curve, loc=<0 0 0>, rot=<0 0 0 1>}
 Lexa_TransformationCurve {type=Curve, loc=<0 0 0>, rot=<0 0 0 1>}
 HideProp {type=Curve, loc=<0 0 0>, rot=<0 0 0 1>}
 CopyPose {type=Curve, loc=<0 0 0>, rot=<0 0 0 1>}
 MaterialCurveA {type=Curve, loc=<0 0 0>, rot=<0 0 0 1>}
 MaterialCurveB {type=Curve, loc=<0 0 0>, rot=<0 0 0 1>}
 MaterialCurveC {type=Curve, loc=<0 0 0>, rot=<0 0 0 1>}
 MaterialCurveD {type=Curve, loc=<0 0 0>, rot=<0 0 0 1>}
 MaterialCurveE {type=Curve, loc=<0 0 0>, rot=<0 0 0 1>}
 UseEmoteAnimOffset {type=Curve, loc=<0 0 0>, rot=<0 0 0 1>}
 emissive mult 2 {type=Curve, loc=<0 0 0>, rot=<0 0 0 1>}
 MaxHandle {type=Curve, loc=<0 0 0>, rot=<0 0 0 1>}
 FootPhase {type=Curve, loc=<0 0 0>, rot=<0 0 0 1>}
 TacticalDoorBash {type=Curve, loc=<0 0 0>, rot=<0 0 0 1>}
 Custom_Frontend_Face_Idle {type=Curve, loc=<0 0 0>, rot=<0 0 0 1>}
 HandIKInAttachBoneSpace {type=Curve, loc=<0 0 0>, rot=<0 0 0 1>}
 SwingPhase {type=Curve, loc=<0 0 0>, rot=<0 0 0 1>}
 Moving {type=Curve, loc=<0 0 0>, rot=<0 0 0 1>}
 HandAttachRInChestSpace {type=Curve, loc=<0 0 0>, rot=<0 0 0 1>}
 emote_BuiltIn {type=Curve, loc=<0 0 0>, rot=<0 0 0 1>}
 OhanaPackAdj {type=Curve, loc=<0 0 0>, rot=<0 0 0 1>}
 UseRootSlopeOrientation {type=Curve, loc=<0 0 0>, rot=<0 0 0 1>}
 Backpack_Anim_Override {type=Curve, loc=<0 0 0>, rot=<0 0 0 1>}
 PhysicsGroundOffset {type=Curve, loc=<0 0 0>, rot=<0 0 0 1>}
 AllowRidingNoiseAdditive {type=Curve, loc=<0 0 0>, rot=<0 0 0 1>}
 LeftLegIKOnGroundVehicle {type=Curve, loc=<0 0 0>, rot=<0 0 0 1>}
 RightLegIKOnGroundVehicle {type=Curve, loc=<0 0 0>, rot=<0 0 0 1>}
 BaseHeadDefaultFaceIdle {type=Curve, loc=<0 0 0>, rot=<0 0 0 1>}
 Hurdle_100fwd {type=Curve, loc=<0 0 0>, rot=<0 0 0 1>}
 Hurdle_250fwd {type=Curve, loc=<0 0 0>, rot=<0 0 0 1>}
 Hurdle_100up {type=Curve, loc=<0 0 0>, rot=<0 0 0 1>}
 Hurdle_200up {type=Curve, loc=<0 0 0>, rot=<0 0 0 1>}
 Hurdle_50up {type=Curve, loc=<0 0 0>, rot=<0 0 0 1>}
 Hurdle_Mantle {type=Curve, loc=<0 0 0>, rot=<0 0 0 1>}
 Idle {type=Curve, loc=<0 0 0>, rot=<0 0 0 1>}
 Interruptible {type=Curve, loc=<0 0 0>, rot=<0 0 0 1>}
 CustomFaceIdle {type=Curve, loc=<0 0 0>, rot=<0 0 0 1>}
 Hurdle_Land {type=Curve, loc=<0 0 0>, rot=<0 0 0 1>}
 Hurdle_Fall {type=Curve, loc=<0 0 0>, rot=<0 0 0 1>}
 TacticalJump {type=Curve, loc=<0 0 0>, rot=<0 0 0 1>}
 copyWeaponRtoAttachL {type=Curve, loc=<0 0 0>, rot=<0 0 0 1>}
 CTRL_expressions_browDownL {type=Curve, loc=<0 0 0>, rot=<0 0 0 1>}
 CTRL_expressions_browDownR {type=Curve, loc=<0 0 0>, rot=<0 0 0 1>}
 CTRL_expressions_browLateralL {type=Curve, loc=<0 0 0>, rot=<0 0 0 1>}
 CTRL_expressions_browLateralR {type=Curve, loc=<0 0 0>, rot=<0 0 0 1>}
 CTRL_expressions_browRaiseInL {type=Curve, loc=<0 0 0>, rot=<0 0 0 1>}
 CTRL_expressions_browRaiseInR {type=Curve, loc=<0 0 0>, rot=<0 0 0 1>}
 CTRL_expressions_browRaiseOuterL {type=Curve, loc=<0 0 0>, rot=<0 0 0 1>}
 CTRL_expressions_browRaiseOuterR {type=Curve, loc=<0 0 0>, rot=<0 0 0 1>}
 CTRL_expressions_browUINL {type=Curve, loc=<0 0 0>, rot=<0 0 0 1>}
 CTRL_expressions_browUINR {type=Curve, loc=<0 0 0>, rot=<0 0 0 1>}
 CTRL_expressions_browUOUTL {type=Curve, loc=<0 0 0>, rot=<0 0 0 1>}
 CTRL_expressions_browUOUTR {type=Curve, loc=<0 0 0>, rot=<0 0 0 1>}
 CTRL_expressions_browDINL {type=Curve, loc=<0 0 0>, rot=<0 0 0 1>}
 CTRL_expressions_browDINR {type=Curve, loc=<0 0 0>, rot=<0 0 0 1>}
 CTRL_expressions_browDOUTL {type=Curve, loc=<0 0 0>, rot=<0 0 0 1>}
 CTRL_expressions_browDOUTR {type=Curve, loc=<0 0 0>, rot=<0 0 0 1>}
 CTRL_expressions_eyeBlinkL {type=Curve, loc=<0 0 0>, rot=<0 0 0 1>}
 CTRL_expressions_eyeBlinkR {type=Curve, loc=<0 0 0>, rot=<0 0 0 1>}
 CTRL_expressions_eyeWidenL {type=Curve, loc=<0 0 0>, rot=<0 0 0 1>}
 CTRL_expressions_eyeWidenR {type=Curve, loc=<0 0 0>, rot=<0 0 0 1>}
 CTRL_expressions_eyeSquintInnerL {type=Curve, loc=<0 0 0>, rot=<0 0 0 1>}
 CTRL_expressions_eyeSquintInnerR {type=Curve, loc=<0 0 0>, rot=<0 0 0 1>}
 CTRL_expressions_eyeCheekRaiseL {type=Curve, loc=<0 0 0>, rot=<0 0 0 1>}
 CTRL_expressions_eyeCheekRaiseR {type=Curve, loc=<0 0 0>, rot=<0 0 0 1>}
 CTRL_expressions_eyeUpperLidUpL {type=Curve, loc=<0 0 0>, rot=<0 0 0 1>}
 CTRL_expressions_eyeUpperLidUpR {type=Curve, loc=<0 0 0>, rot=<0 0 0 1>}
 CTRL_expressions_eyeRelaxL {type=Curve, loc=<0 0 0>, rot=<0 0 0 1>}
 CTRL_expressions_eyeRelaxR {type=Curve, loc=<0 0 0>, rot=<0 0 0 1>}
 CTRL_expressions_eyeLowerLidUpL {type=Curve, loc=<0 0 0>, rot=<0 0 0 1>}
 CTRL_expressions_eyeLowerLidUpR {type=Curve, loc=<0 0 0>, rot=<0 0 0 1>}
 CTRL_expressions_eyeLowerLidDownL {type=Curve, loc=<0 0 0>, rot=<0 0 0 1>}
 CTRL_expressions_eyeLowerLidDownR {type=Curve, loc=<0 0 0>, rot=<0 0 0 1>}
 CTRL_expressions_eyelidRollPosUL {type=Curve, loc=<0 0 0>, rot=<0 0 0 1>}
 CTRL_expressions_eyelidRollPosUR {type=Curve, loc=<0 0 0>, rot=<0 0 0 1>}
 CTRL_expressions_eyelidRollPosDL {type=Curve, loc=<0 0 0>, rot=<0 0 0 1>}
 CTRL_expressions_eyelidRollPosDR {type=Curve, loc=<0 0 0>, rot=<0 0 0 1>}
 CTRL_expressions_eyelidRollNegUL {type=Curve, loc=<0 0 0>, rot=<0 0 0 1>}
 CTRL_expressions_eyelidRollNegUR {type=Curve, loc=<0 0 0>, rot=<0 0 0 1>}
 CTRL_expressions_eyelidRollNegDL {type=Curve, loc=<0 0 0>, rot=<0 0 0 1>}
 CTRL_expressions_eyelidRollNegDR {type=Curve, loc=<0 0 0>, rot=<0 0 0 1>}
 CTRL_expressions_eyeLookUpL {type=Curve, loc=<0 0 0>, rot=<0 0 0 1>}
 CTRL_expressions_eyeLookUpR {type=Curve, loc=<0 0 0>, rot=<0 0 0 1>}
 CTRL_expressions_eyeLookDownL {type=Curve, loc=<0 0 0>, rot=<0 0 0 1>}
 CTRL_expressions_eyeLookDownR {type=Curve, loc=<0 0 0>, rot=<0 0 0 1>}
 CTRL_expressions_eyeLookLeftL {type=Curve, loc=<0 0 0>, rot=<0 0 0 1>}
 CTRL_expressions_eyeLookLeftR {type=Curve, loc=<0 0 0>, rot=<0 0 0 1>}
 CTRL_expressions_eyeLookRightL {type=Curve, loc=<0 0 0>, rot=<0 0 0 1>}
 CTRL_expressions_eyeLookRightR {type=Curve, loc=<0 0 0>, rot=<0 0 0 1>}
 CTRL_expressions_eyeEyelidsReduce {type=Curve, loc=<0 0 0>, rot=<0 0 0 1>}
 CTRL_expressions_noseWrinkleL {type=Curve, loc=<0 0 0>, rot=<0 0 0 1>}
 CTRL_expressions_noseWrinkleR {type=Curve, loc=<0 0 0>, rot=<0 0 0 1>}
 CTRL_expressions_noseWrinkleUpperL {type=Curve, loc=<0 0 0>, rot=<0 0 0 1>}
 CTRL_expressions_noseWrinkleUpperR {type=Curve, loc=<0 0 0>, rot=<0 0 0 1>}
 CTRL_expressions_noseNostrilDepressL {type=Curve, loc=<0 0 0>, rot=<0 0 0 1>}
 CTRL_expressions_noseNostrilDepressR {type=Curve, loc=<0 0 0>, rot=<0 0 0 1>}
 CTRL_expressions_noseNostrilDilateL {type=Curve, loc=<0 0 0>, rot=<0 0 0 1>}
 CTRL_expressions_noseNostrilDilateR {type=Curve, loc=<0 0 0>, rot=<0 0 0 1>}
 CTRL_expressions_noseNostrilCompressL {type=Curve, loc=<0 0 0>, rot=<0 0 0 1>}
 CTRL_expressions_noseNostrilCompressR {type=Curve, loc=<0 0 0>, rot=<0 0 0 1>}
 CTRL_expressions_noseRotateL {type=Curve, loc=<0 0 0>, rot=<0 0 0 1>}
 CTRL_expressions_noseRotateR {type=Curve, loc=<0 0 0>, rot=<0 0 0 1>}
 CTRL_expressions_noseSquash {type=Curve, loc=<0 0 0>, rot=<0 0 0 1>}
 CTRL_expressions_noseStretch {type=Curve, loc=<0 0 0>, rot=<0 0 0 1>}
 CTRL_expressions_mouthCheekSuckL {type=Curve, loc=<0 0 0>, rot=<0 0 0 1>}
 CTRL_expressions_mouthCheekSuckR {type=Curve, loc=<0 0 0>, rot=<0 0 0 1>}
 CTRL_expressions_mouthCheekBlowL {type=Curve, loc=<0 0 0>, rot=<0 0 0 1>}
 CTRL_expressions_mouthCheekBlowR {type=Curve, loc=<0 0 0>, rot=<0 0 0 1>}
 CTRL_expressions_mouthLipsBlowL {type=Curve, loc=<0 0 0>, rot=<0 0 0 1>}
 CTRL_expressions_mouthLipsBlowR {type=Curve, loc=<0 0 0>, rot=<0 0 0 1>}
 CTRL_expressions_mouthLeft {type=Curve, loc=<0 0 0>, rot=<0 0 0 1>}
 CTRL_expressions_mouthRight {type=Curve, loc=<0 0 0>, rot=<0 0 0 1>}
 CTRL_expressions_mouthUp {type=Curve, loc=<0 0 0>, rot=<0 0 0 1>}
 CTRL_expressions_mouthDown {type=Curve, loc=<0 0 0>, rot=<0 0 0 1>}
 CTRL_expressions_mouthUpperLipRaiseL {type=Curve, loc=<0 0 0>, rot=<0 0 0 1>}
 CTRL_expressions_mouthUpperLipRaiseR {type=Curve, loc=<0 0 0>, rot=<0 0 0 1>}
 CTRL_expressions_mouthLowerLipDepressL {type=Curve, loc=<0 0 0>, rot=<0 0 0 1>}
 CTRL_expressions_mouthLowerLipDepressR {type=Curve, loc=<0 0 0>, rot=<0 0 0 1>}
 CTRL_expressions_mouthCornerPullL {type=Curve, loc=<0 0 0>, rot=<0 0 0 1>}
 CTRL_expressions_mouthCornerPullR {type=Curve, loc=<0 0 0>, rot=<0 0 0 1>}
 CTRL_expressions_mouthStretchL {type=Curve, loc=<0 0 0>, rot=<0 0 0 1>}
 CTRL_expressions_mouthStretchR {type=Curve, loc=<0 0 0>, rot=<0 0 0 1>}
 CTRL_expressions_mouthDimpleL {type=Curve, loc=<0 0 0>, rot=<0 0 0 1>}
 CTRL_expressions_mouthDimpleR {type=Curve, loc=<0 0 0>, rot=<0 0 0 1>}
 CTRL_expressions_mouthCornerDepressL {type=Curve, loc=<0 0 0>, rot=<0 0 0 1>}
 CTRL_expressions_mouthCornerDepressR {type=Curve, loc=<0 0 0>, rot=<0 0 0 1>}
 CTRL_expressions_mouthPressUL {type=Curve, loc=<0 0 0>, rot=<0 0 0 1>}
 CTRL_expressions_mouthPressUR {type=Curve, loc=<0 0 0>, rot=<0 0 0 1>}
 CTRL_expressions_mouthPressDL {type=Curve, loc=<0 0 0>, rot=<0 0 0 1>}
 CTRL_expressions_mouthPressDR {type=Curve, loc=<0 0 0>, rot=<0 0 0 1>}
 CTRL_expressions_mouthLipsPurseUL {type=Curve, loc=<0 0 0>, rot=<0 0 0 1>}
 CTRL_expressions_mouthLipsPurseUR {type=Curve, loc=<0 0 0>, rot=<0 0 0 1>}
 CTRL_expressions_mouthLipsPurseDL {type=Curve, loc=<0 0 0>, rot=<0 0 0 1>}
 CTRL_expressions_mouthLipsPurseDR {type=Curve, loc=<0 0 0>, rot=<0 0 0 1>}
 CTRL_expressions_mouthLipsTowardsUL {type=Curve, loc=<0 0 0>, rot=<0 0 0 1>}
 CTRL_expressions_mouthLipsTowardsUR {type=Curve, loc=<0 0 0>, rot=<0 0 0 1>}
 CTRL_expressions_mouthLipsTowardsDL {type=Curve, loc=<0 0 0>, rot=<0 0 0 1>}
 CTRL_expressions_mouthLipsTowardsDR {type=Curve, loc=<0 0 0>, rot=<0 0 0 1>}
 CTRL_expressions_mouthFunnelUL {type=Curve, loc=<0 0 0>, rot=<0 0 0 1>}
 CTRL_expressions_mouthFunnelUR {type=Curve, loc=<0 0 0>, rot=<0 0 0 1>}
 CTRL_expressions_mouthFunnelDL {type=Curve, loc=<0 0 0>, rot=<0 0 0 1>}
 CTRL_expressions_mouthFunnelDR {type=Curve, loc=<0 0 0>, rot=<0 0 0 1>}
 CTRL_expressions_mouthLipsTogetherUL {type=Curve, loc=<0 0 0>, rot=<0 0 0 1>}
 CTRL_expressions_mouthLipsTogetherUR {type=Curve, loc=<0 0 0>, rot=<0 0 0 1>}
 CTRL_expressions_mouthLipsTogetherDL {type=Curve, loc=<0 0 0>, rot=<0 0 0 1>}
 CTRL_expressions_mouthLipsTogetherDR {type=Curve, loc=<0 0 0>, rot=<0 0 0 1>}
 CTRL_expressions_mouthUpperLipBiteL {type=Curve, loc=<0 0 0>, rot=<0 0 0 1>}
 CTRL_expressions_mouthUpperLipBiteR {type=Curve, loc=<0 0 0>, rot=<0 0 0 1>}
 CTRL_expressions_mouthLowerLipBiteL {type=Curve, loc=<0 0 0>, rot=<0 0 0 1>}
 CTRL_expressions_mouthLowerLipBiteR {type=Curve, loc=<0 0 0>, rot=<0 0 0 1>}
 CTRL_expressions_mouthCornerSharpenUL {type=Curve, loc=<0 0 0>, rot=<0 0 0 1>}
 CTRL_expressions_mouthCornerSharpenUR {type=Curve, loc=<0 0 0>, rot=<0 0 0 1>}
 CTRL_expressions_mouthCornerSharpenDL {type=Curve, loc=<0 0 0>, rot=<0 0 0 1>}
 CTRL_expressions_mouthCornerSharpenDR {type=Curve, loc=<0 0 0>, rot=<0 0 0 1>}
 CTRL_expressions_mouthCornerRounderUL {type=Curve, loc=<0 0 0>, rot=<0 0 0 1>}
 CTRL_expressions_mouthCornerRounderUR {type=Curve, loc=<0 0 0>, rot=<0 0 0 1>}
 CTRL_expressions_mouthCornerRounderDL {type=Curve, loc=<0 0 0>, rot=<0 0 0 1>}
 CTRL_expressions_mouthCornerRounderDR {type=Curve, loc=<0 0 0>, rot=<0 0 0 1>}
 CTRL_expressions_mouthUpperLipTowardsTeethL {type=Curve, loc=<0 0 0>, rot=<0 0 0 1>}
 CTRL_expressions_mouthUpperLipTowardsTeethR {type=Curve, loc=<0 0 0>, rot=<0 0 0 1>}
 CTRL_expressions_mouthLowerLipTowardsTeethL {type=Curve, loc=<0 0 0>, rot=<0 0 0 1>}
 CTRL_expressions_mouthLowerLipTowardsTeethR {type=Curve, loc=<0 0 0>, rot=<0 0 0 1>}
 CTRL_expressions_mouthCornerUpL {type=Curve, loc=<0 0 0>, rot=<0 0 0 1>}
 CTRL_expressions_mouthCornerUpR {type=Curve, loc=<0 0 0>, rot=<0 0 0 1>}
 CTRL_expressions_mouthCornerDownL {type=Curve, loc=<0 0 0>, rot=<0 0 0 1>}
 CTRL_expressions_mouthCornerDownR {type=Curve, loc=<0 0 0>, rot=<0 0 0 1>}
 CTRL_expressions_mouthCornerWideL {type=Curve, loc=<0 0 0>, rot=<0 0 0 1>}
 CTRL_expressions_mouthCornerWideR {type=Curve, loc=<0 0 0>, rot=<0 0 0 1>}
 CTRL_expressions_mouthCornerNarrowL {type=Curve, loc=<0 0 0>, rot=<0 0 0 1>}
 CTRL_expressions_mouthCornerNarrowR {type=Curve, loc=<0 0 0>, rot=<0 0 0 1>}
 CTRL_expressions_tongueUp {type=Curve, loc=<0 0 0>, rot=<0 0 0 1>}
 CTRL_expressions_tongueDown {type=Curve, loc=<0 0 0>, rot=<0 0 0 1>}
 CTRL_expressions_tongueLeft {type=Curve, loc=<0 0 0>, rot=<0 0 0 1>}
 CTRL_expressions_tongueRight {type=Curve, loc=<0 0 0>, rot=<0 0 0 1>}
 CTRL_expressions_tongueOut {type=Curve, loc=<0 0 0>, rot=<0 0 0 1>}
 CTRL_expressions_tongueIn {type=Curve, loc=<0 0 0>, rot=<0 0 0 1>}
 CTRL_expressions_tongueRollUp {type=Curve, loc=<0 0 0>, rot=<0 0 0 1>}
 CTRL_expressions_tongueRollDown {type=Curve, loc=<0 0 0>, rot=<0 0 0 1>}
 CTRL_expressions_tongueRollLeft {type=Curve, loc=<0 0 0>, rot=<0 0 0 1>}
 CTRL_expressions_tongueRollRight {type=Curve, loc=<0 0 0>, rot=<0 0 0 1>}
 CTRL_expressions_tongueWide {type=Curve, loc=<0 0 0>, rot=<0 0 0 1>}
 CTRL_expressions_tongueNarrow {type=Curve, loc=<0 0 0>, rot=<0 0 0 1>}
 CTRL_expressions_jawOpen {type=Curve, loc=<0 0 0>, rot=<0 0 0 1>}
 CTRL_expressions_jawLeft {type=Curve, loc=<0 0 0>, rot=<0 0 0 1>}
 CTRL_expressions_jawRight {type=Curve, loc=<0 0 0>, rot=<0 0 0 1>}
 CTRL_expressions_jawFwd {type=Curve, loc=<0 0 0>, rot=<0 0 0 1>}
 CTRL_expressions_jawBack {type=Curve, loc=<0 0 0>, rot=<0 0 0 1>}
 CTRL_expressions_jawChinRaiseDL {type=Curve, loc=<0 0 0>, rot=<0 0 0 1>}
 CTRL_expressions_jawChinRaiseDR {type=Curve, loc=<0 0 0>, rot=<0 0 0 1>}
 CTRL_expressions_jawChinRaiseUL {type=Curve, loc=<0 0 0>, rot=<0 0 0 1>}
 CTRL_expressions_jawChinRaiseUR {type=Curve, loc=<0 0 0>, rot=<0 0 0 1>}
 CTRL_expressions_jawOpenExtreme {type=Curve, loc=<0 0 0>, rot=<0 0 0 1>}
 CTRL_expressions_teethUpU {type=Curve, loc=<0 0 0>, rot=<0 0 0 1>}
 CTRL_expressions_teethUpD {type=Curve, loc=<0 0 0>, rot=<0 0 0 1>}
 CTRL_expressions_teethDownU {type=Curve, loc=<0 0 0>, rot=<0 0 0 1>}
 CTRL_expressions_teethDownD {type=Curve, loc=<0 0 0>, rot=<0 0 0 1>}
 CTRL_expressions_teethLeftU {type=Curve, loc=<0 0 0>, rot=<0 0 0 1>}
 CTRL_expressions_teethLeftD {type=Curve, loc=<0 0 0>, rot=<0 0 0 1>}
 CTRL_expressions_teethRightU {type=Curve, loc=<0 0 0>, rot=<0 0 0 1>}
 CTRL_expressions_teethRightD {type=Curve, loc=<0 0 0>, rot=<0 0 0 1>}
 CTRL_expressions_teethFwdU {type=Curve, loc=<0 0 0>, rot=<0 0 0 1>}
 CTRL_expressions_teethFwdD {type=Curve, loc=<0 0 0>, rot=<0 0 0 1>}
 CTRL_expressions_teethBackU {type=Curve, loc=<0 0 0>, rot=<0 0 0 1>}
 CTRL_expressions_teethBackD {type=Curve, loc=<0 0 0>, rot=<0 0 0 1>}
 CTRL_expressions_teethRotateFwdU {type=Curve, loc=<0 0 0>, rot=<0 0 0 1>}
 CTRL_expressions_teethRotateFwdD {type=Curve, loc=<0 0 0>, rot=<0 0 0 1>}
 CTRL_expressions_teethRotateBackU {type=Curve, loc=<0 0 0>, rot=<0 0 0 1>}
 CTRL_expressions_teethRotateBackD {type=Curve, loc=<0 0 0>, rot=<0 0 0 1>}
 is_3L {type=Curve, loc=<0 0 0>, rot=<0 0 0 1>}
 Mirrored {type=Curve, loc=<0 0 0>, rot=<0 0 0 1>}
 LockMovement {type=Curve, loc=<0 0 0>, rot=<0 0 0 1>}
 Walk {type=Curve, loc=<0 0 0>, rot=<0 0 0 1>}
 Sprint {type=Curve, loc=<0 0 0>, rot=<0 0 0 1>}
 Jog {type=Curve, loc=<0 0 0>, rot=<0 0 0 1>}
 Run {type=Curve, loc=<0 0 0>, rot=<0 0 0 1>}
 LeftFootOnGrindRail {type=Curve, loc=<0 0 0>, rot=<0 0 0 1>}
 RightFootOnGrindRail {type=Curve, loc=<0 0 0>, rot=<0 0 0 1>}
 DisableLeftHandIKHipAttach {type=Curve, loc=<0 0 0>, rot=<0 0 0 1>}
 YawCorrectionDirectionSouth {type=Curve, loc=<0 0 0>, rot=<0 0 0 1>}
 YawCorrectionDirectionEast {type=Curve, loc=<0 0 0>, rot=<0 0 0 1>}
 YawCorrectionDirectionWest {type=Curve, loc=<0 0 0>, rot=<0 0 0 1>}
 YawCorrectionDirectionNorth {type=Curve, loc=<0 0 0>, rot=<0 0 0 1>}
 is_face_bone_anim {type=Curve, loc=<0 0 0>, rot=<0 0 0 1>}
 Hover {type=Curve, loc=<0 0 0>, rot=<0 0 0 1>}
 WeaponRinFloorSpace {type=Curve, loc=<0 0 0>, rot=<0 0 0 1>}
 AlphaBlur_R {type=Curve, loc=<0 0 0>, rot=<0 0 0 1>}
 AlphaBlur_L {type=Curve, loc=<0 0 0>, rot=<0 0 0 1>}
 RotationDirection_R {type=Curve, loc=<0 0 0>, rot=<0 0 0 1>}
 RotationDirection_L {type=Curve, loc=<0 0 0>, rot=<0 0 0 1>}
 PlayLowerBodyLocomotionOverFullBodySlot {type=Curve, loc=<0 0 0>, rot=<0 0 0 1>}
 Distance {type=Curve, loc=<0 0 0>, rot=<0 0 0 1>}
 UserCurve01 {type=Curve, loc=<0 0 0>, rot=<0 0 0 1>}
 UserCurve02 {type=Curve, loc=<0 0 0>, rot=<0 0 0 1>}
 LegIKInAttachBoneSpace {type=Curve, loc=<0 0 0>, rot=<0 0 0 1>}
 MoveData_Speed {type=Curve, loc=<0 0 0>, rot=<0 0 0 1>}
 FootSpeed_L {type=Curve, loc=<0 0 0>, rot=<0 0 0 1>}
 FootSpeed_R {type=Curve, loc=<0 0 0>, rot=<0 0 0 1>}
 RightElbowTwist {type=Curve, loc=<0 0 0>, rot=<0 0 0 1>}
 LeftElbowTwist {type=Curve, loc=<0 0 0>, rot=<0 0 0 1>}
 Intensity {type=Curve, loc=<0 0 0>, rot=<0 0 0 1>}
 head {type=Curve, loc=<0 0 0>, rot=<0 0 0 1>}
 neck_01 {type=Curve, loc=<0 0 0>, rot=<0 0 0 1>}
 neck_02 {type=Curve, loc=<0 0 0>, rot=<0 0 0 1>}
 Spine_01 {type=Curve, loc=<0 0 0>, rot=<0 0 0 1>}
 spine_04 {type=Curve, loc=<0 0 0>, rot=<0 0 0 1>}
 Spine_03 {type=Curve, loc=<0 0 0>, rot=<0 0 0 1>}
 spine_05 {type=Curve, loc=<0 0 0>, rot=<0 0 0 1>}
 upperarm_r {type=Curve, loc=<0 0 0>, rot=<0 0 0 1>}
 hand_l {type=Curve, loc=<0 0 0>, rot=<0 0 0 1>}
 lowerarm_l {type=Curve, loc=<0 0 0>, rot=<0 0 0 1>}
 lowerarm_r {type=Curve, loc=<0 0 0>, rot=<0 0 0 1>}
 lowerarm_twist_01_l {type=Curve, loc=<0 0 0>, rot=<0 0 0 1>}
 upperarm_l {type=Curve, loc=<0 0 0>, rot=<0 0 0 1>}
 ball_r {type=Curve, loc=<0 0 0>, rot=<0 0 0 1>}
 index_metacarpal_l {type=Curve, loc=<0 0 0>, rot=<0 0 0 1>}
 clavicle_l {type=Curve, loc=<0 0 0>, rot=<0 0 0 1>}
 hand_r {type=Curve, loc=<0 0 0>, rot=<0 0 0 1>}
 Spine_02 {type=Curve, loc=<0 0 0>, rot=<0 0 0 1>}
 calf_r {type=Curve, loc=<0 0 0>, rot=<0 0 0 1>}
 calf_twist_01_r {type=Curve, loc=<0 0 0>, rot=<0 0 0 1>}
 calf_twist_02_r {type=Curve, loc=<0 0 0>, rot=<0 0 0 1>}
 clavicle_r {type=Curve, loc=<0 0 0>, rot=<0 0 0 1>}
 ik_hand_r {type=Curve, loc=<0 0 0>, rot=<0 0 0 1>}
 thigh_r {type=Curve, loc=<0 0 0>, rot=<0 0 0 1>}
 thigh_twist_01_r {type=Curve, loc=<0 0 0>, rot=<0 0 0 1>}
 calf_l {type=Curve, loc=<0 0 0>, rot=<0 0 0 1>}
 thigh_l {type=Curve, loc=<0 0 0>, rot=<0 0 0 1>}
 pelvis {type=Curve, loc=<0 0 0>, rot=<0 0 0 1>}
 thumb_01_l {type=Curve, loc=<0 0 0>, rot=<0 0 0 1>}
 thumb_02_l {type=Curve, loc=<0 0 0>, rot=<0 0 0 1>}
 thumb_03_l {type=Curve, loc=<0 0 0>, rot=<0 0 0 1>}
 middle_metacarpal_l {type=Curve, loc=<0 0 0>, rot=<0 0 0 1>}
 middle_01_l {type=Curve, loc=<0 0 0>, rot=<0 0 0 1>}
 middle_02_l {type=Curve, loc=<0 0 0>, rot=<0 0 0 1>}
 middle_03_l {type=Curve, loc=<0 0 0>, rot=<0 0 0 1>}
 pinky_01_l {type=Curve, loc=<0 0 0>, rot=<0 0 0 1>}
 pinky_02_l {type=Curve, loc=<0 0 0>, rot=<0 0 0 1>}
 pinky_03_l {type=Curve, loc=<0 0 0>, rot=<0 0 0 1>}
 ring_01_l {type=Curve, loc=<0 0 0>, rot=<0 0 0 1>}
 ring_02_l {type=Curve, loc=<0 0 0>, rot=<0 0 0 1>}
 ring_03_l {type=Curve, loc=<0 0 0>, rot=<0 0 0 1>}
 index_01_r {type=Curve, loc=<0 0 0>, rot=<0 0 0 1>}
 index_02_r {type=Curve, loc=<0 0 0>, rot=<0 0 0 1>}
 index_03_r {type=Curve, loc=<0 0 0>, rot=<0 0 0 1>}
 lowerarm_twist_01_r {type=Curve, loc=<0 0 0>, rot=<0 0 0 1>}
 lowerarm_twist_02_r {type=Curve, loc=<0 0 0>, rot=<0 0 0 1>}
 pinky_metacarpal_r {type=Curve, loc=<0 0 0>, rot=<0 0 0 1>}
 ring_01_r {type=Curve, loc=<0 0 0>, rot=<0 0 0 1>}
 ring_02_r {type=Curve, loc=<0 0 0>, rot=<0 0 0 1>}
 thumb_01_r {type=Curve, loc=<0 0 0>, rot=<0 0 0 1>}
 thumb_02_r {type=Curve, loc=<0 0 0>, rot=<0 0 0 1>}
 thumb_03_r {type=Curve, loc=<0 0 0>, rot=<0 0 0 1>}
 foot_r {type=Curve, loc=<0 0 0>, rot=<0 0 0 1>}
 upperarm_twist_01_l {type=Curve, loc=<0 0 0>, rot=<0 0 0 1>}
 upperarm_twist_01_r {type=Curve, loc=<0 0 0>, rot=<0 0 0 1>}
 Root {type=Curve, loc=<0 0 0>, rot=<0 0 0 1>}
 foot_l {type=Curve, loc=<0 0 0>, rot=<0 0 0 1>}
 HandAttachLinFloorSpace {type=Curve, loc=<0 0 0>, rot=<0 0 0 1>}
 HandAttachRinFloorSpace {type=Curve, loc=<0 0 0>, rot=<0 0 0 1>}
 CanBypassSlideInitiate {type=Curve, loc=<0 0 0>, rot=<0 0 0 1>}
 AttachEmote1InPelvisSpace {type=Curve, loc=<0 0 0>, rot=<0 0 0 1>}
 AttachEmote2InPelvisSpace {type=Curve, loc=<0 0 0>, rot=<0 0 0 1>}
 AttachEmote3InPelvisSpace {type=Curve, loc=<0 0 0>, rot=<0 0 0 1>}
 LeftLegFK {type=Curve, loc=<0 0 0>, rot=<0 0 0 1>}
 RightLegFK {type=Curve, loc=<0 0 0>, rot=<0 0 0 1>}
 Disable_JamV2_FootShiftAdditive {type=Curve, loc=<0 0 0>, rot=<0 0 0 1>}
 Disable_JamV2_SpineMotionAdditive {type=Curve, loc=<0 0 0>, rot=<0 0 0 1>}
 Disable_JamV2_HeadMotionReplace {type=Curve, loc=<0 0 0>, rot=<0 0 0 1>}
 Disable_JamV2_LeftArmReplace {type=Curve, loc=<0 0 0>, rot=<0 0 0 1>}
 Disable_JamV2_RightArmReplace {type=Curve, loc=<0 0 0>, rot=<0 0 0 1>}
 CTRL_expressions_browInnerInL {type=Curve, loc=<0 0 0>, rot=<0 0 0 1>}
 CTRL_expressions_browInnerInR {type=Curve, loc=<0 0 0>, rot=<0 0 0 1>}
 CTRL_expressions_browInnerMidDownL {type=Curve, loc=<0 0 0>, rot=<0 0 0 1>}
 CTRL_expressions_browInnerMidDownR {type=Curve, loc=<0 0 0>, rot=<0 0 0 1>}
 CTRL_expressions_browInnerMidInL {type=Curve, loc=<0 0 0>, rot=<0 0 0 1>}
 CTRL_expressions_browInnerMidInR {type=Curve, loc=<0 0 0>, rot=<0 0 0 1>}
 CTRL_expressions_browInnerMidOutL {type=Curve, loc=<0 0 0>, rot=<0 0 0 1>}
 CTRL_expressions_browInnerMidOutR {type=Curve, loc=<0 0 0>, rot=<0 0 0 1>}
 CTRL_expressions_browInnerMidUpL {type=Curve, loc=<0 0 0>, rot=<0 0 0 1>}
 CTRL_expressions_browInnerMidUpR {type=Curve, loc=<0 0 0>, rot=<0 0 0 1>}
 CTRL_expressions_browInnerOutL {type=Curve, loc=<0 0 0>, rot=<0 0 0 1>}
 CTRL_expressions_browInnerOutR {type=Curve, loc=<0 0 0>, rot=<0 0 0 1>}
 CTRL_expressions_browOuterMidDownL {type=Curve, loc=<0 0 0>, rot=<0 0 0 1>}
 CTRL_expressions_browOuterMidDownR {type=Curve, loc=<0 0 0>, rot=<0 0 0 1>}
 CTRL_expressions_browOuterMidInL {type=Curve, loc=<0 0 0>, rot=<0 0 0 1>}
 CTRL_expressions_browOuterMidInR {type=Curve, loc=<0 0 0>, rot=<0 0 0 1>}
 CTRL_expressions_browOuterMidOutL {type=Curve, loc=<0 0 0>, rot=<0 0 0 1>}
 CTRL_expressions_browOuterMidOutR {type=Curve, loc=<0 0 0>, rot=<0 0 0 1>}
 CTRL_expressions_browOuterMidUpL {type=Curve, loc=<0 0 0>, rot=<0 0 0 1>}
 CTRL_expressions_browOuterMidUpR {type=Curve, loc=<0 0 0>, rot=<0 0 0 1>}
 CTRL_expressions_browRollNegInL {type=Curve, loc=<0 0 0>, rot=<0 0 0 1>}
 CTRL_expressions_browRollNegInR {type=Curve, loc=<0 0 0>, rot=<0 0 0 1>}
 CTRL_expressions_browRollNegOutL {type=Curve, loc=<0 0 0>, rot=<0 0 0 1>}
 CTRL_expressions_browRollNegOutR {type=Curve, loc=<0 0 0>, rot=<0 0 0 1>}
 CTRL_expressions_browRollPosInL {type=Curve, loc=<0 0 0>, rot=<0 0 0 1>}
 CTRL_expressions_browRollPosInR {type=Curve, loc=<0 0 0>, rot=<0 0 0 1>}
 CTRL_expressions_browRollPosOutL {type=Curve, loc=<0 0 0>, rot=<0 0 0 1>}
 CTRL_expressions_browRollPosOutR {type=Curve, loc=<0 0 0>, rot=<0 0 0 1>}
 CTRL_expressions_eyeFaceScrunchL {type=Curve, loc=<0 0 0>, rot=<0 0 0 1>}
 CTRL_expressions_eyeFaceScrunchR {type=Curve, loc=<0 0 0>, rot=<0 0 0 1>}
 CTRL_expressions_eyePupilNarrowL {type=Curve, loc=<0 0 0>, rot=<0 0 0 1>}
 CTRL_expressions_eyePupilNarrowR {type=Curve, loc=<0 0 0>, rot=<0 0 0 1>}
 CTRL_expressions_eyePupilWideL {type=Curve, loc=<0 0 0>, rot=<0 0 0 1>}
 CTRL_expressions_eyePupilWideR {type=Curve, loc=<0 0 0>, rot=<0 0 0 1>}
 CTRL_expressions_mouthLipsPressL {type=Curve, loc=<0 0 0>, rot=<0 0 0 1>}
 CTRL_expressions_mouthLipsPressR {type=Curve, loc=<0 0 0>, rot=<0 0 0 1>}
 CTRL_expressions_mouthLipsPullDL {type=Curve, loc=<0 0 0>, rot=<0 0 0 1>}
 CTRL_expressions_mouthLipsPullDR {type=Curve, loc=<0 0 0>, rot=<0 0 0 1>}
 CTRL_expressions_mouthLipsPullUL {type=Curve, loc=<0 0 0>, rot=<0 0 0 1>}
 CTRL_expressions_mouthLipsPullUR {type=Curve, loc=<0 0 0>, rot=<0 0 0 1>}
 CTRL_expressions_mouthLipsPushDL {type=Curve, loc=<0 0 0>, rot=<0 0 0 1>}
 CTRL_expressions_mouthLipsPushDR {type=Curve, loc=<0 0 0>, rot=<0 0 0 1>}
 CTRL_expressions_mouthLipsPushUL {type=Curve, loc=<0 0 0>, rot=<0 0 0 1>}
 CTRL_expressions_mouthLipsPushUR {type=Curve, loc=<0 0 0>, rot=<0 0 0 1>}
 CTRL_expressions_mouthLipsThickDL {type=Curve, loc=<0 0 0>, rot=<0 0 0 1>}
 CTRL_expressions_mouthLipsThickDR {type=Curve, loc=<0 0 0>, rot=<0 0 0 1>}
 CTRL_expressions_mouthLipsThickInwardDL {type=Curve, loc=<0 0 0>, rot=<0 0 0 1>}
 CTRL_expressions_mouthLipsThickInwardDR {type=Curve, loc=<0 0 0>, rot=<0 0 0 1>}
 CTRL_expressions_mouthLipsThickInwardUL {type=Curve, loc=<0 0 0>, rot=<0 0 0 1>}
 CTRL_expressions_mouthLipsThickInwardUR {type=Curve, loc=<0 0 0>, rot=<0 0 0 1>}
 CTRL_expressions_mouthLipsThickUL {type=Curve, loc=<0 0 0>, rot=<0 0 0 1>}
 CTRL_expressions_mouthLipsThickUR {type=Curve, loc=<0 0 0>, rot=<0 0 0 1>}
 CTRL_expressions_mouthLipsThinDL {type=Curve, loc=<0 0 0>, rot=<0 0 0 1>}
 CTRL_expressions_mouthLipsThinDR {type=Curve, loc=<0 0 0>, rot=<0 0 0 1>}
 CTRL_expressions_mouthLipsThinInwardDL {type=Curve, loc=<0 0 0>, rot=<0 0 0 1>}
 CTRL_expressions_mouthLipsThinInwardDR {type=Curve, loc=<0 0 0>, rot=<0 0 0 1>}
 CTRL_expressions_mouthLipsThinInwardUL {type=Curve, loc=<0 0 0>, rot=<0 0 0 1>}
 CTRL_expressions_mouthLipsThinInwardUR {type=Curve, loc=<0 0 0>, rot=<0 0 0 1>}
 CTRL_expressions_mouthLipsThinUL {type=Curve, loc=<0 0 0>, rot=<0 0 0 1>}
 CTRL_expressions_mouthLipsThinUR {type=Curve, loc=<0 0 0>, rot=<0 0 0 1>}
 CTRL_expressions_mouthLipsTightenDL {type=Curve, loc=<0 0 0>, rot=<0 0 0 1>}
 CTRL_expressions_mouthLipsTightenDR {type=Curve, loc=<0 0 0>, rot=<0 0 0 1>}
 CTRL_expressions_mouthLipsTightenUL {type=Curve, loc=<0 0 0>, rot=<0 0 0 1>}
 CTRL_expressions_mouthLipsTightenUR {type=Curve, loc=<0 0 0>, rot=<0 0 0 1>}
 CTRL_expressions_mouthLowerLipRollInL {type=Curve, loc=<0 0 0>, rot=<0 0 0 1>}
 CTRL_expressions_mouthLowerLipRollInR {type=Curve, loc=<0 0 0>, rot=<0 0 0 1>}
 CTRL_expressions_mouthLowerLipRollOutL {type=Curve, loc=<0 0 0>, rot=<0 0 0 1>}
 CTRL_expressions_mouthLowerLipRollOutR {type=Curve, loc=<0 0 0>, rot=<0 0 0 1>}
 CTRL_expressions_mouthSharpCornerPullL {type=Curve, loc=<0 0 0>, rot=<0 0 0 1>}
 CTRL_expressions_mouthSharpCornerPullR {type=Curve, loc=<0 0 0>, rot=<0 0 0 1>}
 CTRL_expressions_mouthUpperLipRollInL {type=Curve, loc=<0 0 0>, rot=<0 0 0 1>}
 CTRL_expressions_mouthUpperLipRollInR {type=Curve, loc=<0 0 0>, rot=<0 0 0 1>}
 CTRL_expressions_mouthUpperLipRollOutL {type=Curve, loc=<0 0 0>, rot=<0 0 0 1>}
 CTRL_expressions_mouthUpperLipRollOutR {type=Curve, loc=<0 0 0>, rot=<0 0 0 1>}
 CTRL_expressions_noseNasolabialDeepenL {type=Curve, loc=<0 0 0>, rot=<0 0 0 1>}
 CTRL_expressions_noseNasolabialDeepenR {type=Curve, loc=<0 0 0>, rot=<0 0 0 1>}
 CTRL_expressions_tongueBendDown {type=Curve, loc=<0 0 0>, rot=<0 0 0 1>}
 CTRL_expressions_tongueBendUp {type=Curve, loc=<0 0 0>, rot=<0 0 0 1>}
 CTRL_expressions_tongueTwistLeft {type=Curve, loc=<0 0 0>, rot=<0 0 0 1>}
 CTRL_expressions_tongueTwistRight {type=Curve, loc=<0 0 0>, rot=<0 0 0 1>}
 WM_BrowLateral_L {type=Curve, loc=<0 0 0>, rot=<0 0 0 1>}
 WM_BrowLateral_Offset_L {type=Curve, loc=<0 0 0>, rot=<0 0 0 1>}
 WM_BrowLateral_Offset_R {type=Curve, loc=<0 0 0>, rot=<0 0 0 1>}
 WM_BrowLateral_R {type=Curve, loc=<0 0 0>, rot=<0 0 0 1>}
 WM_BrowRaiseInner_L {type=Curve, loc=<0 0 0>, rot=<0 0 0 1>}
 WM_BrowRaiseInner_Offset_L {type=Curve, loc=<0 0 0>, rot=<0 0 0 1>}
 WM_BrowRaiseInner_Offset_R {type=Curve, loc=<0 0 0>, rot=<0 0 0 1>}
 WM_BrowRaiseInner_R {type=Curve, loc=<0 0 0>, rot=<0 0 0 1>}
 WM_BrowRaiseOutter_L {type=Curve, loc=<0 0 0>, rot=<0 0 0 1>}
 WM_BrowRaiseOutter_Offset_L {type=Curve, loc=<0 0 0>, rot=<0 0 0 1>}
 WM_BrowRaiseOutter_Offset_R {type=Curve, loc=<0 0 0>, rot=<0 0 0 1>}
 WM_BrowRaiseOutter_R {type=Curve, loc=<0 0 0>, rot=<0 0 0 1>}
 WM_CheekRaise_L {type=Curve, loc=<0 0 0>, rot=<0 0 0 1>}
 WM_CheekRaise_Offset_L {type=Curve, loc=<0 0 0>, rot=<0 0 0 1>}
 WM_CheekRaise_Offset_R {type=Curve, loc=<0 0 0>, rot=<0 0 0 1>}
 WM_CheekRaise_R {type=Curve, loc=<0 0 0>, rot=<0 0 0 1>}
 WM_Lips_L {type=Curve, loc=<0 0 0>, rot=<0 0 0 1>}
 WM_Lips_Offset_L {type=Curve, loc=<0 0 0>, rot=<0 0 0 1>}
 WM_Lips_Offset_R {type=Curve, loc=<0 0 0>, rot=<0 0 0 1>}
 WM_Lips_R {type=Curve, loc=<0 0 0>, rot=<0 0 0 1>}
 WM_MouthStretch_L {type=Curve, loc=<0 0 0>, rot=<0 0 0 1>}
 WM_MouthStretch_Offset_L {type=Curve, loc=<0 0 0>, rot=<0 0 0 1>}
 WM_MouthStretch_Offset_R {type=Curve, loc=<0 0 0>, rot=<0 0 0 1>}
 WM_MouthStretch_R {type=Curve, loc=<0 0 0>, rot=<0 0 0 1>}
 WM_NoseWrinkle_L {type=Curve, loc=<0 0 0>, rot=<0 0 0 1>}
 WM_NoseWrinkle_Offset_L {type=Curve, loc=<0 0 0>, rot=<0 0 0 1>}
 WM_NoseWrinkle_Offset_R {type=Curve, loc=<0 0 0>, rot=<0 0 0 1>}
 WM_NoseWrinkle_R {type=Curve, loc=<0 0 0>, rot=<0 0 0 1>}
 WM_Smile_L {type=Curve, loc=<0 0 0>, rot=<0 0 0 1>}
 WM_Smile_Offset_L {type=Curve, loc=<0 0 0>, rot=<0 0 0 1>}
 WM_Smile_Offset_R {type=Curve, loc=<0 0 0>, rot=<0 0 0 1>}
 WM_Smile_R {type=Curve, loc=<0 0 0>, rot=<0 0 0 1>}
]
variables [
 Enable Weapon Scale {type=bool, default=True}
 Hand IK Alpha Clamp {type=double, default=1}
 Hand IK Retargeting Weight {type=double, default=0.5}
 IKRig_IKRigAsset {
  type="TObjectPtr<UIKRigDefinition>"
  default=/Script/IKRig.IKRigDefinition'&/IK_Biped_Base'
 }
 Left Hand IK Alpha {type=double, default=0}
 Right Hand IK Alpha {type=double, default=0}
 Run IK {type=bool, default=True}
 SetScale_Item_Name {type="TArray<FName>", default=("weapon_l","weapon_r","hand_attach_l","hand_attach_r","ik_hand_gun")}
 Snap IK Hand Gun {type=bool, default=True}
 Weapon Scale {type=double, default=1}
]
