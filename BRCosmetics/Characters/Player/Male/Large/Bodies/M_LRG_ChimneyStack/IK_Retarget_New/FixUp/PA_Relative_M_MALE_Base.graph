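# Physics asset for a male base skeleton, stored with the IK retarget fix-up assets of the large ChimneyStack player body.
# It approximates the body with one capsule on each major bone: pelvis, two spine bones, neck, head, and both arms and legs down to hands and feet.
# Constraints chain the capsules into a tree rooted at the pelvis, with elbow, neck and head joints explicitly set to free swing and twist.
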PhysicsAsset /BRCosmetics/Characters/Player/Male/Large/Bodies/M_LRG_ChimneyStack/IK_Retarget_New/FixUp/PA_Relative_M_MALE_Base
bodies {
 calf_l [{shapeType=Capsule, length=33, radius=6, shapeName=calf_l_capsule, center=<-21.4627 -2.5351 0>, rotation=<rot -90 -0.0003 0>}]
 calf_r [{shapeType=Capsule, length=33, radius=6, shapeName=calf_r_capsule, center=<21.4627 2.5351 0>, rotation=<rot -90 -166.7373 -13.263>}]
 foot_l [{shapeType=Capsule, length=15.7046, radius=6.1683, shapeName=foot_l_capsule, center=<-3.9617 7.5075 0>, rotation=<rot 0 0 90>}]
 foot_r [{shapeType=Capsule, length=15.7046, radius=6.1683, shapeName=foot_r_capsule, center=<3.9617 -7.5075 0>, rotation=<rot 0 0 -90>}]
 hand_l [{shapeType=Capsule, length=11, radius=4.5, shapeName=hand_l_capsule, center=<5.5152 -1.1721 0>, rotation=<rot 90 -9.9998 0>}]
 hand_r [{shapeType=Capsule, length=11, radius=4.5, shapeName=hand_r_capsule, center=<-5.5152 1.1721 0>, rotation=<rot 90 0 -170.0001>}]
 head [{shapeType=Capsule, length=8, radius=7, shapeName=head_capsule, center=<3.9204 4.023 0>, rotation=<rot 0 49.9984 -90>}]
 lowerarm_l [{shapeType=Capsule, length=16, radius=4.5, shapeName=lowerarm_l_capsule, center=<9.0972 0.3092 0>, rotation=<rot 90 -179.9992 0>}]
 lowerarm_r [{shapeType=Capsule, length=16, radius=4.5, shapeName=lowerarm_r_capsule, center=<-9.0972 -0.3092 0>, rotation=<rot 90 180 179.9992>}]
 neck_02 [{shapeType=Capsule, length=4, radius=5, shapeName=neck_02_capsule, center=<-1.7834 0.3806 0>, rotation=<rot -90 -0.0001 -0.0002>}]
 pelvis [{shapeType=Capsule, length=7.5, radius=13, shapeName=pelvis_capsule, center=<5.0345 -1.3168 0>, rotation=<rot -90 -0.0007 0>}]
 spine_03 [{shapeType=Capsule, length=6, radius=11.5, shapeName=Spine_03_capsule, center=<6.2202 0.5311 0>}]
 spine_05 [{shapeType=Capsule, length=5.5, radius=14, shapeName=spine_05_capsule, center=<-2.4431 0 0>, rotation=<rot 0 -0.0012 0>}]
 thigh_l [{shapeType=Capsule, length=29, radius=7, shapeName=thigh_l_capsule, center=<-18.5125 0.0008 0>, rotation=<rot -90 -0.0032 0>}]
 thigh_r [{shapeType=Capsule, length=29, radius=7, shapeName=thigh_r_capsule, center=<18.5125 -0.0008 0>, rotation=<rot -90 -160.5284 -19.4748>}]
 upperarm_l [{shapeType=Capsule, length=20.5, radius=5, shapeName=upperarm_l_capsule, center=<11.1091 0.0001 0>, rotation=<rot 0.0001 90.0011 -90>}]
 upperarm_r [{shapeType=Capsule, length=20.5, radius=5, shapeName=upperarm_r_capsule, center=<-11.1091 -0.0001 0>, rotation=<rot 0.0001 90.0011 90>}]
}
constraints [
 {bone1Name=spine_03, bone2Name=pelvis, swing1LimitDegrees=45, swing2LimitDegrees=45, twistLimitDegrees=45}
 {bone1Name=spine_05, bone2Name=spine_03, swing1LimitDegrees=45, swing2LimitDegrees=45, twistLimitDegrees=45}
 {bone1Name=upperarm_l, bone2Name=spine_05, swing1LimitDegrees=45, swing2LimitDegrees=45, twistLimitDegrees=45}
 {bone1Name=hand_l, bone2Name=lowerarm_l, swing1LimitDegrees=45, swing2LimitDegrees=45, twistLimitDegrees=45}
 {bone1Name=upperarm_r, bone2Name=spine_05, swing1LimitDegrees=45, swing2LimitDegrees=45, twistLimitDegrees=45}
 {bone1Name=hand_r, bone2Name=lowerarm_r, swing1LimitDegrees=45, swing2LimitDegrees=45, twistLimitDegrees=45}
 {bone1Name=thigh_l, bone2Name=pelvis, swing1LimitDegrees=45, swing2LimitDegrees=45, twistLimitDegrees=45}
 {bone1Name=calf_l, bone2Name=thigh_l, swing1LimitDegrees=45, swing2LimitDegrees=45, twistLimitDegrees=45}
 {bone1Name=foot_l, bone2Name=calf_l, swing1LimitDegrees=45, swing2LimitDegrees=45, twistLimitDegrees=45}
 {bone1Name=thigh_r, bone2Name=pelvis, swing1LimitDegrees=45, swing2LimitDegrees=45, twistLimitDegrees=45}
 {bone1Name=calf_r, bone2Name=thigh_r, swing1LimitDegrees=45, swing2LimitDegrees=45, twistLimitDegrees=45}
 {bone1Name=foot_r, bone2Name=calf_r, swing1LimitDegrees=45, swing2LimitDegrees=45, twistLimitDegrees=45}
 {
  bone1Name=lowerarm_r
  bone2Name=upperarm_r
  swing1LimitDegrees=45
  swing1Motion=Free
  swing2LimitDegrees=45
  swing2Motion=Free
  twistLimitDegrees=45
  twistMotion=Free
 }
 {
  bone1Name=lowerarm_l
  bone2Name=upperarm_l
  swing1LimitDegrees=45
  swing1Motion=Free
  swing2LimitDegrees=45
  swing2Motion=Free
  twistLimitDegrees=45
  twistMotion=Free
 }
 {
  bone1Name=neck_02
  bone2Name=spine_05
  swing1LimitDegrees=45
  swing1Motion=Free
  swing2LimitDegrees=45
  swing2Motion=Free
  twistLimitDegrees=45
  twistMotion=Free
 }
 {
  bone1Name=head
  bone2Name=neck_02
  swing1LimitDegrees=45
  swing1Motion=Free
  swing2LimitDegrees=45
  swing2Motion=Free
  twistLimitDegrees=45
  twistMotion=Free
 }
]
