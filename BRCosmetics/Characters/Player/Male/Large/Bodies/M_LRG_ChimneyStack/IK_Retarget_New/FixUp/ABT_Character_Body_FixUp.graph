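# Template AnimBlueprint that retargets Fortnite player animation onto character bodies and fixes up the result.
# Retarget Pose From Mesh feeds hand IK Control Rig, optional de-intersection retarget while parachuting upright, weapon bone offsets, curve-driven blend to copied pose, then twist-bone rig.
# Manual Reinit copies retarget profiles and configures op controllers for body intersection, pelvis, stretch, floor constraint and others.
# Apply Ops Settings updates those ops, varying by gliding tag; emote events track prop attachments.
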AnimBlueprint /BRCosmetics/Characters/Player/Male/Large/Bodies/M_LRG_ChimneyStack/IK_Retarget_New/FixUp/ABT_Character_Body_FixUp
target_skeleton=null
graphs [
 AnimGraph {nodes [
   n1 = Control Rig(Source=Use cached pose 'Retarget Pose'()) {
    bSetRefPoseFromSkeleton=true
    alphaBoolBlend {blendOption=Linear}
    controlRigAssetReference {blueprintRigClass=&/MR_TwistBones.MR_TwistBones_C}
   }
   Output Pose(Result=@n1)
   n2 = Control Rig!(
    Source=Retarget Pose From Mesh!(
     IKRetargeterAsset:IKRetargeter Object Reference=Get IKRetargeterAsset() -> (IKRetargeterAsset:IKRetargeter Object Reference)
     CustomRetargetProfile:Retarget Profile Structure=Get Current Retarget Profile() -> (Current Retarget Profile:Retarget Profile Structure)
    ) -> (Pose:Pose Link Structure) {is=title AnimGraphNode_RetargetPoseFromMesh, lODThresholdForIK=1}
    Alpha
    Left Hand IK Alpha:double=Get Left Hand IK Alpha() -> (Left Hand IK Alpha:double)
    Right Hand IK Alpha:double=Get Right Hand IK Alpha() -> (Right Hand IK Alpha:double)
    Hand IK Retargeting Weight:double=Get Hand IK Retargeting Weight() -> (Hand IK Retargeting Weight:double)
    Hand IK Alpha Clamp:double=Get Hand IK Alpha Clamp() -> (Hand IK Alpha Clamp:double)
    Weapon Scale:double=Get Weapon Scale() -> (Weapon Scale:double)
   ) -> (Pose) {controlRigAssetReference {blueprintRigClass=&/CR_HandIK_Fixup.CR_HandIK_Fixup_C}}
   Save cached pose 'Post Hand IK'(Pose=@n2) {cacheName=Post Hand IK}
   n3 = Control Rig(
    Source=Retarget Input Pose!(
     Source:Pose Link Structure=Use cached pose 'Post Hand IK'()
     IKRetargeterAsset:IKRetargeter Object Reference=&/IKR_Biped_Base_IK_Post
     CustomRetargetProfile:Retarget Profile Structure=Get Intersection Retarget Profile() -> (Intersection Retarget Profile:Retarget Profile Structure)
    ) -> (Pose:Pose Link Structure) {is=title AnimGraphNode_RetargetPoseFromMesh, lODThresholdForIK=1, retargetFrom=SourcePosePin}
    Left Hand IK Alpha=Get Left Hand IK Alpha()
    Right Hand IK Alpha=Get Right Hand IK Alpha()
    Hand IK Retargeting Weight=1
    Hand IK Alpha Clamp=Get Hand IK Alpha Clamp()
    Weapon Scale=Get Weapon Scale()
   ) {controlRigAssetReference {blueprintRigClass=&CR_HandIK_Fixup_C}}
   n4 = Blend Poses by bool(BlendPose_0=@n3, BlendPose_1=Use cached pose 'Post Hand IK'(), bActiveValue=Get Hand IK De-Intersect() -> (Hand IK De-Intersect:bool))
   n5 = Use cached pose 'Post Hand IK'()
   n6 = Get bParachuteOpenedAndStandingUpright() -> (bParachuteOpenedAndStandingUpright:bool)
   n7 = "Transform (Modify) Bone - Bone: weapon_r"!(
    ComponentPose=Local To Component(LocalPose=Blend Poses by bool(BlendPose_0=@n5, BlendPose_1=@n4, bActiveValue=@n6))
    Translation=Get Weapon Offset() -> (Weapon Offset:Vector)
   ) -> (Pose) {translationMode=BMM_Additive, translationSpace=BCS_BoneSpace, boneToModify {boneName=weapon_r}}
   n8 = Get Weapon Offset()
   n9 = "Transform (Modify) Bone - Bone: hand_attach_l"!(
    ComponentPose="Transform (Modify) Bone - Bone: weapon_l"!(
     ComponentPose="Transform (Modify) Bone - Bone: hand_attach_r"!(ComponentPose=@n7, Translation=@n8) -> (Pose) {translationMode=BMM_Additive, translationSpace=BCS_BoneSpace, boneToModify {boneName=hand_attach_r}}
     Translation=Get Weapon Offset()
    ) -> (Pose) {translationMode=BMM_Additive, translationSpace=BCS_BoneSpace, boneToModify {boneName=weapon_l}}
    Translation=Get Weapon Offset()
   ) -> (Pose) {translationMode=BMM_Additive, translationSpace=BCS_BoneSpace, boneToModify {boneName=hand_attach_l}}
   Save cached pose 'Retarget Pose'(Pose=Two Way Blend(
     A=Component To Local(ComponentPose=@n9)
     B=Copy Pose From Mesh() {bCopyCurves=true, bUseAttachedParent=true}
     Alpha=Get RetargetOverrideBlend() -> (RetargetOverrideBlend:float)
    ) {blendNode {alphaBoolBlend {blendOption=Linear}}}) {cacheName=Retarget Pose}
  ]}
 Apply Ops Settings {nodes [
   Apply Ops Settings() {is=title K2Node_FunctionEntry}
   > n1 = Sequence!(execute) -> (then_0, then_1, then_2:Exec, then_3:Exec, then_4:Exec, then_5:Exec, then_6:Exec, then_7:Exec, then_8:Exec)
   Check Controller!(
    execute=@n1.then_2
    self
    Controller:Object Reference=Get Root Offset Pin Bone Controller() -> (Root Offset Pin Bone Controller:IKRetarget Pin Bone Controller Object Reference)
   ) -> (then, Valid:bool)
   > SetPinBoneOffsets!(
    execute
    self
    Controller:IKRetarget Pin Bone Controller Object Reference=Get Root Offset Pin Bone Controller()
    Global Offset:Vector="0, 0, 0"
    Local Offset:Vector=Get CurrentPelvisTranslationOffset() -> (CurrentPelvisTranslationOffset:Vector)
   ) -> (then)
   Return Node(execute=@n1.then_8)
   n2 = Branch(execute=@n1.then_6, Condition=Check gameplay tag!(Tag:Gameplay Tag Structure=(TagName="Quest.Metadata.IsGliding")) -> (ReturnValue:bool))
   > n3 = SetSettings!(
    execute
    self:IKRetarget Relative IKController Object Reference=Get Relative IKController() -> (Relative IKController:IKRetarget Relative IKController Object Reference)
    InSettings:Relative IK Settings Structure="Get Relative IK Settings (Without Pins)"() -> ("Relative IK Settings (Without Pins)":Relative IK Settings Structure)
   ) -> (then)
   n4 = GetSettings!(
    execute=[@n3.then, @n6.then]
    self:IKRetarget Stride Warping Controller Object Reference=Get Stride Warping Controller() -> (Stride Warping Controller:IKRetarget Stride Warping Controller Object Reference)
   ) -> (then, ReturnValue:Stride Warp Settings Structure)
   > n5 = Set members in Stride Warp Settings!(
    execute:Exec
    StructRef:"Stride Warp Settings Structure (by ref)"=@n4
    SidewaysOffset:double=Get Current Feet Sideways Offset() -> (Current Feet Sideways Offset:float)
   ) -> (then:Exec, StructOut:"Stride Warp Settings Structure (by ref)") {is=title K2Node_SetFieldsInStruct}
   > SetSettings(self:IKRetarget Stride Warping Controller Object Reference=Get Stride Warping Controller(), InSettings:Stride Warp Settings Structure=@n5)
   n6 = SetSettings(
    execute=@n2.else
    self:IKRetarget Relative IKController Object Reference=Get Relative IKController()
    InSettings:Relative IK Settings Structure="Get Relative IK Settings (With Pins)"() -> ("Relative IK Settings (With Pins)":Relative IK Settings Structure)
   )
   Check Controller(
    execute=@n1.then_1
    Controller=Get IKRetargetStretchChainController() -> (IKRetargetStretchChainController:IKRetarget Stretch Chain Controller Object Reference)
   )
   > n7 = Branch(Condition=Get Stretch Spine() -> (Stretch Spine:bool))
   > SetSettings(
    self:IKRetarget Stretch Chain Controller Object Reference=Get IKRetargetStretchChainController()
    InSettings:Stretch Chain Settings Structure=Get Spine Stretch Settings() -> (Spine Stretch Settings:Stretch Chain Settings Structure)
   )
   SetSettings(
    execute=@n7.else
    self=Get IKRetargetStretchChainController()
    InSettings=Get No Spine Stretch Settings() -> (No Spine Stretch Settings:Stretch Chain Settings Structure)
   )
   n8 = Sequence!(execute=@n1.then_0) -> (then_0, then_1)
   > n9 = Cast To IKRetargetBodyIntersectController!(
    execute
    Object=GetOpControllerFromRetargetProfile!(
     InRetargetProfile:"Retarget Profile Structure (by ref)"=Get CustomRetargetProfile() -> (CustomRetargetProfile:Retarget Profile Structure)
     InRetargetOpName:Name=Body Intersect Goals
    ) -> (ReturnValue:IKRetarget Op Controller Base Object Reference)
   ) -> (then, CastFailed, AsIKRetarget Body Intersect Controller:IKRetarget Body Intersect Controller Object Reference) {is=title K2Node_DynamicCast}
   n10 = Get Disable Intersect() -> (Disable Intersect:bool)
   > n11 = Branch(Condition=@n10)
   n12 = Check gameplay tag(Tag=(TagName="Quest.Metadata.IsGliding"))
   > SetSettings(
    self:IKRetarget Body Intersect Controller Object Reference=@n9
    InSettings:Body Intersect IK Settings Structure=Select(
     Option 0:Body Intersect IK Settings Structure=Get No Hand Intersect Settings() -> (No Hand Intersect Settings:Body Intersect IK Settings Structure)
     Option 1:Body Intersect IK Settings Structure=Get No Hand and Elbow Intersect Settings() -> (No Hand and Elbow Intersect Settings:Body Intersect IK Settings Structure)
     Index=@n12
    ) -> (ReturnValue:Body Intersect IK Settings Structure)
   )
   SetSettings(execute=@n11.else, self=@n9, InSettings=Get Hand Intersect Settings() -> (Hand Intersect Settings:Body Intersect IK Settings Structure))
   n13 = Cast To IKRetargetBodyIntersectController(
    execute=@n8.then_1
    Object=GetOpControllerFromRetargetProfile(InRetargetProfile=Get Intersection Retarget Profile(), InRetargetOpName=Body Intersect Goals)
   )
   > n14 = Branch(Condition=@n10)
   > SetSettings(self=@n13, InSettings=Select(Option 0=Get No Hand Intersect Settings(), Option 1=Get No Hand and Elbow Intersect Settings(), Index=@n12))
   SetSettings(
    execute=@n14.else
    self=@n13
    InSettings=Get Post Hand Intersect Settings() -> (Post Hand Intersect Settings:Body Intersect IK Settings Structure)
   )
   n15 = Cast To IKRetargetFloorConstraintController!(
    execute=@n1.then_5
    Object=GetOpControllerFromRetargetProfile(InRetargetProfile=Get CustomRetargetProfile(), InRetargetOpName=Floor Constraint)
   ) -> (then, CastFailed, AsIKRetarget Floor Constraint Controller:IKRetarget Floor Constraint Controller Object Reference) {is=title K2Node_DynamicCast}
   > n16 = GetSettings(self:IKRetarget Floor Constraint Controller Object Reference=@n15) -> (ReturnValue:Floor Goal Op Settings Structure)
   > n17 = Set Floor Constraint Settings(Floor Constraint Settings:Floor Goal Op Settings Structure=@n16) -> (Output_Get:Floor Goal Op Settings Structure)
   > n18 = Set Floor Constraint Chains(Floor Constraint Chains:Array of Floor Constraint Chain Settings Structures=Break Floor Goal Op Settings!(IKRetargetFloorConstraintOpSettings:"Floor Goal Op Settings Structure (by ref)"=@n17) -> (
     ChainsToAffect:Array of Floor Constraint Chain Settings Structures
     Alpha:double=~
     HeightFalloffOffset:double=~
     HeightFalloffDistance:double=~
     bEnabled:bool=~
     LODThreshold:int
    ) {is=title K2Node_BreakStruct}) -> (Output_Get:Array of Floor Constraint Chain Settings Structures)
   > n19 = For Each Loop(Array:Array of Floor Constraint Chain Settings Structures=@n18) -> (Array Element:Floor Constraint Chain Settings Structure)
   n20 = Set members in Floor Goal Op Settings!(
    execute:Exec=@n19.Completed
    StructRef:"Floor Goal Op Settings Structure (by ref)"=Get Floor Constraint Settings() -> (Floor Constraint Settings:Floor Goal Op Settings Structure)
    ChainsToAffect:Array of Floor Constraint Chain Settings Structures=Get Floor Constraint Chains() -> (Floor Constraint Chains:Array of Floor Constraint Chain Settings Structures)
   ) -> (then:Exec, StructOut:"Floor Goal Op Settings Structure (by ref)") {is=title K2Node_SetFieldsInStruct}
   > SetSettings(self:IKRetarget Floor Constraint Controller Object Reference=@n15, InSettings:Floor Goal Op Settings Structure=@n20)
   n21 = Break Floor Constraint Chain Settings!(FloorConstraintChainSettings:"Floor Constraint Chain Settings Structure (by ref)"=@n19) -> (
    TargetChainName:Name
    EnableFloorConstraint:bool=~
    Alpha:double=~
    MaintainHeightOffset:double=~
    bUseFoot:bool=~
    bUseToes:bool=~
    Foot:Floor Constraint Foot Definition Structure
    Toes:Floor Constraint Toes Definition Structure
   ) {is=title K2Node_BreakStruct}
   Branch(execute=@n19.LoopBody, Condition=OR Boolean(A="Equal (Name)"(A=@n21, B=LeftLeg), B="Equal (Name)"(A=@n21, B=RightLeg)))
   > n22 = Set members in Floor Constraint Chain Settings!(
    execute:Exec
    StructRef:"Floor Constraint Chain Settings Structure (by ref)"=@n19
    EnableFloorConstraint:bool=Get Enable Floor Constraint() -> (Enable Floor Constraint:bool)
   ) -> (then:Exec, StructOut:"Floor Constraint Chain Settings Structure (by ref)") {is=title K2Node_SetFieldsInStruct}
   > Set Array Elem!(
    execute
    TargetArray:Array of Floor Constraint Chain Settings Structures=Get Floor Constraint Chains()
    Index:int=@n19.Array Index
    Item:"Floor Constraint Chain Settings Structure (by ref)"=@n22
    bSizeToFit:bool=false
   ) -> (then)
   Check Controller(
    execute=@n1.then_3
    Controller=Get IKRetarget Blend To Source Controller() -> (IKRetarget Blend To Source Controller:IKRetarget Blend To Source Controller Object Reference)
   )
   > SetIKChainByName!(
    execute
    self
    BlendToSourceController:IKRetarget Blend To Source Controller Object Reference=Get IKRetarget Blend To Source Controller()
    TargetChainName:Name=LeftArm
    Blend To Source:double=Get Left Hand IK Blend To Source() -> (Left Hand IK Blend To Source:double)
   ) -> (then)
   > SetIKChainByName(
    BlendToSourceController=Get IKRetarget Blend To Source Controller()
    TargetChainName=RightArm
    Blend To Source=Get Right Hand IK Blend To Source() -> (Right Hand IK Blend To Source:double)
   )
   > SetIKChainByName(
    BlendToSourceController=Get IKRetarget Blend To Source Controller()
    TargetChainName=LeftLeg
    Blend To Source=Get Foot IK Blend To Source() -> (Foot IK Blend To Source:double)
   )
   > SetIKChainByName(BlendToSourceController=Get IKRetarget Blend To Source Controller(), TargetChainName=RightLeg, Blend To Source=Get Foot IK Blend To Source())
   n23 = Cast To IKRetargetPelvisMotionController!(execute=@n1.then_4, Object=GetOpControllerFromRetargetProfile(InRetargetProfile=Get CustomRetargetProfile(), InRetargetOpName=Pelvis Motion)) -> (then, CastFailed, AsIKRetarget Pelvis Motion Controller:IKRetarget Pelvis Motion Controller Object Reference) {is=title K2Node_DynamicCast}
   > n24 = GetSettings(self:IKRetarget Pelvis Motion Controller Object Reference=@n23) -> (ReturnValue:Pelvis Motion Settings Structure)
   > n25 = Set members in Pelvis Motion Settings!(
    execute:Exec
    StructRef:"Pelvis Motion Settings Structure (by ref)"=@n24
    BlendToSourceTranslation:double=Get Pelvis Blend To Source Translation() -> (Pelvis Blend To Source Translation:double)
    AffectIKVertical:double=Get Affect IK Vertical() -> (Affect IK Vertical:double)
   ) -> (then:Exec, StructOut:"Pelvis Motion Settings Structure (by ref)") {is=title K2Node_SetFieldsInStruct}
   > SetSettings(self:IKRetarget Pelvis Motion Controller Object Reference=@n23, InSettings:Pelvis Motion Settings Structure=@n25)
   Reroute Node()
   n26 = Get IKRetargetPelvisMotionController() -> (IKRetargetPelvisMotionController:IKRetarget Pelvis Motion Controller Object Reference)
   > Check Controller(Controller=@n26)
   > n27 = GetSettings(self=@n26)
   > n28 = Set members in Pelvis Motion Settings!(execute, StructRef=@n27, TranslationOffsetLocal:Vector=Get CurrentPelvisTranslationOffset(), TranslationOffsetGlobal:Vector="0,0,0") -> (then, StructOut)
   > SetSettings(self=Get IKRetargetPelvisMotionController(), InSettings=@n28)
   "Equal (Name)"(A=@n21, B=LeftArm)
   "Equal (Name)"(A=@n21, B=RightArm)
   Get Affect IK Horizontal() -> (Affect IK Horizontal:double)
  ]}
 BlueprintThreadSafeUpdateAnimation {nodes [
   BlueprintThreadSafeUpdateAnimation!() -> (then, DeltaTime:float) {is=title K2Node_FunctionEntry}
   > Set RetargetOverrideBlend(RetargetOverrideBlend:float=GetCurveValueWithDefault!(self:Anim Instance Object Reference, CurveName:Name=RetargetOverrideBlend, DefaultValue:float=0) -> (OutValue:float, ReturnValue:bool)) -> (Output_Get:float)
   > n1 = Sequence()
   n2 = Branch(execute=@n1.then_1, Condition=Get bIgnoreSignificanceManagerAndAlwaysTick() -> (bIgnoreSignificanceManagerAndAlwaysTick:bool))
   Set RetargetOverrideBlend(execute=@n2.else, RetargetOverrideBlend=1)
   n3 = Branch(execute=@n1.then_0, Condition=Get bHasIKRetargeterAsset() -> (bHasIKRetargeterAsset:bool))
   Set RetargetOverrideBlend(execute=@n3.else, RetargetOverrideBlend=1)
  ]}
 Check Controller {nodes [
   n1 = Check Controller!() -> (then, Controller:Object Reference) {is=title K2Node_FunctionEntry}
   > n2 = Branch(Condition=IsValid(Object=@n1))
   > Return Node!(execute, Valid:bool)
   Manual Reinit(execute=@n2.else)
   > Return Node(Valid=false)
  ]}
 Check gameplay tag {nodes [
   n1 = Check gameplay tag!() -> (then, Tag:Gameplay Tag Structure) {is=title K2Node_FunctionEntry}
   n2 = GetFortAbilitySystemComponent!(Actor:Actor Object Reference=TryGetPawnOwner()) -> (ReturnValue:Fort Ability System Component Object Reference)
   > Return Node!(
    execute
    ReturnValue:bool=HasTag!(
     TagContainer:"Gameplay Tag Container Structure (by ref)"=GetOwnedGameplayTags!(TagContainerInterface:Gameplay Tag Asset Interface Interface=@n2) -> (ReturnValue:Gameplay Tag Container Structure)
     Tag:Gameplay Tag Structure=@n1
     bExactMatch:bool=true
    ) -> (ReturnValue:bool)
   )
  ]}
 "Check object child of Classes " {nodes [
   n1 = "Check object child of Classes "!() -> (then, Object:Object Reference, Classes to check:Array of Object Soft Class References) {is=title K2Node_FunctionEntry}
   > n2 = For Each Loop with Break!(Exec:Exec, Array:Array of Object Soft Class References=@n1.Classes to check, Break:Exec=@n4.then_0) -> (LoopBody:Exec, Array Element:Object Soft Class Reference, Array Index:int, Completed:Exec)
   Return Node!(execute=@n2.Completed, Return:bool=false)
   > n3 = LoadClassAsset_Blocking(AssetClass=@n2)
   > Branch(Condition=ClassIsChildOf(TestClass=Get Class(Object=@n1) -> (ReturnValue:Object Class Reference), ParentClass=@n3))
   > n4 = Sequence()
   Return Node(execute=@n4.then_1, Return=true)
  ]}
 EventGraph {nodes [
   n1 = Manual Reinit() {is=title K2Node_CustomEvent}
   n2 = Sequence!(execute=[@n1.then, @n150.then]) -> (then_0, then_1, then_2:Exec)
   > Set bHasIKRetargeterAsset(bHasIKRetargeterAsset:bool=IsValid(Object=Get IKRetargeterAsset())) -> (Output_Get:bool)
   > Branch(Condition=Get bHasIKRetargeterAsset())
   > n3 = Sequence!(execute) -> (then_0, then_1, then_2, then_3:Exec, then_4:Exec, then_5:Exec, then_6:Exec, then_7:Exec, then_8:Exec, then_9:Exec, then_10:Exec)
   Branch(
    execute=@n3.then_3
    Condition=IsValid(Object=Get Relative Body Physics Asset() -> (Relative Body Physics Asset:Physics Asset Object Reference))
   )
   > n4 = Sequence!(execute) -> (then_0, then_1)
   > n5 = Cast To IKRetargetRelativeIKController!(execute, Object=GetOpControllerFromRetargetProfile(InRetargetProfile=Get CustomRetargetProfile(), InRetargetOpName=Relative IK Goals)) -> (then, CastFailed, AsIKRetarget Relative IKController:IKRetarget Relative IKController Object Reference) {is=title K2Node_DynamicCast}
   > n6 = Set Relative IKController(Relative IKController:IKRetarget Relative IKController Object Reference=@n5) -> (Output_Get:IKRetarget Relative IKController Object Reference)
   > n7 = GetSettings(self:IKRetarget Relative IKController Object Reference=@n6) -> (ReturnValue:Relative IK Settings Structure)
   > n8 = Set members in Relative IK Settings!(
    execute:Exec
    StructRef:"Relative IK Settings Structure (by ref)"=@n7
    TargetPhysicsAssetOverride:Physics Asset Object Reference=Get Relative Body Physics Asset()
   ) -> (then:Exec, StructOut:"Relative IK Settings Structure (by ref)") {is=title K2Node_SetFieldsInStruct}
   > n9 = "Set Relative IK Settings (With Pins)"("Relative IK Settings (With Pins)":Relative IK Settings Structure=@n8) -> (Output_Get:Relative IK Settings Structure)
   > SetSettings(self:IKRetarget Relative IKController Object Reference=@n6, InSettings:Relative IK Settings Structure=@n9)
   > n10 = Set members in Relative IK Settings!(
    execute
    StructRef=@n7
    TargetPhysicsAssetOverride=Get Relative Body Physics Asset()
    PinPropBones:Array of Pin Bone Settings Structures=Make Array("[0]":Pin Bone Settings Structure) -> (Array:Array of Pin Bone Settings Structures)
   ) -> (then, StructOut)
   > "Set Relative IK Settings (Without Pins)"("Relative IK Settings (Without Pins)":Relative IK Settings Structure=@n10) -> (Output_Get:Relative IK Settings Structure)
   n11 = Sequence(execute=@n4.then_1)
   > n12 = Cast To IKRetargetBodyIntersectController(Object=GetOpControllerFromRetargetProfile(InRetargetProfile=Get CustomRetargetProfile(), InRetargetOpName=Body Intersect Goals))
   > Set Body Intersect Controller(Body Intersect Controller:IKRetarget Body Intersect Controller Object Reference=@n12) -> (Output_Get:IKRetarget Body Intersect Controller Object Reference)
   > n13 = GetSettings(self:IKRetarget Body Intersect Controller Object Reference=Get Body Intersect Controller() -> (Body Intersect Controller:IKRetarget Body Intersect Controller Object Reference)) -> (ReturnValue:Body Intersect IK Settings Structure)
   > n14 = Sequence!(execute) -> (then_0, then_1, then_2:Exec)
   > n15 = Set Hand Intersect Settings(Hand Intersect Settings:Body Intersect IK Settings Structure=@n13) -> (Output_Get:Body Intersect IK Settings Structure)
   n16 = Get Elbow Scale() -> (Elbow Scale:double)
   n17 = Make Array!(
    "[0]":Pole Vector Settings Structure=Make Pole Vector Settings!(BodyName:Name=lowerarm_l, SphereScale:double=@n16, SphereOffset:Vector="0,0,0") -> (IKPoleVectorIntersectSettings:Pole Vector Settings Structure) {is=title K2Node_MakeStruct}
    "[1]":Pole Vector Settings Structure=Make Pole Vector Settings(BodyName=lowerarm_r, SphereScale=@n16, SphereOffset="0,0,0")
   ) -> (Array:Array of Pole Vector Settings Structures)
   > n18 = Set members in Body Intersect IK Settings!(
    execute:Exec
    StructRef:"Body Intersect IK Settings Structure (by ref)"=@n15
    TargetPhysicsAssetOverride:Physics Asset Object Reference=Get Relative Body Physics Asset()
    bEnableGoalIntersect:bool=True
    IntersectGoalSettings:Array of Goal Intersect Settings Structures=Get Intersect Goal Settings() -> (Intersect Goal Settings:Array of Goal Intersect Settings Structures)
    IntersectBodies:Array of Names=Get Intersect Bodies() -> (Intersect Bodies:Array of Names)
    bEnablePoleVectorIntersect:bool=True
    PoleVectorIntersectSettings:Array of Pole Vector Settings Structures=@n17
   ) -> (then:Exec, StructOut:"Body Intersect IK Settings Structure (by ref)") {is=title K2Node_SetFieldsInStruct}
   > SetSettings(
    self:IKRetarget Body Intersect Controller Object Reference=Get Body Intersect Controller()
    InSettings:Body Intersect IK Settings Structure=@n18
   )
   n19 = Set No Hand Intersect Settings(execute=@n14.then_1, No Hand Intersect Settings:Body Intersect IK Settings Structure=@n13) -> (Output_Get:Body Intersect IK Settings Structure)
   > Set members in Body Intersect IK Settings(
    StructRef=@n19
    TargetPhysicsAssetOverride=Get Relative Body Physics Asset()
    bEnableGoalIntersect=false
    IntersectGoalSettings=Get Intersect Goal Settings()
    IntersectBodies=Get Intersect Bodies()
    bEnablePoleVectorIntersect=True
    PoleVectorIntersectSettings=@n17
   )
   n20 = Set No Hand and Elbow Intersect Settings(execute=@n14.then_2, No Hand and Elbow Intersect Settings:Body Intersect IK Settings Structure=@n13) -> (Output_Get:Body Intersect IK Settings Structure)
   > Set members in Body Intersect IK Settings!(
    execute
    StructRef=@n20
    TargetPhysicsAssetOverride=Get Relative Body Physics Asset()
    bEnableGoalIntersect=false
    IntersectGoalSettings=Get Intersect Goal Settings()
    IntersectBodies=Get Intersect Bodies()
    bEnablePoleVectorIntersect=false
   ) -> (then, StructOut)
   n21 = GetSettings(execute=@n11.then_1, self=Get Body Intersect Controller())
   > n22 = Set members in Body Intersect IK Settings!(
    execute
    StructRef=@n21
    PropIntersectSettings:Array of Prop Intersect Settings Structures=Get Prop Intersect Settings() -> (Prop Intersect Settings:Array of Prop Intersect Settings Structures)
   ) -> (then, StructOut)
   > SetSettings(self=Get Body Intersect Controller(), InSettings=@n22)
   > n23 = Sequence!(execute) -> (then_0, then_1)
   n24 = Break Body Intersect IK Settings!(IKRetargetBodyIntersectIKOpSettings:"Body Intersect IK Settings Structure (by ref)"=@n22) -> (
    TargetPhysicsAssetOverride:Physics Asset Object Reference
    bEnableGoalIntersect:bool=~
    IntersectGoalSettings:Array of Goal Intersect Settings Structures
    PropIntersectSettings:Array of Prop Intersect Settings Structures
    IntersectBodies:Array of Names
    bEnablePoleVectorIntersect:bool=~
    PoleVectorIntersectSettings:Array of Pole Vector Settings Structures
    bEnabled:bool=~
    LODThreshold:int
   ) {is=title K2Node_BreakStruct}
   > Set Regular Prop Intersect Settings(Regular Prop Intersect Settings:Array of Prop Intersect Settings Structures=@n24.PropIntersectSettings) -> (Output_Get:Array of Prop Intersect Settings Structures)
   n25 = Set Pin Bone Prop Intersect Settings(execute=@n23.then_1, Pin Bone Prop Intersect Settings:Array of Prop Intersect Settings Structures=@n24.PropIntersectSettings) -> (Output_Get:Array of Prop Intersect Settings Structures)
   > n26 = For Each Loop(Array:Array of Prop Intersect Settings Structures=@n25) -> (Array Element:Prop Intersect Settings Structure)
   n27 = Break Prop Intersect Settings!(IKPropIntersectSettings:"Prop Intersect Settings Structure (by ref)"=@n26) -> (BoneName:Name, bPinBone:bool=~, CapsuleRadius:float=~, CapsuleLength:float=~, CapsuleRotation:Rotator=~, CapsuleTranslation:Vector=~) {is=title K2Node_BreakStruct}
   > Branch(Condition=Contains Item(
     TargetArray:Array of Strings=Get world space emote anim attach bone() -> (world space emote anim attach bone:Array of Strings)
     ItemToFind:"String (by ref)"="To String (Name)"(InName=@n27)
    ))
   > Set members in Prop Intersect Settings!(execute:Exec, StructRef:"Prop Intersect Settings Structure (by ref)"=@n26, bPinBone:bool=true) -> (then:Exec, StructOut:"Prop Intersect Settings Structure (by ref)") {is=title K2Node_SetFieldsInStruct}
   Branch(execute=@n3.then_5, Condition=IsValid(Object=Get Relative Body Physics Asset()))
   > n28 = CopyRetargetProfileFromRetargetAsset!(
    execute
    InRetargetAsset:IKRetargeter Object Reference=&IKR_Biped_Base_IK_Post
   ) -> (then, ReturnValue:Retarget Profile Structure)
   > n29 = Set Intersection Retarget Profile(Intersection Retarget Profile:Retarget Profile Structure=@n28) -> (Output_Get:Retarget Profile Structure)
   > n30 = Cast To IKRetargetBodyIntersectController(Object=GetOpControllerFromRetargetProfile(InRetargetProfile=@n29, InRetargetOpName=Body Intersect Goals))
   > n31 = GetSettings(self=@n30)
   > n32 = Set members in Body Intersect IK Settings!(
    execute
    StructRef=@n31
    TargetPhysicsAssetOverride:Physics Asset Object Reference=Get Relative Body Physics Asset()
    IntersectGoalSettings:Array of Goal Intersect Settings Structures=Get Intersect Goal Settings()
    IntersectBodies:Array of Names=Get Intersect Bodies()
   ) -> (then, StructOut)
   > n33 = Set Post Hand Intersect Settings(Post Hand Intersect Settings:Body Intersect IK Settings Structure=@n32) -> (Output_Get:Body Intersect IK Settings Structure)
   > SetSettings(self=@n30, InSettings=@n33)
   n34 = Sequence!(execute=@n3.then_6) -> (then_0, then_1, then_2:Exec)
   > n35 = Cast To IKRetargetBlendToSourceController!(execute, Object=GetOpControllerFromRetargetProfile(InRetargetProfile=Get CustomRetargetProfile(), InRetargetOpName=Blend to Source)) -> (then, CastFailed, AsIKRetarget Blend to Source Controller:IKRetarget Blend To Source Controller Object Reference) {is=title K2Node_DynamicCast}
   > Set IKRetarget Blend To Source Controller(IKRetarget Blend To Source Controller:IKRetarget Blend To Source Controller Object Reference=@n35) -> (Output_Get:IKRetarget Blend To Source Controller Object Reference)
   n36 = Cast To IKRetargetStretchChainController!(
    execute=@n34.then_1
    Object=GetOpControllerFromRetargetProfile(InRetargetProfile=Get CustomRetargetProfile(), InRetargetOpName=Stretch Chains)
   ) -> (then, CastFailed, AsIKRetarget Stretch Chain Controller:IKRetarget Stretch Chain Controller Object Reference) {is=title K2Node_DynamicCast}
   > Set IKRetargetStretchChainController(IKRetargetStretchChainController:IKRetarget Stretch Chain Controller Object Reference=@n36) -> (Output_Get:IKRetarget Stretch Chain Controller Object Reference)
   > n37 = Sequence!(execute) -> (then_0, then_1)
   > Set Spine Stretch Settings(Spine Stretch Settings:Stretch Chain Settings Structure=Make Stretch Chain Settings!(
     IKRigAsset:IKRig Definition Object Reference=&/IK_Biped_Base
     ChainsToStretch:Array of Retarget Stretch Chain Settings Structures=Get spine stretch chains() -> (spine stretch chains:Array of Retarget Stretch Chain Settings Structures)
     bEnabled:bool=True
     LODThreshold:int=-1
    ) -> (IKRetargetStretchChainOpSettings:Stretch Chain Settings Structure) {is=title K2Node_MakeStruct}) -> (Output_Get:Stretch Chain Settings Structure)
   Set No Spine Stretch Settings(
    execute=@n37.then_1
    No Spine Stretch Settings:Stretch Chain Settings Structure=Make Stretch Chain Settings(
     IKRigAsset=&IK_Biped_Base
     ChainsToStretch=Get no stretch chains() -> (no stretch chains:Array of Retarget Stretch Chain Settings Structures)
     bEnabled=True
     LODThreshold=-1
    )
   ) -> (Output_Get:Stretch Chain Settings Structure)
   n38 = Cast To IKRetargetRelativePelvicMotionController!(
    execute=@n3.then_8
    Object=GetOpControllerFromRetargetProfile(InRetargetProfile=Get CustomRetargetProfile(), InRetargetOpName=Relative Pelvic Motion)
   ) -> (then, CastFailed, AsIKRetarget Relative Pelvic Motion Controller:IKRetarget Relative Pelvic Motion Controller Object Reference) {is=title K2Node_DynamicCast}
   > n39 = Set Pelvic Motion Controller(Pelvic Motion Controller:IKRetarget Relative Pelvic Motion Controller Object Reference=@n38) -> (Output_Get:IKRetarget Relative Pelvic Motion Controller Object Reference)
   > n40 = GetSettings(self:IKRetarget Relative Pelvic Motion Controller Object Reference=@n39) -> (ReturnValue:Relative Pelvic Motion Settings Structure)
   > n41 = Set members in Relative Pelvic Motion Settings!(
    execute:Exec
    StructRef:"Relative Pelvic Motion Settings Structure (by ref)"=@n40
    TargetPhysicsAssetOverride:Physics Asset Object Reference=Get Relative Body Physics Asset()
   ) -> (then:Exec, StructOut:"Relative Pelvic Motion Settings Structure (by ref)") {is=title K2Node_SetFieldsInStruct}
   > SetSettings(
    self:IKRetarget Relative Pelvic Motion Controller Object Reference=Get Pelvic Motion Controller() -> (Pelvic Motion Controller:IKRetarget Relative Pelvic Motion Controller Object Reference)
    InSettings:Relative Pelvic Motion Settings Structure=@n41
   )
   n42 = Cast To IKRetargetPinBoneController!(execute=@n3.then_9, Object=GetOpControllerFromRetargetProfile(InRetargetProfile=Get CustomRetargetProfile(), InRetargetOpName=Root Offset)) -> (then, CastFailed, AsIKRetarget Pin Bone Controller:IKRetarget Pin Bone Controller Object Reference) {is=title K2Node_DynamicCast}
   > Set Root Offset Pin Bone Controller(Root Offset Pin Bone Controller:IKRetarget Pin Bone Controller Object Reference=@n42) -> (Output_Get:IKRetarget Pin Bone Controller Object Reference)
   > n43 = Cast To IKRetargetStrideWarpingController!(execute, Object=GetOpControllerFromRetargetProfile(InRetargetProfile=Get CustomRetargetProfile(), InRetargetOpName=Stride Warp IK Goals)) -> (then, CastFailed, AsIKRetarget Stride Warping Controller:IKRetarget Stride Warping Controller Object Reference) {is=title K2Node_DynamicCast}
   > Set Stride Warping Controller(Stride Warping Controller:IKRetarget Stride Warping Controller Object Reference=@n43) -> (Output_Get:IKRetarget Stride Warping Controller Object Reference)
   > n44 = Cast To IKRetargetRootMotionController!(execute, Object=GetOpControllerFromRetargetProfile(InRetargetProfile=Get CustomRetargetProfile(), InRetargetOpName=Stride Warp IK Goals)) -> (then, CastFailed, AsIKRetarget Root Motion Controller:IKRetarget Root Motion Controller Object Reference) {is=title K2Node_DynamicCast}
   > Set Root Motion Controller(Root Motion Controller:IKRetarget Root Motion Controller Object Reference=@n44) -> (Output_Get:IKRetarget Root Motion Controller Object Reference)
   n45 = Cast To IKRetargetFloorConstraintController(
    execute=@n3.then_7
    Object=GetOpControllerFromRetargetProfile(InRetargetProfile=Get CustomRetargetProfile(), InRetargetOpName=Floor Constraint)
   )
   > n46 = GetSettings(self:IKRetarget Floor Constraint Controller Object Reference=@n45) -> (ReturnValue:Floor Goal Op Settings Structure)
   n47 = Break Floor Goal Op Settings(IKRetargetFloorConstraintOpSettings=@n46) -> (Alpha=~, HeightFalloffOffset=~, HeightFalloffDistance=~, bEnabled=~)
   n48 = Make Floor Constraint Chain Settings!(
    TargetChainName:Name=RightHandIK
    EnableFloorConstraint:bool=true
    Alpha:double=1
    MaintainHeightOffset:double=0
    bUseFoot:bool=False
    bUseToes:bool=False
    Foot:Floor Constraint Foot Definition Structure=(MedialOffset=10,LateralOffset=10,HeelOffset=10,ToeOffset=10,VerticalOffset=0)
    Toes:Floor Constraint Toes Definition Structure=(Alpha=1,Responsiveness=0.02,CutoffFrequency=8,VelocityCutoffFrequency=1,AllToes=)
   ) -> (FloorConstraintChainSettings:Floor Constraint Chain Settings Structure) {is=title K2Node_MakeStruct}
   n49 = Make Floor Constraint Chain Settings(
    TargetChainName=LeftHandIK
    EnableFloorConstraint=true
    Alpha=1
    MaintainHeightOffset=0
    bUseFoot=False
    bUseToes=False
    Foot=(MedialOffset=10,LateralOffset=10,HeelOffset=10,ToeOffset=10,VerticalOffset=0)
    Toes=(Alpha=1,Responsiveness=0.02,CutoffFrequency=8,VelocityCutoffFrequency=1,AllToes=)
   )
   n50 = Get Left Foot IK Floor Constraint() -> (Left Foot IK Floor Constraint:Floor Constraint Chain Settings Structure)
   n51 = Get Right Foot IK Floor Constraint() -> (Right Foot IK Floor Constraint:Floor Constraint Chain Settings Structure)
   > SetSettings(
    self:IKRetarget Floor Constraint Controller Object Reference=@n45
    InSettings:Floor Goal Op Settings Structure=Make Floor Goal Op Settings!(
     ChainsToAffect:Array of Floor Constraint Chain Settings Structures=Make Array!(
      "[0]":Floor Constraint Chain Settings Structure=@n50
      "[1]":Floor Constraint Chain Settings Structure=@n51
      "[2]":Floor Constraint Chain Settings Structure=@n49
      "[3]":Floor Constraint Chain Settings Structure=@n48
     ) -> (Array:Array of Floor Constraint Chain Settings Structures)
     Alpha:double=1
     HeightFalloffOffset:double=@n47.HeightFalloffOffset
     HeightFalloffDistance:double=@n47.HeightFalloffDistance
     bEnabled:bool=True
     LODThreshold:int=@n47.LODThreshold
    ) -> (IKRetargetFloorConstraintOpSettings:Floor Goal Op Settings Structure) {is=title K2Node_MakeStruct}
   )
   n52 = Cast To IKRetargetFilterBoneController!(execute=@n3.then_4, Object=GetOpControllerFromRetargetProfile(InRetargetProfile=Get CustomRetargetProfile(), InRetargetOpName=Filter Bones)) -> (then, CastFailed, AsIKRetarget Filter Bone Controller:IKRetarget Filter Bone Controller Object Reference) {is=title K2Node_DynamicCast}
   > n53 = GetSettings(self:IKRetarget Filter Bone Controller Object Reference=@n52) -> (ReturnValue:Filter Bone Settings Structure)
   > n54 = Set members in Filter Bone Settings!(
    execute:Exec
    StructRef:"Filter Bone Settings Structure (by ref)"=@n53
    BonesToFilter:Array of Filter Bone Data Structures=Get Bones To Filter() -> (Bones To Filter:Array of Filter Bone Data Structures)
   ) -> (then:Exec, StructOut:"Filter Bone Settings Structure (by ref)") {is=title K2Node_SetFieldsInStruct}
   > SetSettings(self:IKRetarget Filter Bone Controller Object Reference=@n52, InSettings:Filter Bone Settings Structure=@n54)
   n55 = CopyRetargetProfileFromRetargetAsset(execute=@n3.then_0, InRetargetAsset=Get IKRetargeterAsset())
   > Set CustomRetargetProfile(CustomRetargetProfile:Retarget Profile Structure=@n55) -> (Output_Get:Retarget Profile Structure)
   n56 = Cast To IKRetargetPelvisMotionController(execute=@n3.then_2, Object=GetOpControllerFromRetargetProfile(InRetargetProfile=Get CustomRetargetProfile(), InRetargetOpName=Pelvis Motion))
   > n57 = Set IKRetargetPelvisMotionController(IKRetargetPelvisMotionController:IKRetarget Pelvis Motion Controller Object Reference=@n56) -> (Output_Get:IKRetarget Pelvis Motion Controller Object Reference)
   > n58 = GetSettings(self:IKRetarget Pelvis Motion Controller Object Reference=@n57) -> (ReturnValue:Pelvis Motion Settings Structure)
   > n59 = Set members in Pelvis Motion Settings!(execute, StructRef=@n58, TargetCrotchOffset:double=Get Crotch Offset() -> (Crotch Offset:double)) -> (then, StructOut)
   > SetSettings(self:IKRetarget Pelvis Motion Controller Object Reference=@n56, InSettings:Pelvis Motion Settings Structure=@n59)
   n60 = Get Target Pose Name() -> (Target Pose Name:Name=None)
   n61 = Set members in Retarget Profile!(
    execute:Exec=@n3.then_1
    StructRef:"Retarget Profile Structure (by ref)"=Get CustomRetargetProfile()
    bApplyTargetRetargetPose:bool="Not Equal (Name)"(A=@n60, B=Default Pose) -> (ReturnValue=~)
    TargetRetargetPoseName:Name=@n60
   ) -> (then:Exec, StructOut:"Retarget Profile Structure (by ref)") {is=title K2Node_SetFieldsInStruct}
   > Set CustomRetargetProfile(CustomRetargetProfile=@n61)
   Set Default Retarget Profile(execute=@n2.then_1, Default Retarget Profile:Retarget Profile Structure=Get CustomRetargetProfile()) -> (Output_Get:Retarget Profile Structure)
   n62 = Cast To FortPawn!(execute=@n2.then_2, Object=GetOwningActor()) -> (then, CastFailed, AsFort Pawn:Fort Pawn Object Reference) {is=title K2Node_DynamicCast}
   > Bind Event to On Pawn Started Emote(self:Fort Pawn Object Reference=@n62, Delegate=Create Event()) {is=title K2Node_AddDelegate}
   > Bind Event to On Pawn Stopped Emote(self:Fort Pawn Object Reference=@n62, Delegate=Create Event()) {is=title K2Node_AddDelegate}
   n63 = Emote Started!() -> (OutputDelegate, then, MontageItemDef:Fort Item Definition Object Reference, PawnEmoting:Fort Pawn Object Reference) {is=title K2Node_CustomEvent}
   > Set Current Emote Def(Current Emote Def:Fort Item Definition Object Reference=@n63.MontageItemDef) -> (Output_Get:Fort Item Definition Object Reference)
   > Set Emoting Pawn(Emoting Pawn:Fort Pawn Object Reference=@n63.PawnEmoting) -> (Output_Get:Fort Pawn Object Reference)
   n64 = Get Current Emote Def() -> (Current Emote Def:Fort Item Definition Object Reference)
   > Set Current Playing Emote Name(Current Playing Emote Name:String=Split(SourceString=GetDisplayName(Object=@n64), InStr=EID_, SearchCase=CaseSensitive, SearchDir=FromStart).RightS) -> (Output_Get:String)
   > n65 = DelayUntilNextTick!(execute) -> (then)
   n66 = Find Emote Prop Attachment() {is=title K2Node_CustomEvent}
   n67 = Set Timer by Event(execute=@n65.then, Delegate=@n66, Time=0.5, bLooping=true, bMaxOncePerFrame=true, InitialStartDelay=0.01)
   > Set Find Emote Attachment Timer(Find Emote Attachment Timer:Timer Handle Structure=@n67) -> (Output_Get:Timer Handle Structure)
   Emote Stopped!() -> (OutputDelegate, then, MontageItemDef:Fort Item Definition Object Reference, PawnEmoting:Fort Pawn Object Reference) {is=title K2Node_CustomEvent}
   > Set Emote playing in world space(Emote playing in world space:bool=false) -> (Output_Get:bool)
   > Clear and Invalidate Timer by Handle(Handle=Get Find Emote Attachment Timer() -> (Find Emote Attachment Timer:Timer Handle Structure))
   > Set Current Emote Def()
   > Set Is playing emote(Is playing emote:bool=false) -> (Output_Get:bool)
   n68 = Cast To PlayerPawn_Athena(Object=TryGetPawnOwner())
   > n69 = For Each Loop(Array:Array of Names=Make Array("[0]":Name=weapon_l, "[1]":Name=weapon_r, "[2]":Name=hand_attach_l, "[3]":Name=hand_attach_r) -> (Array:Array of Names)) -> (Array Element:Name)
   PrintString!(
    execute=@n69.Completed
    InString="To String (Boolean)"(InBool=Property Access!() -> (Value:bool=~) {is=title K2Node_PropertyAccess})
    bPrintToScreen
    bPrintToLog
    TextColor
    Duration
    Key
   ) -> (then)
   n70 = GetSocketTransform!(
    self=Get RootComponent!(self:Actor Object Reference=@n68) -> (RootComponent:Scene Component Object Reference)
    InSocketName=@n69
    TransformSpace=RTS_World
   ) -> (ReturnValue_Location:Vector, ReturnValue_Rotation:Rotator, ReturnValue_Scale:Vector)
   PrintString(
    execute=@n69.LoopBody
    InString=Append(A="To String (Name)"(InName=@n69), B="To String (Vector)"!(InVec:Vector=@n70.ReturnValue_Scale) -> (ReturnValue:String))
   )
   n71 = Event BlueprintUpdateAnimation()
   > Set Delta Time X(Delta Time X=@n71.DeltaTimeX)
   > Branch(Condition=Get bHasIKRetargeterAsset())
   > n72 = Branch(Condition="float >= float"(A=GetCurveValue(CurveName=relativeIK), B=1))
   > Do Once(Reset=@n73.then_0)
   > Set Current Retarget Profile(Current Retarget Profile:Retarget Profile Structure=Get Default Retarget Profile() -> (Default Retarget Profile:Retarget Profile Structure)) -> (Output_Get:Retarget Profile Structure)
   n73 = Sequence(execute=@n72.else)
   Set Current Retarget Profile(execute=@n73.then_1, Current Retarget Profile=Get CustomRetargetProfile())
   > n74 = Sequence()
   > n75 = Sequence!(execute) -> (then_0, then_1, then_2:Exec, then_3:Exec)
   Sequence!(execute=@n75.then_3) -> (then_0, then_1)
   > Check Controller(Controller=Get Body Intersect Controller()) {is=title K2Node_CallFunction}
   > n76 = Branch(Condition=Get Emote playing in world space() -> (Emote playing in world space:bool))
   n77 = Break Body Intersect IK Settings(IKRetargetBodyIntersectIKOpSettings=Get Hand Intersect Settings()) -> (bEnableGoalIntersect=~, bEnablePoleVectorIntersect=~, bEnabled=~)
   > n78 = For Each Loop(Array:Array of Goal Intersect Settings Structures=@n77.IntersectGoalSettings) -> (Array Element:Goal Intersect Settings Structure)
   n79 = Set members in Body Intersect IK Settings!(
    execute=@n78.Completed
    StructRef=Get Hand Intersect Settings()
    PropIntersectSettings:Array of Prop Intersect Settings Structures=Get Pin Bone Prop Intersect Settings() -> (Pin Bone Prop Intersect Settings:Array of Prop Intersect Settings Structures)
   ) -> (then, StructOut)
   > SetSettings(
    self:IKRetarget Body Intersect Controller Object Reference=Get Body Intersect Controller()
    InSettings:Body Intersect IK Settings Structure=@n79
   )
   n80 = Set members in Goal Intersect Settings!(
    execute:Exec=@n78.LoopBody
    StructRef:"Goal Intersect Settings Structure (by ref)"=@n78
    PropIntersectBone:Name=Get Current Emote Attach Socket name() -> (Current Emote Attach Socket name:Name=None)
    BlendMultiMaxOffset:double="To Float (Boolean)"(InBool=Get Emote playing in world space())
   ) -> (then:Exec, StructOut:"Goal Intersect Settings Structure (by ref)") {is=title K2Node_SetFieldsInStruct}
   > Set Array Elem(
    TargetArray:Array of Goal Intersect Settings Structures=@n77.IntersectGoalSettings
    Index=@n78.Array Index
    Item:"Goal Intersect Settings Structure (by ref)"=@n80
    bSizeToFit=false
   )
   n81 = Set members in Body Intersect IK Settings!(
    execute=@n76.else
    StructRef=Get Hand Intersect Settings()
    IntersectGoalSettings:Array of Goal Intersect Settings Structures=Get Intersect Goal Settings()
    PropIntersectSettings=Get Regular Prop Intersect Settings() -> (Regular Prop Intersect Settings:Array of Prop Intersect Settings Structures)
   ) -> (then, StructOut)
   > SetSettings(self=Get Body Intersect Controller(), InSettings=@n81)
   n82 = Property Access() -> (Value:EGliderType Enum)
   n83 = Make Soft Reference!(Input:Skeletal Mesh Object Reference=Property Access() -> (Value:Skeletal Mesh Object Reference)) -> (Output:Skeletal Mesh Soft Object Reference)
   n84 = Get Wind Surf Glider() -> (Wind Surf Glider:Skeletal Mesh Soft Object Reference)
   n85 = AND Boolean(
    A=Get bParachuteOpenedAndStandingUpright()
    B=OR Boolean!(
     A="Equal (Enum)"(A:EGliderType Enum=@n82, B:EGliderType Enum=HangGlider)
     B="Equal (Enum)"(A=@n82, B=Umbrella)
     C:bool="Equal (Enum)"(A=@n82, B=Cape)
     D:bool="Equal (SoftObjectReference)"(A:"Soft Object Reference (by ref)"=@n84, B:"Soft Object Reference (by ref)"=@n83) -> (ReturnValue:bool=~) {is=title K2Node_PromotableOperator}
    ) -> (ReturnValue)
   )
   n86 = AND Boolean(
    A=Check gameplay tag(Tag=(TagName="MovementMode.ZipliningOnSpline")) {is=title K2Node_CallFunction}
    B=OR Boolean!(
     A=Check gameplay tag(Tag=(TagName="Gameplay.Action.Player.Ascender.Ascending"))
     B=Check gameplay tag(Tag=(TagName="Gameplay.Action.Player.Ascender.Descending"))
    ) -> (ReturnValue)
   )
   n87 = AND Boolean!(
    A=Property Access() -> (Value:bool=~)
    B=Get bIsInVehicle() -> (bIsInVehicle:bool)
    C:bool=NOT Boolean(A=Get bIsJackalVehicleDriver() -> (bIsJackalVehicleDriver:bool))
   ) -> (ReturnValue)
   n88 = Get Emote playing in world space()
   n89 = Get bIsOctopusVehicleDriver() -> (bIsOctopusVehicleDriver:bool)
   n90 = Get bIsMeatballDriver() -> (bIsMeatballDriver:bool)
   n91 = Get bIsFerretVehicleDriver() -> (bIsFerretVehicleDriver:bool)
   n92 = FInterpTo(
    Current=Get Left Hand IK Blend To Source()
    Target="To Float (Boolean)"(InBool=OR Boolean!(A=@n85, B=@n87, C:bool=@n91, D:bool=@n90, E:bool=@n89, F:bool=@n86, G:bool=@n88) -> (ReturnValue))
    DeltaTime=Get Delta Time X()
    InterpSpeed=5
   )
   Set Left Hand IK Blend To Source(execute=@n75.then_1, Left Hand IK Blend To Source:double=@n92) -> (Output_Get:double)
   n93 = AND Boolean!(
    A=Check gameplay tag(Tag=(TagName="MovementMode.ZipliningOnSpline"))
    B=OR Boolean!(
     A=Check gameplay tag(Tag=(TagName="Gameplay.Action.Player.Ascender.Ascending"))
     B=Check gameplay tag(Tag=(TagName="Gameplay.Action.Player.Ascender.Descending"))
    ) -> (ReturnValue)
   ) -> (ReturnValue)
   > Set Right Hand IK Blend To Source(Right Hand IK Blend To Source:double=float * float(A=@n92, B:bool=NOT Boolean(A=@n93)) {is=title K2Node_PromotableOperator}) -> (Output_Get:double)
   n94 = NOT Boolean(A=Get bVehicleSeatForcesCrouch() -> (bVehicleSeatForcesCrouch:bool))
   n95 = Get bIsInShoppingCart() -> (bIsInShoppingCart:bool)
   n96 = AND Boolean(A=Get bIsInVehicle(), B=NOT Boolean(A=AND Boolean(A=@n95, B=@n94)))
   n97 = Property Access() -> (Value:EGliderType Enum)
   n98 = AND Boolean(A=Get bParachuteOpenedAndStandingUpright(), B=OR Boolean(A="Equal (Enum)"(A=@n97, B=SurfingSimple), B="Equal (Enum)"(A=@n97, B=Surfing)))
   n99 = FInterpTo(
    Current=Get Foot IK Blend To Source()
    Target="To Float (Boolean)"(InBool=OR Boolean(A=@n96, B=@n98))
    DeltaTime=Get Delta Time X()
    InterpSpeed=5
   )
   > Set Foot IK Blend To Source(Foot IK Blend To Source:double=@n99) -> (Output_Get:double)
   > Set Current Feet Sideways Offset(Current Feet Sideways Offset:float=Ease!(
     Function:EEasingFunc Enum=Linear
     Alpha:double=float * float(A=@n99, B=@n98)
     A:float
     B:float=Get Feet Sideways Offset Surfing Glider() -> (Feet Sideways Offset Surfing Glider:float)
    ) -> (Result:float=~) {is=title K2Node_EaseFunction}) -> (Output_Get:float)
   > Set Stretch Spine(Stretch Spine:bool=OR Boolean(
     A=Check gameplay tag(Tag=(TagName="Gameplay.Status.InVehicle.Seat.Roofed"))
     B=Get bIsGolfCartVehicleFrontPassenger() -> (bIsGolfCartVehicleFrontPassenger:bool)
    )) -> (Output_Get:bool)
   n100 = OR Boolean!(
    A=Get bIsSurfaceSwimming()
    B=Get bIsSkyDiving()
    C:bool=Get bParachuteOpenedAndStandingUpright()
    D:bool=Check gameplay tag(Tag=Get DBNO Ragdoll Tag() -> (DBNO Ragdoll Tag:Gameplay Tag Structure))
   ) -> (ReturnValue)
   > Set Pelvis Blend To Source Translation(Pelvis Blend To Source Translation:double=FInterpTo(Current=Get Pelvis Blend To Source Translation(), Target="To Float (Boolean)"(InBool=@n100), DeltaTime=Get Delta Time X(), InterpSpeed=5)) -> (Output_Get:double)
   n101 = Get bIsGolfCartVehicleBackPassenger() -> (bIsGolfCartVehicleBackPassenger:bool)
   n102 = Get bIsGolfCartVehicleFrontPassenger()
   n103 = Get bIsAntelopeVehiclePassenger() -> (bIsAntelopeVehiclePassenger:bool)
   n104 = Get bIsFerretVehiclePassenger() -> (bIsFerretVehiclePassenger:bool)
   n105 = Get bIsMeatballPassenger() -> (bIsMeatballPassenger:bool)
   n106 = Get bIsValetPassenger() -> (bIsValetPassenger:bool)
   n107 = Get bIsMotorcyclePassenger() -> (bIsMotorcyclePassenger:bool)
   n108 = "float <= float"(A=Get Right Hand IK Blend To Source()) -> (ReturnValue=~)
   n109 = "float <= float"(A=Get Left Hand IK Blend To Source()) -> (ReturnValue=~)
   n110 = OR Boolean!(
    A=@n100
    B=AND Boolean(
     A=OR Boolean!(A=@n101, B=@n102, C:bool=@n103, D:bool=@n104, E:bool=@n105, F:bool=@n106, G:bool=@n107) -> (ReturnValue)
     B=AND Boolean(A=@n109, B=@n108)
    )
   ) -> (ReturnValue)
   > Set Affect IK Vertical(Affect IK Vertical:double=FInterpTo(Current=Get Affect IK Vertical(), Target="To Float (Boolean)"(InBool=@n110), DeltaTime=Get Delta Time X(), InterpSpeed=5)) -> (Output_Get:double)
   n111 = NOT Boolean(A=@n110)
   > Set Affect IK Horizontal(Affect IK Horizontal:double=FInterpTo(Current=Get Affect IK Horizontal(), Target="To Float (Boolean)"(InBool=@n111), DeltaTime=Get Delta Time X(), InterpSpeed=5)) -> (Output_Get:double)
   n112 = Get bIsJackalVehicleDriver()
   > Set Enable Floor Constraint(Enable Floor Constraint:bool=AND Boolean(A=@n111, B=NOT Boolean(A=@n112))) -> (Output_Get:bool)
   n113 = NOT Boolean(A=Get bIsJackalVehicleDriver())
   n114 = Property Access() -> (Value:bool=~)
   n115 = Get bIsInVehicle()
   n116 = Property Access() -> (Value:EGliderType Enum)
   n117 = Property Access() -> (Value:Skeletal Mesh Object Reference)
   n118 = OR Boolean!(
    A="Equal (Enum)"(A=@n116, B=HangGlider)
    B="Equal (Enum)"(A=@n116, B=Umbrella)
    C="Equal (SoftObjectReference)"(A=Get Wind Surf Glider(), B=Make Soft Reference(Input=@n117)) -> (ReturnValue=~)
   ) -> (ReturnValue)
   n119 = Get bParachuteOpenedAndStandingUpright()
   > Set Disable Intersect(Disable Intersect:bool=OR Boolean!(A=AND Boolean!(A=@n114, B=@n115, C:bool=@n113) -> (ReturnValue), B=AND Boolean!(A=@n119, B=@n118) -> (ReturnValue)) -> (ReturnValue)) -> (Output_Get:bool)
   > n120 = Branch(Condition=Check gameplay tag(Tag=(TagName="Quest.Metadata.IsGliding")))
   n121 = Get Pelvis Offset Surfing Glider() -> (Pelvis Offset Surfing Glider:Vector)
   > Set CurrentPelvisTranslationOffset(CurrentPelvisTranslationOffset:Vector=Select!(
     HangGlider:Vector=Get Pelvis Offset Hang Glider() -> (Pelvis Offset Hang Glider:Vector)
     Umbrella:Vector=Get Pelvis Offset Umbrella Glider() -> (Pelvis Offset Umbrella Glider:Vector)
     Surfing:Vector=@n121
     SurfingSimple:Vector=@n121
     Cape:Vector=Get Pelvis Offset Cape Glider() -> (Pelvis Offset Cape Glider:Vector)
     Stairs:Vector=Get Pelvis Offset Stairs Glider() -> (Pelvis Offset Stairs Glider:Vector)
     NoGlider:Vector=Get Pelvis Offset No Glider Glider() -> (Pelvis Offset No Glider Glider:Vector)
     Index:EGliderType Enum=Property Access() -> (Value:EGliderType Enum)
    ) -> (ReturnValue:Vector)) -> (Output_Get:Vector)
   n122 = OR Boolean!(A=Get bIsJackalVehicleDriver(), B=Get bIsInShoppingCart(), C:bool=Get bIsGolfCartVehicleBackPassenger()) -> (ReturnValue)
   n123 = Branch(execute=@n120.else, Condition=AND Boolean(A=Get bIsInVehicle(), B=NOT Boolean(A=@n122)))
   n124 = Get bIsAntelopeVehicleDriver() -> (bIsAntelopeVehicleDriver:bool)
   n125 = Get bIsMotorcycleDriver() -> (bIsMotorcycleDriver:bool)
   > Set CurrentPelvisTranslationOffset(CurrentPelvisTranslationOffset=SelectVector(
     A=Get In Motorcycle Pelvis Translation Offset() -> (In Motorcycle Pelvis Translation Offset:Vector)
     B=Get In Vehicle Pelvis Translation Offset() -> (In Vehicle Pelvis Translation Offset:Vector)
     bPickA=OR Boolean!(A=@n125, B=@n124) -> (ReturnValue)
    ))
   n126 = Check gameplay tag(Tag=(TagName="Gameplay.Action.Player.Ascender.Descending"))
   n127 = Check gameplay tag(Tag=(TagName="Gameplay.Action.Player.Ascender.Ascending"))
   n128 = Branch(
    execute=@n123.else
    Condition=AND Boolean(A=Check gameplay tag(Tag=(TagName="MovementMode.ZipliningOnSpline")), B=OR Boolean(A=@n127, B=@n126))
   )
   > Set CurrentPelvisTranslationOffset(CurrentPelvisTranslationOffset=Get Pelvis Offset Vertical ZipLining() -> (Pelvis Offset Vertical ZipLining:Vector))
   Set CurrentPelvisTranslationOffset(execute=@n128.else, CurrentPelvisTranslationOffset="0, 0, 0")
   n129 = Get Hand IK Retargeting Weight Bias() -> (Hand IK Retargeting Weight Bias:double)
   n130 = float - float(A=1, B=float * float(A=@n129, B:double=2))
   n131 = "float > float"(A=@n129, B=0.5)
   Set Hand IK Retargeting Weight(
    execute=@n75.then_0
    Hand IK Retargeting Weight:double=Lerp(A=Property Access() -> (Value:float=~), B=SelectFloat(A=1, B=0, bPickA=@n131), Alpha="Absolute (Float)"(A=@n130))
   ) -> (Output_Get:double)
   n132 = Sequence!(execute=@n75.then_2) -> (then_0, then_1, then_2:Exec, then_3:Exec)
   n133 = AND Boolean!(A=Property Access() -> (Value:bool=~), B=Property Access() -> (Value=~), C:bool=NOT Boolean(A=Get bParachuteOpenedAndStandingUpright())) -> (ReturnValue)
   n134 = Property Access() -> (Value:float=~)
   n135 = Get Weapons Should not use hand IK() -> (Weapons Should not use hand IK:bool)
   n136 = Get Left Hand IK Blend To Source()
   n137 = TryGetFortPawn()
   n138 = Contains(
    SearchIn=GetDisplayName(Object=GetCurrentWeapon!(self:Fort Pawn Object Reference=@n137) -> (ReturnValue:Fort Weapon Object Reference))
    Substring=BallisticShield
    bUseCase=true
   )
   n139 = IsInBuildMode!(self:Fort Player Controller Object Reference=GetFortPlayerControllerFromActor!(Actor:Actor Object Reference=@n137) -> (ReturnValue:Fort Player Controller Object Reference)) -> (ReturnValue:bool)
   n140 = Check gameplay tag(Tag=(TagName="Item.Weapon.Ranged.Bow"))
   n141 = float - float(A=1, B="To Float (Boolean)"(InBool=OR Boolean!(A=@n139, B=@n138, C:bool=@n140) -> (ReturnValue)))
   n142 = float - float(A=1, B=GetCurveValue(CurveName=LeftHankFK))
   n143 = float - float(A=1, B=GetCurveValue(CurveName=BothHandFK))
   n144 = float - float(A=1, B=GetCurveValue(CurveName=LeftHandFK))
   Set Left Hand IK Alpha(
    execute=@n132.then_1
    Left Hand IK Alpha:double=float * float!(
     A=float * float!(A=@n134, B:bool=@n133, C:bool=@n135) -> (ReturnValue)
     B=@n141
     C:double=float * float(A=@n142, B:double=@n144, C:double=@n143)
     D:double=float - float(A=1, B=@n136)
    ) -> (ReturnValue)
   ) -> (Output_Get:double)
   n145 = float - float(A=1, B=GetCurveValue(CurveName=BothHandFK))
   n146 = float - float(A=1, B=GetCurveValue(CurveName=RightHandFK))
   n147 = Property Access() -> (Value=~)
   n148 = Get Right Hand IK Blend To Source()
   Set Right Hand IK Alpha(
    execute=@n132.then_2
    Right Hand IK Alpha:double=float * float!(
     A=float * float!(A=@n147, B:bool=@n133, C:bool=@n135) -> (ReturnValue)
     B=@n141
     C=float * float!(A=@n145, B:double=@n146) -> (ReturnValue)
     D:double=float - float(A=1, B=@n148)
    ) -> (ReturnValue)
   ) -> (Output_Get:double)
   n149 = "Check object child of Classes "!(
    execute=@n132.then_0
    self
    Object:Object Reference=Get CurrentWeapon() -> (CurrentWeapon:Fort Weapon Object Reference)
    Classes to check:Array of Object Soft Class References=Get Weapon classes to not use IK() -> (Weapon classes to not use IK:Array of Object Soft Class References)
   ) -> (then, Return:bool) {is=title K2Node_CallFunction}
   > Set Weapons Should not use hand IK(Weapons Should not use hand IK:bool=NOT Boolean(A=@n149)) -> (Output_Get:bool)
   Apply Ops Settings(execute=@n74.then_1) {is=title K2Node_CallFunction}
   n150 = Event BlueprintInitializeAnimation()
   n151 = float - float(A=Get Crotch Offset(), B=10)
   vector + vector(B=MakeVector(X="Absolute (Float)"(A=@n151), Y=0))
   Get Feasibility Length Bias() -> (Feasibility Length Bias:double)
   Get Dist Half Point Lambda() -> (Dist Half Point Lambda:double)
   Get Distance Threshold() -> (Distance Threshold:double)
   n152 = GetEmotePropAttachMesh!(execute=@n66.then, self:Fort Pawn Object Reference=Get Emoting Pawn() -> (Emoting Pawn:Fort Pawn Object Reference)) -> (then, ReturnValue:Skeletal Mesh Component Object Reference)
   > n153 = For Each Loop(Array:Array of Scene Component Object References=GetChildrenComponents!(self:Scene Component Object Reference=@n152, bIncludeAllDescendants:bool=true) -> (Children:Array of Scene Component Object References)) -> (Array Element:Scene Component Object Reference)
   > n154 = Cast To SkeletalMeshComponentBudgeted!(execute, Object=@n153) -> (then, CastFailed, AsSkeletal Mesh Component Budgeted:Skeletal Mesh Component Budgeted Object Reference) {is=title K2Node_DynamicCast}
   n155 = GetAttachSocketName!(self:Scene Component Object Reference=@n154) -> (ReturnValue:Name=None)
   Set Emote playing in world space(
    execute=@n156.then
    Emote playing in world space=Contains Item(TargetArray=Get world space emote anim attach bone(), ItemToFind="To String (Name)"(InName=@n155))
   )
   Set Is playing emote(execute=@n153.Completed, Is playing emote=true)
   n156 = Clear and Invalidate Timer by Handle(execute=@n161.then, Handle=Get Find Emote Attachment Timer())
   n157 = Contains(SearchIn=GetDisplayName(Object=@n154), Substring=GetDisplayName(Object=Get Emoting Pawn()))
   n158 = GetDisplayName(Object=@n154)
   n159 = Contains(SearchIn=@n158, Substring=Get Current Playing Emote Name() -> (Current Playing Emote Name:String), bSearchFromEnd=true)
   n160 = AND Boolean!(A=@n159, B=@n157) -> (ReturnValue)
   Branch(execute=@n154.then, Condition=@n160)
   > n161 = Set Current Emote Attach Socket name(Current Emote Attach Socket name:Name=@n155) -> (Output_Get:Name=None)
  ]}
 FindIKChainByName {nodes [
   n1 = FindIKChainByName!() -> (then, ChainName:Name, Ik Chains:Array of IKRetarget Blend To Source Chain Settings Structures) {is=title K2Node_FunctionEntry}
   > n2 = For Each Loop(Array:Array of IKRetarget Blend To Source Chain Settings Structures=@n1.Ik Chains) -> (Array Element:IKRetarget Blend To Source Chain Settings Structure)
   Return Node!(execute=@n2.Completed, Out Chain:IKRetarget Blend To Source Chain Settings Structure, Found Index:int=-1)
   > Branch(Condition="Equal (Name)"(
     A=Break IKRetarget Blend to Source Chain Settings!(IKRetargetBlendToSourceChainSettings:"IKRetarget Blend To Source Chain Settings Structure (by ref)"=@n2) -> (
      TargetChainName:Name
      BlendToSource:double=~
      TranslationAlpha:double=~
      RotationAlpha:double=~
      TranslationPerAxisAlpha:Vector=~
      ApplyPelvisOffset:double=~
     ) {is=title K2Node_BreakStruct}
     B=Get ChainName() -> (ChainName:Name=None)
    ))
   > Return Node(Out Chain=@n2, Found Index=@n2.Array Index)
  ]}
 FindStretchChainByName {nodes [
   n1 = FindStretchChainByName!() -> (then, Array:Array of Retarget Stretch Chain Settings Structures, ChainName:Name) {is=title K2Node_FunctionEntry}
   > n2 = For Each Loop with Break(Array:Array of Retarget Stretch Chain Settings Structures=@n1, Break=@n6.then) -> (Array Element:Retarget Stretch Chain Settings Structure)
   Return Node!(
    execute=@n2.Completed
    Array Element:Retarget Stretch Chain Settings Structure=Get Found Stretch() -> (Found Stretch:Retarget Stretch Chain Settings Structure)
    Found Index=Get Found Stretch Index() -> (Found Stretch Index:int=0)
   )
   n3 = Break Retarget Stretch Chain Settings!(RetargetStretchChainSettings:"Retarget Stretch Chain Settings Structure (by ref)"=@n2) -> (TargetChainName:Name, bEnabled:bool=~, MatchSourceLength:double=~, ScaleChainLength:double=~) {is=title K2Node_BreakStruct}
   n4 = Get ChainName() -> (ChainName=None)
   > n5 = Branch(Condition="Equal (Name)"(A=@n3, B=@n4))
   > Set Found Stretch(Found Stretch:Retarget Stretch Chain Settings Structure=@n2) -> (Output_Get:Retarget Stretch Chain Settings Structure)
   > Set Found Stretch Index(Found Stretch Index:int=@n2.Array Index) -> (Output_Get:int=0)
   > n6 = PrintString(InString=Append!(A="To String (Name)"(InName=@n4), B=___, C:String="To String (Boolean)"(InBool=@n3.bEnabled)) -> (ReturnValue))
   Set Found Stretch Index(execute=@n5.else, Found Stretch Index=-1) -> (Output_Get=0)
  ]}
 Interpolate Pelvis Offset {nodes [
   n1 = Interpolate Pelvis Offset!() -> (then, Target:Vector=~, InterpSpeed:float=~) {is=title K2Node_FunctionEntry}
   > n2 = Set CurrentPelvisTranslationOffset(CurrentPelvisTranslationOffset=VInterpTo!(
     Current:Vector=Get CurrentPelvisTranslationOffset()
     Target:Vector=@n1
     DeltaTime:float=Get Delta Time X()
     InterpSpeed:float=@n1.InterpSpeed
    ) -> (ReturnValue:Vector))
   > Return Node!(execute, Output:Vector=@n2)
  ]}
 Is in driver seat {nodes [
   Is in driver seat() {is=title K2Node_FunctionEntry}
   n1 = Get bIsAntelopeVehicleDriver()
   n2 = Get bIsFerretVehicleDriver()
   n3 = Get bIsGolfCartVehicleDriver() -> (bIsGolfCartVehicleDriver:bool)
   n4 = Get bIsMeatballDriver()
   n5 = Get bIsMotorcycleDriver()
   n6 = Get bIsOctopusVehicleDriver()
   n7 = Get bIsOstrichVehicleDriver() -> (bIsOstrichVehicleDriver:bool)
   n8 = Get bIsValetDriver() -> (bIsValetDriver:bool)
   > Return Node!(
    execute
    ReturnValue:bool=AND Boolean(A=Get bIsInVehicle(), B=OR Boolean!(A=@n1, B=@n2, C=@n3, D:bool=@n4, E:bool=@n5, F:bool=@n6, G:bool=@n7, H:bool=@n8) -> (ReturnValue))
   )
  ]}
 SetIKChainByName {nodes [
   n1 = SetIKChainByName!() -> (then, BlendToSourceController:IKRetarget Blend To Source Controller Object Reference, TargetChainName:Name, Blend To Source:double=~) {is=title K2Node_FunctionEntry}
   > Is Valid(InputObject=@n1)
   > n2 = GetSettings(self:IKRetarget Blend To Source Controller Object Reference=@n1) -> (ReturnValue:Blend To Source Settings Structure)
   > Set Current Settings(Current Settings:Blend To Source Settings Structure=@n2) -> (Output_Get:Blend To Source Settings Structure)
   > Set Chains(Chains:Array of IKRetarget Blend To Source Chain Settings Structures=Break Blend to Source Settings!(IKRetargetBlendToSourceOpSettings:"Blend To Source Settings Structure (by ref)"=@n2) -> (Chains:Array of IKRetarget Blend To Source Chain Settings Structures, Alpha:double=~, bEnabled:bool=~, LODThreshold:int) {is=title K2Node_BreakStruct}) -> (Output_Get:Array of IKRetarget Blend To Source Chain Settings Structures)
   > n3 = FindIKChainByName!(
    execute
    self
    ChainName:Name=Get TargetChainName() -> (TargetChainName:Name=None)
    Ik Chains:Array of IKRetarget Blend To Source Chain Settings Structures=Get Chains() -> (Chains:Array of IKRetarget Blend To Source Chain Settings Structures)
   ) -> (then, Out Chain:IKRetarget Blend To Source Chain Settings Structure, Found Index:int=0) {is=title K2Node_CallFunction}
   > Branch(Condition="Not Equal (Integer)"(A=@n3.Found Index, B=-1))
   > n4 = Set members in IKRetarget Blend to Source Chain Settings!(
    execute:Exec
    StructRef:"IKRetarget Blend To Source Chain Settings Structure (by ref)"=@n3
    BlendToSource:double=Get Blend To Source() -> (Blend To Source:double)
   ) -> (then:Exec, StructOut:"IKRetarget Blend To Source Chain Settings Structure (by ref)") {is=title K2Node_SetFieldsInStruct}
   > Set Array Elem(
    TargetArray:Array of IKRetarget Blend To Source Chain Settings Structures=Get Chains()
    Index=@n3.Found Index
    Item:"IKRetarget Blend To Source Chain Settings Structure (by ref)"=@n4
    bSizeToFit=false
   )
   > n5 = Set members in Blend to Source Settings!(
    execute:Exec
    StructRef:"Blend To Source Settings Structure (by ref)"=Get Current Settings() -> (Current Settings:Blend To Source Settings Structure)
    Chains:Array of IKRetarget Blend To Source Chain Settings Structures=Get Chains()
   ) -> (then:Exec, StructOut:"Blend To Source Settings Structure (by ref)") {is=title K2Node_SetFieldsInStruct}
   > SetSettings(
    self:IKRetarget Blend To Source Controller Object Reference=Get BlendToSourceController() -> (BlendToSourceController:IKRetarget Blend To Source Controller Object Reference)
    InSettings:Blend To Source Settings Structure=@n5
   )
  ]}
 SetPinBoneOffsets {nodes [
   n1 = SetPinBoneOffsets!() -> (then, Controller:IKRetarget Pin Bone Controller Object Reference, Global Offset:Vector=~, Local Offset:Vector=~) {is=title K2Node_FunctionEntry}
   > n2 = GetSettings(self:IKRetarget Pin Bone Controller Object Reference=@n1) -> (ReturnValue:Pin Bone Settings Structure)
   > n3 = Set members in Pin Bone Settings!(
    execute:Exec
    StructRef:"Pin Bone Settings Structure (by ref)"=@n2
    GlobalOffset:Transform=MakeTransform(Location=@n1.Global Offset, Rotation="0, 0, 0")
    LocalOffset:Transform=MakeTransform(Location=@n1.Local Offset, Rotation="0, 0, 0")
   ) -> (then:Exec, StructOut:"Pin Bone Settings Structure (by ref)") {is=title K2Node_SetFieldsInStruct}
   > SetSettings(
    self:IKRetarget Pin Bone Controller Object Reference=Get Controller() -> (Controller:IKRetarget Pin Bone Controller Object Reference)
    InSettings:Pin Bone Settings Structure=@n3
   )
  ]}
 SetStretchChainByName {nodes [
   n1 = SetStretchChainByName!() -> (
    then
    IkRetargetStretchChainController:IKRetarget Stretch Chain Controller Object Reference
    ChainName:Name
    EnableStretch:bool=~
    Match Source Length:double=~
   ) {is=title K2Node_FunctionEntry}
   n2 = GetSettings(execute=@n11.Is Valid, self:IKRetarget Stretch Chain Controller Object Reference=@n1) -> (ReturnValue:Stretch Chain Settings Structure)
   n3 = Set members in Stretch Chain Settings!(
    execute:Exec=@n9.then
    StructRef:"Stretch Chain Settings Structure (by ref)"=@n2
    ChainsToStretch:Array of Retarget Stretch Chain Settings Structures=Get ChainsToStretch() -> (ChainsToStretch:Array of Retarget Stretch Chain Settings Structures)
   ) -> (then:Exec, StructOut:"Stretch Chain Settings Structure (by ref)") {is=title K2Node_SetFieldsInStruct}
   n4 = SetSettings(
    self:IKRetarget Stretch Chain Controller Object Reference=Get IkRetargetStretchChainController() -> (IkRetargetStretchChainController:IKRetarget Stretch Chain Controller Object Reference)
    InSettings:Stretch Chain Settings Structure=@n3
   )
   Return Node!(execute=[@n4.then, @n3.then], Out Stretch Settings:"Stretch Chain Settings Structure (by ref)"=@n3)
   n5 = Break Stretch Chain Settings!(IKRetargetStretchChainOpSettings:"Stretch Chain Settings Structure (by ref)"=@n2) -> (
    IKRigAsset:IKRig Definition Object Reference
    ChainsToStretch:Array of Retarget Stretch Chain Settings Structures
    bEnabled:bool=~
    LODThreshold:int
   ) {is=title K2Node_BreakStruct}
   n6 = Set ChainsToStretch(execute=@n2.then, ChainsToStretch:Array of Retarget Stretch Chain Settings Structures=@n5.ChainsToStretch) -> (Output_Get:Array of Retarget Stretch Chain Settings Structures)
   > n7 = FindStretchChainByName!(execute, self, Array:Array of Retarget Stretch Chain Settings Structures=@n6, ChainName:Name=Get ChainName() -> (ChainName=None)) -> (then, Array Element:Retarget Stretch Chain Settings Structure, Found Index:int=0) {is=title K2Node_CallFunction}
   n8 = Set members in Retarget Stretch Chain Settings!(
    execute:Exec=@n10.then
    StructRef:"Retarget Stretch Chain Settings Structure (by ref)"=@n7
    bEnabled:bool=Get EnableStretch() -> (EnableStretch:bool)
    MatchSourceLength:double=Get Match Source Length() -> (Match Source Length:double)
   ) -> (then:Exec, StructOut:"Retarget Stretch Chain Settings Structure (by ref)") {is=title K2Node_SetFieldsInStruct}
   > n9 = Set Array Elem(
    TargetArray:Array of Retarget Stretch Chain Settings Structures=Get ChainsToStretch()
    Index=@n7.Found Index
    Item:"Retarget Stretch Chain Settings Structure (by ref)"=@n8
    bSizeToFit=false
   )
   n10 = Branch(execute=@n7.then, Condition="Not Equal (Integer)"(A=@n7.Found Index, B=-1))
   n11 = Is Valid(exec=@n1.then, InputObject=@n1)
  ]}
]
variables [
 Affect IK Horizontal {type=number}
 Affect IK Vertical {type=number}
 bHasIKRetargeterAsset {type=boolean}
 Body Intersect Controller {type=IKRetargetBodyIntersectController, default=null}
 Bones To Filter {type="array<FilterBoneData>", default []}
 Crotch Offset {type=number}
 Current Emote Attach Socket name {type=string, default=None}
 Current Emote Def {type=FortItemDefinition, default=null}
 Current Feet Sideways Offset {type=number}
 Current Playing Emote Name {type=string, default=""}
 Current Retarget Profile {
  type=RetargetProfile
  default {
   bApplyChainSettings=false
   bApplyGlobalSettings=false
   bApplyRootSettings=false
   bApplySourceRetargetPose=false
   bApplyTargetRetargetPose=false
   bForceAllIKOff=false
   sourceRetargetPoseName=None
   targetRetargetPoseName=None
   chainSettings {}
   globalSettings {
    bCopyBasePose=false
    bEnableFK=true
    bEnableIK=true
    bEnablePost=true
    bEnableRoot=true
    bWarping=false
    copyBasePoseRoot=None
    directionChain=None
    directionSource=Goals
    forwardDirection=Y
    sidewaysOffset=0
    sourceScaleFactor=1
    warpForwards=1
    warpSplay=1
   }
   retargetOpProfiles []
   rootSettings {
    affectIKHorizontal=1
    affectIKVertical=0
    blendToSource=0
    rotationAlpha=1
    scaleHorizontal=1
    scaleVertical=1
    translationAlpha=1
    blendToSourceWeights=<1 1 1>
    rotationOffset=<rot 0 0 0>
    translationOffset=<0 0 0>
   }
  }
 }
 CurrentPelvisTranslationOffset {type=Vector}
 CustomRetargetProfile {
  type=RetargetProfile
  default {
   bApplyChainSettings=false
   bApplyGlobalSettings=false
   bApplyRootSettings=false
   bApplySourceRetargetPose=false
   bApplyTargetRetargetPose=false
   bForceAllIKOff=false
   sourceRetargetPoseName=None
   targetRetargetPoseName=None
   chainSettings {}
   globalSettings {
    bCopyBasePose=false
    bEnableFK=true
    bEnableIK=true
    bEnablePost=true
    bEnableRoot=true
    bWarping=false
    copyBasePoseRoot=None
    directionChain=None
    directionSource=Goals
    forwardDirection=Y
    sidewaysOffset=0
    sourceScaleFactor=1
    warpForwards=1
    warpSplay=1
   }
   retargetOpProfiles []
   rootSettings {
    affectIKHorizontal=1
    affectIKVertical=0
    blendToSource=0
    rotationAlpha=1
    scaleHorizontal=1
    scaleVertical=1
    translationAlpha=1
    blendToSourceWeights=<1 1 1>
    rotationOffset=<rot 0 0 0>
    translationOffset=<0 0 0>
   }
  }
 }
 DBNO Ragdoll Tag {type=GameplayTag, default {tagName=Gameplay.Status.DBNO.TrendConcentrate.Ragdoll}}
 Default Retarget Profile {
  type=RetargetProfile
  default {
   bApplyChainSettings=false
   bApplyGlobalSettings=false
   bApplyRootSettings=false
   bApplySourceRetargetPose=false
   bApplyTargetRetargetPose=false
   bForceAllIKOff=false
   sourceRetargetPoseName=None
   targetRetargetPoseName=None
   chainSettings {}
   globalSettings {
    bCopyBasePose=false
    bEnableFK=true
    bEnableIK=true
    bEnablePost=true
    bEnableRoot=true
    bWarping=false
    copyBasePoseRoot=None
    directionChain=None
    directionSource=Goals
    forwardDirection=Y
    sidewaysOffset=0
    sourceScaleFactor=1
    warpForwards=1
    warpSplay=1
   }
   retargetOpProfiles []
   rootSettings {
    affectIKHorizontal=1
    affectIKVertical=0
    blendToSource=0
    rotationAlpha=1
    scaleHorizontal=1
    scaleVertical=1
    translationAlpha=1
    blendToSourceWeights=<1 1 1>
    rotationOffset=<rot 0 0 0>
    translationOffset=<0 0 0>
   }
  }
 }
 Delta Time X {type=number}
 Disable Intersect {type=boolean}
 Dist Half Point Lambda {type=number, default=0.25}
 Distance Threshold {type=number, default=70}
 Elbow Scale {type=number, default=1}
 Emote playing in world space {type=boolean}
 Emoting Pawn {type=FortPawn, default=null}
 Enable Floor Constraint {type=boolean}
 Feasibility Length Bias {type=number}
 Feet Sideways Offset Surfing Glider {type=number}
 Find Emote Attachment Timer {type=TimerHandle}
 Floor Constraint Chains {type="array<FloorConstraintChainSettings>", default []}
 Floor Constraint Settings {
  type=IKRetargetFloorConstraintOpSettings
  default {alpha=1, bDebugDraw=true, bEnabled=true, heightFalloffDistance=20, heightFalloffOffset=8, lODThreshold=-1, chainsToAffect []}
 }
 Foot IK Blend To Source {type=number}
 Hand IK Alpha Clamp {type=number, default=1}
 Hand IK De-Intersect {type=boolean, default=true}
 Hand IK Retargeting Weight {type=number}
 Hand IK Retargeting Weight Bias {type=number, default=0.5}
 Hand Intersect Settings {
  type=IKRetargetBodyIntersectIKOpSettings
  default {
   bDebugDraw=true
   bEnableGoalIntersect=true
   bEnablePoleVectorIntersect=true
   bEnabled=true
   lODThreshold=-1
   targetPhysicsAssetOverride=null
   intersectBodies []
   intersectGoalSettings []
   poleVectorIntersectSettings []
   propIntersectSettings []
  }
 }
 IKRetarget Blend To Source Controller {type=IKRetargetBlendToSourceController, default=null}
 IKRetargeterAsset {type=IKRetargeter, default=&/IKR_Biped_Base_IK}
 IKRetargetPelvisMotionController {type=IKRetargetPelvisMotionController, default=null}
 IKRetargetStretchChainController {type=IKRetargetStretchChainController, default=null}
 In Motorcycle Pelvis Translation Offset {type=Vector}
 In Vehicle Pelvis Translation Offset {type=Vector}
 Intersect Bodies {type="array<string>", default [head,spine_05,pelvis]}
 Intersect Goal Settings {
  type="array<IKGoalIntersectShapeSettings>"
  default [
   {
    bIncludeParentBodyIntersect=false
    blendMultiMaxOffset=0
    goal=LeftHandIK
    propIntersectBone=None
    goalShapeOffset=<0 0 0>
    goalShapeScale=<1 1 1>
   }
   {
    bIncludeParentBodyIntersect=false
    blendMultiMaxOffset=0
    goal=RightHandIK
    propIntersectBone=None
    goalShapeOffset=<0 0 0>
    goalShapeScale=<1 1 1>
   }
  ]
 }
 Intersection Retarget Profile {
  type=RetargetProfile
  default {
   bApplyChainSettings=false
   bApplyGlobalSettings=false
   bApplyRootSettings=false
   bApplySourceRetargetPose=false
   bApplyTargetRetargetPose=false
   bForceAllIKOff=false
   sourceRetargetPoseName=None
   targetRetargetPoseName=None
   chainSettings {}
   globalSettings {
    bCopyBasePose=false
    bEnableFK=true
    bEnableIK=true
    bEnablePost=true
    bEnableRoot=true
    bWarping=false
    copyBasePoseRoot=None
    directionChain=None
    directionSource=Goals
    forwardDirection=Y
    sidewaysOffset=0
    sourceScaleFactor=1
    warpForwards=1
    warpSplay=1
   }
   retargetOpProfiles []
   rootSettings {
    affectIKHorizontal=1
    affectIKVertical=0
    blendToSource=0
    rotationAlpha=1
    scaleHorizontal=1
    scaleVertical=1
    translationAlpha=1
    blendToSourceWeights=<1 1 1>
    rotationOffset=<rot 0 0 0>
    translationOffset=<0 0 0>
   }
  }
 }
 Is playing emote {type=boolean}
 Left Foot IK Floor Constraint {
  type=FloorConstraintChainSettings
  default {
   alpha=1
   bUseFoot=false
   bUseToes=false
   enableFloorConstraint=true
   maintainHeightOffset=1
   targetChainName=LeftFootIK
   foot {heelOffset=10, lateralOffset=10, medialOffset=10, toeOffset=10, verticalOffset=0}
   toes {
    alpha=1
    cutoffFrequency=8
    responsiveness=0.02
    velocityCutoffFrequency=1
    allToes [{length=4, verticalOffset=0, yawOffset=0, toeBone {boneName=ball_l}}]
   }
  }
 }
 Left Hand IK Alpha {type=number}
 Left Hand IK Blend To Source {type=number}
 No Hand and Elbow Intersect Settings {
  type=IKRetargetBodyIntersectIKOpSettings
  default {
   bDebugDraw=true
   bEnableGoalIntersect=true
   bEnablePoleVectorIntersect=true
   bEnabled=true
   lODThreshold=-1
   targetPhysicsAssetOverride=null
   intersectBodies []
   intersectGoalSettings []
   poleVectorIntersectSettings []
   propIntersectSettings []
  }
 }
 No Hand Intersect Settings {
  type=IKRetargetBodyIntersectIKOpSettings
  default {
   bDebugDraw=true
   bEnableGoalIntersect=true
   bEnablePoleVectorIntersect=true
   bEnabled=true
   lODThreshold=-1
   targetPhysicsAssetOverride=null
   intersectBodies []
   intersectGoalSettings []
   poleVectorIntersectSettings []
   propIntersectSettings []
  }
 }
 No Spine Stretch Settings {type=IKRetargetStretchChainOpSettings, default {bDebugDraw=true, bEnabled=true, iKRigAsset=null, lODThreshold=-1, chainsToStretch []}}
 no stretch chains {
  type="array<RetargetStretchChainSettings>"
  default [
   {bEnabled=false, matchSourceLength=0, scaleChainLength=1, targetChainName=root}
   {bEnabled=false, matchSourceLength=0, scaleChainLength=1, targetChainName=Spine}
   {bEnabled=false, matchSourceLength=0, scaleChainLength=1, targetChainName=Neck}
   {bEnabled=false, matchSourceLength=0, scaleChainLength=1, targetChainName=head}
   {bEnabled=false, matchSourceLength=0, scaleChainLength=1, targetChainName=LeftLeg}
   {bEnabled=false, matchSourceLength=0, scaleChainLength=1, targetChainName=LeftClavicle}
   {bEnabled=false, matchSourceLength=0, scaleChainLength=1, targetChainName=LeftArm}
   {bEnabled=false, matchSourceLength=0, scaleChainLength=1, targetChainName=LeftThumb}
   {bEnabled=false, matchSourceLength=0, scaleChainLength=1, targetChainName=LeftIndex}
   {bEnabled=false, matchSourceLength=0, scaleChainLength=1, targetChainName=LeftMiddle}
   {bEnabled=false, matchSourceLength=0, scaleChainLength=1, targetChainName=LeftRing}
   {bEnabled=false, matchSourceLength=0, scaleChainLength=1, targetChainName=LeftPinky}
   {bEnabled=false, matchSourceLength=0, scaleChainLength=1, targetChainName=LeftIndexMetacarpal}
   {bEnabled=false, matchSourceLength=0, scaleChainLength=1, targetChainName=LeftMiddleMetacarpal}
   {bEnabled=false, matchSourceLength=0, scaleChainLength=1, targetChainName=LeftRingMetacarpal}
   {bEnabled=false, matchSourceLength=0, scaleChainLength=1, targetChainName=LeftPinkyMetacarpal}
   {bEnabled=false, matchSourceLength=0, scaleChainLength=1, targetChainName=RightLeg}
   {bEnabled=false, matchSourceLength=0, scaleChainLength=1, targetChainName=RightClavicle}
   {bEnabled=false, matchSourceLength=0, scaleChainLength=1, targetChainName=RightArm}
   {bEnabled=false, matchSourceLength=0, scaleChainLength=1, targetChainName=RightThumb}
   {bEnabled=false, matchSourceLength=0, scaleChainLength=1, targetChainName=RightIndex}
   {bEnabled=false, matchSourceLength=0, scaleChainLength=1, targetChainName=RightMiddle}
   {bEnabled=false, matchSourceLength=0, scaleChainLength=1, targetChainName=RightRing}
   {bEnabled=false, matchSourceLength=0, scaleChainLength=1, targetChainName=RightPinky}
   {bEnabled=false, matchSourceLength=0, scaleChainLength=1, targetChainName=RightIndexMetacarpal}
   {bEnabled=false, matchSourceLength=0, scaleChainLength=1, targetChainName=RightMiddleMetacarpal}
   {bEnabled=false, matchSourceLength=0, scaleChainLength=1, targetChainName=RightRingMetacarpal}
   {bEnabled=false, matchSourceLength=0, scaleChainLength=1, targetChainName=RightPinkyMetacarpal}
   {bEnabled=false, matchSourceLength=0, scaleChainLength=1, targetChainName=LeftBall}
   {bEnabled=false, matchSourceLength=0, scaleChainLength=1, targetChainName=RightBall}
  ]
 }
 Pelvic Motion Controller {type=IKRetargetRelativePelvicMotionController, default=null}
 Pelvis Blend To Source Translation {type=number}
 Pelvis Offset Cape Glider {type=Vector}
 Pelvis Offset Hang Glider {type=Vector}
 Pelvis Offset No Glider Glider {type=Vector}
 Pelvis Offset Stairs Glider {type=Vector}
 Pelvis Offset Surfing Glider {type=Vector}
 Pelvis Offset Umbrella Glider {type=Vector}
 Pelvis Offset Vertical ZipLining {type=Vector}
 Pin Bone Prop Intersect Settings {type="array<IKPropIntersectSettings>", default []}
 Post Hand Intersect Settings {
  type=IKRetargetBodyIntersectIKOpSettings
  default {
   bDebugDraw=true
   bEnableGoalIntersect=true
   bEnablePoleVectorIntersect=true
   bEnabled=true
   lODThreshold=-1
   targetPhysicsAssetOverride=null
   intersectBodies []
   intersectGoalSettings []
   poleVectorIntersectSettings []
   propIntersectSettings []
  }
 }
 Prop Intersect Settings {
  type="array<IKPropIntersectSettings>"
  default [
   {bPinBone=false, boneName=weapon_l, capsuleLength=0, capsuleRadius=0, capsuleRotation=<rot 0 0 0>, capsuleTranslation=<0 0 0>}
   {bPinBone=false, boneName=weapon_r, capsuleLength=0, capsuleRadius=0, capsuleRotation=<rot 0 0 0>, capsuleTranslation=<0 0 0>}
   {bPinBone=false, boneName=root, capsuleLength=0, capsuleRadius=0, capsuleRotation=<rot 0 0 0>, capsuleTranslation=<0 0 0>}
   {bPinBone=false, boneName=attach, capsuleLength=0, capsuleRadius=0, capsuleRotation=<rot 0 0 0>, capsuleTranslation=<0 0 0>}
   {bPinBone=false, boneName=attach_emote_01, capsuleLength=0, capsuleRadius=0, capsuleRotation=<rot 0 0 0>, capsuleTranslation=<0 0 0>}
   {bPinBone=false, boneName=attach_emote_02, capsuleLength=0, capsuleRadius=0, capsuleRotation=<rot 0 0 0>, capsuleTranslation=<0 0 0>}
   {bPinBone=false, boneName=attach_emote_03, capsuleLength=0, capsuleRadius=0, capsuleRotation=<rot 0 0 0>, capsuleTranslation=<0 0 0>}
  ]
 }
 Regular Prop Intersect Settings {type="array<IKPropIntersectSettings>", default []}
 Relative Body Physics Asset {
  type=PhysicsAsset
  default=&/PA_Relative_M_MALE_Base
  class=~
  bodies=~
 }
 "Relative IK Settings (With Pins)" {
  type=IKRetargetRelativeIKOpSettings
  default {
   adaptiveTemporalSmoothing=false
   amplifyCloseDistance=3
   amplifyCloseThreshold=15
   amplifyScale=6
   bApplyImportanceWeighting=true
   bApplyPostSolvePinning=true
   bDebugDraw=true
   bDebugDrawBodyPairs=false
   bDebugDrawBodyTransforms=false
   bDebugDrawGoalContributions=false
   bDebugDrawPhysicsBodies=false
   bDebugDrawRetargetVertAverages=false
   bDryRun=false
   bEnableIntersectPushOut=true
   bEnabled=true
   bIgnoreSourceScale=true
   bSolveWithPropDamping=false
   bSolveWithPropScale=false
   blendPinsToSource=0
   contributionSumWeight=1
   distHalfPointLambda=0.15
   distanceThreshold=70
   distanceWeightMode=Quadratic
   feasibilityLengthBias=0
   floorConstraintScalar=10
   floorThreshold=5
   glueThreshold=5
   intersectOffsetScale=1
   intersectPhysicsAssetOverride=null
   intersectPushOutIter=1
   lODThreshold=-1
   multiBoneRegularization=0.001
   propBoneRegularization=0.001
   propDampingLearningRate=0.1
   propDampingRange=0.25
   propScalarDamping=500
   propScalarLearningRate=0.001
   propScalarRange=0.25
   targetPhysicsAssetOverride=null
   temporalSmoothingRadius=5
   bodyMapping {}
   debugFullRetargetPairBones []
   effectorEffectorEdges []
   effectorVolumes []
   intersectBodies []
   pinPropBones []
   trackingVolumes []
  }
 }
 "Relative IK Settings (Without Pins)" {
  type=IKRetargetRelativeIKOpSettings
  default {
   adaptiveTemporalSmoothing=false
   amplifyCloseDistance=3
   amplifyCloseThreshold=15
   amplifyScale=6
   bApplyImportanceWeighting=true
   bApplyPostSolvePinning=true
   bDebugDraw=true
   bDebugDrawBodyPairs=false
   bDebugDrawBodyTransforms=false
   bDebugDrawGoalContributions=false
   bDebugDrawPhysicsBodies=false
   bDebugDrawRetargetVertAverages=false
   bDryRun=false
   bEnableIntersectPushOut=true
   bEnabled=true
   bIgnoreSourceScale=true
   bSolveWithPropDamping=false
   bSolveWithPropScale=false
   blendPinsToSource=0
   contributionSumWeight=1
   distHalfPointLambda=0.15
   distanceThreshold=70
   distanceWeightMode=Quadratic
   feasibilityLengthBias=0
   floorConstraintScalar=10
   floorThreshold=5
   glueThreshold=5
   intersectOffsetScale=1
   intersectPhysicsAssetOverride=null
   intersectPushOutIter=1
   lODThreshold=-1
   multiBoneRegularization=0.001
   propBoneRegularization=0.001
   propDampingLearningRate=0.1
   propDampingRange=0.25
   propScalarDamping=500
   propScalarLearningRate=0.001
   propScalarRange=0.25
   targetPhysicsAssetOverride=null
   temporalSmoothingRadius=5
   bodyMapping {}
   debugFullRetargetPairBones []
   effectorEffectorEdges []
   effectorVolumes []
   intersectBodies []
   pinPropBones []
   trackingVolumes []
  }
 }
 Relative IKController {type=IKRetargetRelativeIKController, default=null}
 RetargetOverrideBlend {type=number}
 Right Foot IK Floor Constraint {
  type=FloorConstraintChainSettings
  default {
   alpha=1
   bUseFoot=false
   bUseToes=false
   enableFloorConstraint=true
   maintainHeightOffset=1
   targetChainName=RightFootIK
   foot {heelOffset=10, lateralOffset=10, medialOffset=10, toeOffset=10, verticalOffset=0}
   toes {
    alpha=1
    cutoffFrequency=8
    responsiveness=0.02
    velocityCutoffFrequency=1
    allToes [{length=4, verticalOffset=0, yawOffset=0, toeBone {boneName=ball_r}}]
   }
  }
 }
 Right Hand IK Alpha {type=number}
 Right Hand IK Blend To Source {type=number}
 Root Motion Controller {type=IKRetargetRootMotionController, default=null}
 Root Offset Pin Bone Controller {type=IKRetargetPinBoneController, default=null}
 spine stretch chains {
  type="array<RetargetStretchChainSettings>"
  default [
   {bEnabled=false, matchSourceLength=0, scaleChainLength=1, targetChainName=root}
   {bEnabled=true, matchSourceLength=1, scaleChainLength=1, targetChainName=Spine}
   {bEnabled=false, matchSourceLength=0, scaleChainLength=1, targetChainName=Neck}
   {bEnabled=false, matchSourceLength=0, scaleChainLength=1, targetChainName=head}
   {bEnabled=false, matchSourceLength=0, scaleChainLength=1, targetChainName=LeftLeg}
   {bEnabled=false, matchSourceLength=0, scaleChainLength=1, targetChainName=LeftClavicle}
   {bEnabled=false, matchSourceLength=0, scaleChainLength=1, targetChainName=LeftArm}
   {bEnabled=false, matchSourceLength=0, scaleChainLength=1, targetChainName=LeftThumb}
   {bEnabled=false, matchSourceLength=0, scaleChainLength=1, targetChainName=LeftIndex}
   {bEnabled=false, matchSourceLength=0, scaleChainLength=1, targetChainName=LeftMiddle}
   {bEnabled=false, matchSourceLength=0, scaleChainLength=1, targetChainName=LeftRing}
   {bEnabled=false, matchSourceLength=0, scaleChainLength=1, targetChainName=LeftPinky}
   {bEnabled=false, matchSourceLength=0, scaleChainLength=1, targetChainName=LeftIndexMetacarpal}
   {bEnabled=false, matchSourceLength=0, scaleChainLength=1, targetChainName=LeftMiddleMetacarpal}
   {bEnabled=false, matchSourceLength=0, scaleChainLength=1, targetChainName=LeftRingMetacarpal}
   {bEnabled=false, matchSourceLength=0, scaleChainLength=1, targetChainName=LeftPinkyMetacarpal}
   {bEnabled=false, matchSourceLength=0, scaleChainLength=1, targetChainName=RightLeg}
   {bEnabled=false, matchSourceLength=0, scaleChainLength=1, targetChainName=RightClavicle}
   {bEnabled=false, matchSourceLength=0, scaleChainLength=1, targetChainName=RightArm}
   {bEnabled=false, matchSourceLength=0, scaleChainLength=1, targetChainName=RightThumb}
   {bEnabled=false, matchSourceLength=0, scaleChainLength=1, targetChainName=RightIndex}
   {bEnabled=false, matchSourceLength=0, scaleChainLength=1, targetChainName=RightMiddle}
   {bEnabled=false, matchSourceLength=0, scaleChainLength=1, targetChainName=RightRing}
   {bEnabled=false, matchSourceLength=0, scaleChainLength=1, targetChainName=RightPinky}
   {bEnabled=false, matchSourceLength=0, scaleChainLength=1, targetChainName=RightIndexMetacarpal}
   {bEnabled=false, matchSourceLength=0, scaleChainLength=1, targetChainName=RightMiddleMetacarpal}
   {bEnabled=false, matchSourceLength=0, scaleChainLength=1, targetChainName=RightRingMetacarpal}
   {bEnabled=false, matchSourceLength=0, scaleChainLength=1, targetChainName=RightPinkyMetacarpal}
   {bEnabled=false, matchSourceLength=0, scaleChainLength=1, targetChainName=LeftBall}
   {bEnabled=false, matchSourceLength=0, scaleChainLength=1, targetChainName=RightBall}
  ]
 }
 Spine Stretch Settings {type=IKRetargetStretchChainOpSettings, default {bDebugDraw=true, bEnabled=true, iKRigAsset=null, lODThreshold=-1, chainsToStretch []}}
 SpineStretch {type=boolean}
 Stretch Spine {type=boolean}
 Stride Warping Controller {type=IKRetargetStrideWarpingController, default=null}
 Target Pose Name {type=string, default=Default Pose}
 Weapon classes to not use IK {type="array<Class>", default [/BRCosmetics/Weapons/FORT_Melee/Blueprints/B_Athena_Pickaxe_ChainsDual.B_Athena_Pickaxe_ChainsDual_C]}
 Weapon Offset {type=Vector}
 Weapon Scale {type=number, default=1}
 Weapons Should not use hand IK {type=boolean}
 Wind Surf Glider {type=SkeletalMesh, default=/BRCosmetics/Gadgets/Assets/VinderTech_GliderChute/Glider_Keen_Male/Meshes/Keen_Male_Glider}
 world space emote anim attach bone {type="array<string>", default [attach_emote_01, attach_emote_02, attach_emote_03, root, Root, attach, AttachSocket]}
]
