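# Control Rig module that drives twist bones at runtime, in the retarget fix-up folder for the ChimneyStack large male body.
# Construction resolves start, end and twist-bone connectors, enabling the forward solve only when all connect and both blend arrays match the bone count.
# The solve extracts local X-axis twist from initial pose, using the end bone's or the inverted start bone's.
# Each bone gets slerped twist and scaled end-bone translation.
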ControlRigBlueprint /BRCosmetics/Characters/Player/Male/Large/Bodies/M_LRG_ChimneyStack/IK_Retarget_New/FixUp/CRM_Runtime_Twist
graphs [
 ContainedGraph {nodes [
   n1 = RigVMFunctionEntryNode()
   GetTwistRotation = RigVMFunctionReferenceNode(EndBone=@n1.End Bone, ExecuteContext=@n1.Execute, StartBone=@n1.Start Bone, TwistFromEnd=@n1.Twist From End)
   > n2 = RigVMDispatchNode(Array=RigVMVariableNode() {variable=Twist Bones}.Value) {template="DISPATCH_RigVMDispatch_ArrayIterator(in Array,out Element,out Index,out Count,out Ratio)"}
   Interpolate = /Script/RigVM.RigVMFunction_MathQuaternionSlerp(
    A=(X=0,Y=0,Z=0,W=1)
    T=RigVMDispatchNode(Index=@n2.Index, Array=RigVMVariableNode() {variable=Twist Blend}.Value) {template="DISPATCH_RigVMDispatch_ArrayGetAtIndex(in Array,in Index,out Element)"}.Element
    B=@GetTwistRotation.OutTwistRotation
   ) {is=struct RigVMUnitNode, template="Interpolate::Execute(in A,in B,in T,out Result)"}
   DISPATCH_RigVMDispatch_ArrayGetAtIndex = RigVMDispatchNode(Index=@n2.Index, Array=RigVMVariableNode() {variable=Position Blend}.Value) {template="DISPATCH_RigVMDispatch_ArrayGetAtIndex(in Array,in Index,out Element)"}
   Scale = /Script/RigVM.RigVMFunction_MathVectorScale(Factor=@DISPATCH_RigVMDispatch_ArrayGetAtIndex.Element, Value=@GetTwistRotation.OutPosition) {is=struct RigVMUnitNode, template="Scale::Execute(in Value,in Factor,out Result)"}
   > RigUnit_SetTransform = &RigUnit_SetTransform(
    Space=LocalSpace
    Value=(Rotation=(X=-0.0434,Y=0.7058,Z=0.0434,W=-0.7058),Translation=(X=0,Y=-8.3248,Z=134.6484),Scale3D=(X=1,Y=1,Z=1))
    Weight=1
    bInitial=False
    bPropagateToChildren=false
    Value.Rotation=@Interpolate.Result
    Value.Translation=@Scale.Result
    Item=@n2.Element
   )
   RigVMFunctionReturnNode(Execute=@n2.Completed)
  ]}
 New Function_ContainedGraph {nodes [
   RigVMFunctionEntryNode()
   n1 = RigVMVariableNode() {variable=StartBone}
   n2 = RigVMVariableNode() {variable=StartBone}
   n3 = &RigVMFunction_MathQuaternionSwingTwist(
    TwistAxis=(X=1,Y=0,Z=0)
    Input=/Script/RigVM.RigVMFunction_MathQuaternionMakeRelative(
     Parent=&RigUnit_GetTransform(Space=LocalSpace, bInitial=true, Item=@n1.Value).Transform.Rotation
     Global=&RigUnit_GetTransform(Space=LocalSpace, Item=@n2.Value).Transform.Rotation
    ) {is=struct RigVMUnitNode, template="Make Relative::Execute(in Global,in Parent,out Local)"}.Local
   )
   n4 = &RigUnit_GetTransform(Space=LocalSpace, Item=RigVMVariableNode() {variable=EndBone}.Value)
   n5 = &RigVMFunction_MathQuaternionMakeRelative(
    Parent=&RigUnit_GetTransform(Space=LocalSpace, bInitial=true, Item=RigVMVariableNode() {variable=EndBone}.Value).Transform.Rotation
    Global=@n4.Transform.Rotation
   ) {template="Make Relative::Execute(in Global,in Parent,out Local)"}
   > RigVMFunctionReturnNode(
    OutTwistRotation=RigVMDispatchNode(
     False=/Script/RigVM.RigVMFunction_MathQuaternionInverse(Value=@n3.Twist) {is=struct RigVMUnitNode, template="Inverse::Execute(in Value,out Result)"}.Result
     True=&RigVMFunction_MathQuaternionSwingTwist(TwistAxis=(X=1,Y=0,Z=0), Input=@n5.Local).Twist
     Condition=RigVMVariableNode() {variable=TwistFromEnd}.Value
    ) {template="DISPATCH_RigVMDispatch_If(in Condition,in True,in False,out Result)"}.Result
    OutPosition=@n4.Transform.Translation
   )
  ]}
 RigVMFunctionLibrary {nodes [
   GetTwistRotation = RigVMCollapseNode(EndBone=(Type=Bone,Name="upperarm_l"), ExecuteContext=(), StartBone=(Type=Bone,Name="upperarm_l"), TwistFromEnd=False)
   SetTwistBones = RigVMCollapseNode(End Bone=(), Execute=(), Position Blend=(), Start Bone=(), Twist Blend=(), Twist Bones=())
  ]}
 RigVMFunction_MathBoolAnd_1_ContainedGraph {nodes [
   n1 = RigVMFunctionEntryNode()
   RigVMFunction_MathBoolAnd = &RigVMFunction_MathBoolAnd(A=@n1.A, B=@n1.B)
   n2 = &RigVMFunction_MathBoolAnd(B=@n1.C, A=@RigVMFunction_MathBoolAnd.Result)
   RigVMFunctionReturnNode(Result=&RigVMFunction_MathBoolAnd(B=@n1.E, A=&RigVMFunction_MathBoolAnd(B=@n1.D, A=@n2.Result).Result).Result)
  ]}
 RigVMModel {nodes [
   RigUnit_BeginExecution = &RigUnit_BeginExecution()
   RigVMFunction_ControlFlowBranch = &RigVMFunction_ControlFlowBranch(ExecuteContext=@RigUnit_BeginExecution.ExecutePin, Condition=RigVMVariableNode() {variable=Compute Forward}.Value)
   SetTwistBones = RigVMFunctionReferenceNode(
    Execute=@RigVMFunction_ControlFlowBranch.True
    Twist From End=RigVMVariableNode() {variable=Twist From End}.Value
    Position Blend=RigVMVariableNode() {variable=Position Blend}.Value
    Twist Blend=RigVMVariableNode() {variable=Twist Blend}.Value
    Start Bone=RigVMVariableNode() {variable=Start Bone}.Value
    End Bone=RigVMVariableNode() {variable=End Bone}.Value
    Twist Bones=RigVMVariableNode() {variable=Driven Bones}.Value
   )
  ]}
 RigVMModel Construction Event Graph {nodes [
   RigUnit_PrepareForExecution = &RigUnit_PrepareForExecution()
   n1 = /Script/ControlRig.RigUnit_ResolveArrayConnector(Connector=(Type=Connector,Name="TwistBones"), SkipSocket=False) {is=struct RigVMUnitNode}
   Num = RigVMDispatchNode(Array=@n1.Result) {template="DISPATCH_RigVMDispatch_ArrayGetNum(in Array,out Num)"}
   n2 = RigVMVariableNode() {variable=Twist Blend}
   n3 = RigVMVariableNode() {variable=Position Blend}
   n4 = /Script/ControlRig.RigUnit_ResolveConnector(Connector=(Type=Connector,Name="StartBone"), SkipSocket=False) {is=struct RigVMUnitNode}
   n5 = &RigUnit_ResolveConnector(Connector=(Type=Connector,Name="EndBone"), SkipSocket=False)
   RigVMFunction_MathBoolAnd = RigVMAggregateNode(
    D=RigVMDispatchNode(A=@Num.Num, B=RigVMDispatchNode(Array=@n2.Value) {template="DISPATCH_RigVMDispatch_ArrayGetNum(in Array,out Num)"}.Num) {template="DISPATCH_RigVMDispatch_CoreEquals(in A,in B,out Result)"}.Result
    E=RigVMDispatchNode(A=@Num.Num, B=RigVMDispatchNode(Array=@n3.Value) {template="DISPATCH_RigVMDispatch_ArrayGetNum(in Array,out Num)"}.Num) {template="DISPATCH_RigVMDispatch_CoreEquals(in A,in B,out Result)"}.Result
    C=@n1.bIsConnected
    A=@n4.bIsConnected
    B=@n5.bIsConnected
   )
   RigVMFunction_ControlFlowBranch = &RigVMFunction_ControlFlowBranch(ExecuteContext=@RigUnit_PrepareForExecution.ExecutePin, Condition=@RigVMFunction_MathBoolAnd.Result)
   RigVMVariableNode(Value=@n4.Result, ExecuteContext=@RigVMFunction_ControlFlowBranch.True) {variable=Start Bone}
   > RigVMVariableNode(Value=@n5.Result) {variable=End Bone}
   > RigVMVariableNode(Value=@n1.Result) {variable=Driven Bones}
   > RigVMVariableNode(Value=true) {variable=Compute Forward}
  ]}
]
hierarchy [
 StartBone {type=Connector, loc=<0 0 0>, rot=<0 0 0 1>}
 TwistBones {type=Connector, loc=<0 0 0>, rot=<0 0 0 1>}
 Root {type=Socket, loc=<0 0 0>, rot=<0 0 0 1>}
 EndBone {type=Connector, loc=<0 0 0>, rot=<0 0 0 1>}
]
variables [
 Compute Forward {type=bool, default=False}
 Driven Bones {type="TArray<FRigElementKey>"}
 End Bone {type=FRigElementKey, default=(Type=None,Name="")}
 Position Blend {type="TArray<double>"}
 Start Bone {type=FRigElementKey, default=(Type=None,Name="")}
 Twist Blend {type="TArray<double>"}
 Twist From End {type=bool, default=False}
]
